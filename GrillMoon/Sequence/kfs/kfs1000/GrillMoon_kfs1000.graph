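# GrillMoon kfs1000 master Level Sequence, a container that assembles one cinematic from per-department subsequences at 60 fps.
# Parallel sub tracks for animation, audio, camera, gameplay, lighting, techart, VFX and a general Subsequences track span the same frame range.
# A _kfs1000_MoverCube static mesh actor has an always-on spawn track and keyed location with constant rotation and scale.
# Its location keys all sit at negative frames, before the subsequence sections begin.
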LevelSequence /GrillMoon/Sequence/kfs/kfs1000/GrillMoon_kfs1000
bindings [_kfs1000_MoverCube {
  id=#1
  possessed_class=FortStaticMeshActor
  tracks [
   {type=MovieSceneSpawnTrack, sections [{type=MovieSceneSpawnSection, channels [None {type=MovieSceneScriptingBoolChannel, default=true}]}]}
   {
    type=MovieScene3DTransformTrack
    sections [{
      type=MovieScene3DTransformSection
      channels [
       Location.X {type=MovieSceneScriptingDoubleChannel, default=-55483.4353, keys [[-3600,-7293.8252],[-1613,-52191.3477],[-1530,-52191.3477]]}
       Location.Y {type=MovieSceneScriptingDoubleChannel, default=24660, keys [[-3600,24660],[-1613,55772.957],[-1530,55772.957]]}
       Location.Z {type=MovieSceneScriptingDoubleChannel, default=-2951, keys [[-1613,7942.1841],[-1530,5256.8037]]}
       Rotation.X {type=MovieSceneScriptingDoubleChannel, default=0, const [0,3]}
       Rotation.Y {type=MovieSceneScriptingDoubleChannel, default=0, const [0,3]}
       Rotation.Z {type=MovieSceneScriptingDoubleChannel, default=0, const [0,3]}
       Scale.X {type=MovieSceneScriptingDoubleChannel, default=1, const [1,3]}
       Scale.Y {type=MovieSceneScriptingDoubleChannel, default=1, const [1,3]}
       Scale.Z {type=MovieSceneScriptingDoubleChannel, default=1, const [1,3]}
      ]
     }]
   }
  ]
 }]
display_rate {denominator=1, numerator=60}
tick_resolution {denominator=1, numerator=24000}
tracks [
 {type=MovieSceneSubTrack, display_name=animation, sections [{type=MovieSceneSubSection, end_frame=5773, start_frame=0}]}
 {type=MovieSceneSubTrack, display_name=audio, sections [{type=MovieSceneSubSection, end_frame=5773, start_frame=0}]}
 {type=MovieSceneSubTrack, display_name=camera, sections [{type=MovieSceneSubSection, end_frame=5773, start_frame=0}]}
 {type=MovieSceneSubTrack, display_name=gameplay, sections [{type=MovieSceneSubSection, end_frame=5773, start_frame=0}]}
 {type=MovieSceneSubTrack, display_name=lighting, sections [{type=MovieSceneSubSection, end_frame=5773, start_frame=0}]}
 {type=MovieSceneSubTrack, display_name=techart, sections [{type=MovieSceneSubSection, end_frame=5773, start_frame=0}]}
 {type=MovieSceneSubTrack, display_name=vfx, sections [{type=MovieSceneSubSection, end_frame=5773, start_frame=0}]}
 {type=MovieSceneSubTrack, display_name=Subsequences, sections [{type=MovieSceneSubSection, end_frame=5773, start_frame=0}]}
]
