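# Animation Level Sequence for GrillMoon shot kfs1000, staging characters UnluckyRoll and MistMylar with axes, splinter blades and a mask gadget.
# Both characters play four back-to-back animation sections with matching timings, while attach tracks position characters and props alongside a 100x-scaled mover cube.
# UnluckyRoll's body, face and axes get scale-up overlay materials.
# Visibility keys toggle props and flicker UnluckyRoll late on; MistMylar's orb stays hidden.
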LevelSequence /GrillMoon/Sequence/kfs/kfs1000/GrillMoon_kfs1000_anm
bindings [
 UnluckyRoll_Axe_Secondary {
  id=#1
  possessed_class=FortSkeletalMeshActor
  tracks [
   {type=MovieSceneSpawnTrack, sections [{type=MovieSceneSpawnSection, channels [None {type=MovieSceneScriptingBoolChannel, default=true}]}]}
   {
    type=MovieScene3DTransformTrack
    sections [{
      type=MovieScene3DTransformSection
      channels [
       Location.X {type=MovieSceneScriptingDoubleChannel, default=-12.5451, keys [[-1735,-12.5451],[-1110,-12.5451],[-761,0]]}
       Location.Y {type=MovieSceneScriptingDoubleChannel, default=4.7349, keys [[-1735,4.7349],[-1110,4.7349],[-761,0]]}
       Location.Z {type=MovieSceneScriptingDoubleChannel, default=4.8322, keys [[-1735,4.8322],[-1110,4.8322],[-761,0]]}
       Rotation.X {type=MovieSceneScriptingDoubleChannel, default=-1.0289, keys [[-1735,-1.0289],[-1110,-1.0289],[-761,0]]}
       Rotation.Y {type=MovieSceneScriptingDoubleChannel, default=38.5806, keys [[-1735,38.5806],[-1110,38.5806],[-761,0]]}
       Rotation.Z {type=MovieSceneScriptingDoubleChannel, default=90, keys [[-1735,-24.8141],[-1110,-24.8141],[-761,0],[-538,90]]}
       Scale.X {type=MovieSceneScriptingDoubleChannel, default=1, const [1,3]}
       Scale.Y {type=MovieSceneScriptingDoubleChannel, default=1, const [1,3]}
       Scale.Z {type=MovieSceneScriptingDoubleChannel, default=1, const [1,3]}
      ]
     }]
   }
   {type=MovieSceneSkeletalAnimationTrack, sections []}
   {
    type=MovieSceneVisibilityTrack
    sections [{
      type=MovieSceneVisibilitySection
      channels [None {type=MovieSceneScriptingBoolChannel, default=false, keys [[-54,false],[3078,true],[3334,false]]}]
     }]
   }
   {type=MovieScene3DAttachTrack, sections [{type=MovieScene3DAttachSection}]}
  ]
 }
 Body {
  id=#2
  parent_id=#3
  possessed_class=SkeletalMeshComponentBudgeted
  tracks [{
    type=MovieSceneObjectPropertyTrack
    display_name=Overlay Material
    sections [{
      type=MovieSceneObjectPropertySection
      channels [None {
        type=MovieSceneScriptingObjectPathChannel
        default="<Object '/GrillMoon/Assets/TechArt/Character_ScaleUp/MI_GrillMoon_Overlay_UnluckyRoll_Body' (0x85051d18000) Class 'MaterialInstanceConstant'>"
        const [
         "<Object '&MI_GrillMoon_Overlay_UnluckyRoll_Body' (0x85051d18000) Class 'MaterialInstanceConstant'>"
         1
        ]
       }]
     }]
   }]
 }
 BP_GrillMoon_Countdown_SplinterBlade_A_R {
  id=#4
  possessed_class=BP_GrillMoon_Countdown_SplinterBlade_A_C
  tracks [
   {type=MovieSceneSpawnTrack, sections [{type=MovieSceneSpawnSection, channels [None {type=MovieSceneScriptingBoolChannel, default=true}]}]}
   {
    type=MovieScene3DTransformTrack
    sections [{
      type=MovieScene3DTransformSection
      channels [
       Location.X {type=MovieSceneScriptingDoubleChannel, default=0}
       Location.Y {type=MovieSceneScriptingDoubleChannel, default=0}
       Location.Z {type=MovieSceneScriptingDoubleChannel, default=0}
       Rotation.X {type=MovieSceneScriptingDoubleChannel, default=0}
       Rotation.Y {type=MovieSceneScriptingDoubleChannel, default=0}
       Rotation.Z {type=MovieSceneScriptingDoubleChannel, default=0}
       Scale.X {type=MovieSceneScriptingDoubleChannel, default=1}
       Scale.Y {type=MovieSceneScriptingDoubleChannel, default=1}
       Scale.Z {type=MovieSceneScriptingDoubleChannel, default=1}
      ]
     }]
   }
   {type=MovieScene3DAttachTrack, sections [{type=MovieScene3DAttachSection}]}
  ]
 }
 GrillMoon_UnluckyRoll {
  id=#3
  possessed_class=FortPlayerMannequin
  tracks [
   {
    type=MovieScene3DTransformTrack
    sections [{
      type=MovieScene3DTransformSection
      channels [
       Location.X {type=MovieSceneScriptingDoubleChannel, default=0, const [0,1]}
       Location.Y {type=MovieSceneScriptingDoubleChannel, default=0, const [0,1]}
       Location.Z {type=MovieSceneScriptingDoubleChannel, default=0, const [0,1]}
       Rotation.X {type=MovieSceneScriptingDoubleChannel, default=0, const [0,1]}
       Rotation.Y {type=MovieSceneScriptingDoubleChannel, default=0, const [0,1]}
       Rotation.Z {type=MovieSceneScriptingDoubleChannel, default=0, const [0,1]}
       Scale.X {type=MovieSceneScriptingDoubleChannel, default=1, const [1,1]}
       Scale.Y {type=MovieSceneScriptingDoubleChannel, default=1, const [1,1]}
       Scale.Z {type=MovieSceneScriptingDoubleChannel, default=1, const [1,1]}
      ]
     }]
   }
   {type=MovieSceneSpawnTrack, sections [{type=MovieSceneSpawnSection, channels [None {type=MovieSceneScriptingBoolChannel, default=true}]}]}
   {
    type=MovieSceneVisibilityTrack
    sections [{
      type=MovieSceneVisibilitySection
      channels [None {
        type=MovieSceneScriptingBoolChannel
        default=true
        keys [[-434,true],[2541,false],[2602,true],[5811,false],[9965,false],[9969,true],[9973,false],[9977,true],[9981,false],[9985,true],[9989,false],[9993,true],[9997,false],[10001,true],[10005,false],[10009,true],[10013,false],[10017,true],[10021,false],[10025,true],[10029,false]]
       }]
     }]
   }
   {
    type=MovieSceneSkeletalAnimationTrack
    sections [
     {type=MovieSceneSkeletalAnimationSection, end_frame=2100, start_frame=-303}
     {type=MovieSceneSkeletalAnimationSection, end_frame=3734, start_frame=2100}
     {type=MovieSceneSkeletalAnimationSection, end_frame=4082, start_frame=3733}
     {type=MovieSceneSkeletalAnimationSection, end_frame=5410, start_frame=4082}
    ]
   }
   {type=MovieScene3DAttachTrack, sections [{type=MovieScene3DAttachSection, end_frame=7727}]}
  ]
 }
 GrillMoon_Mask_Fire_Gadget {
  id=#5
  possessed_class=FortSkeletalMeshActor
  tracks [
   {type=MovieSceneSpawnTrack, sections [{type=MovieSceneSpawnSection, channels [None {type=MovieSceneScriptingBoolChannel, default=true}]}]}
   {
    type=MovieScene3DTransformTrack
    sections [{
      type=MovieScene3DTransformSection
      channels [
       Location.X {type=MovieSceneScriptingDoubleChannel, default=12.1633, const [12.1633,1]}
       Location.Y {type=MovieSceneScriptingDoubleChannel, default=-7.32, const [-7.32,1]}
       Location.Z {type=MovieSceneScriptingDoubleChannel, default=-3.2666, const [-3.2666,1]}
       Rotation.X {type=MovieSceneScriptingDoubleChannel, default=146.6006, const [146.6006,1]}
       Rotation.Y {type=MovieSceneScriptingDoubleChannel, default=-38.8868, const [-38.8868,1]}
       Rotation.Z {type=MovieSceneScriptingDoubleChannel, default=31.7401, const [31.7401,1]}
       Scale.X {type=MovieSceneScriptingDoubleChannel, default=0.5, const [0.5,1]}
       Scale.Y {type=MovieSceneScriptingDoubleChannel, default=0.5, const [0.5,1]}
       Scale.Z {type=MovieSceneScriptingDoubleChannel, default=0.5, const [0.5,1]}
      ]
     }]
   }
   {type=MovieSceneSkeletalAnimationTrack, sections []}
   {type=MovieScene3DAttachTrack, sections [{type=MovieScene3DAttachSection}]}
   {
    type=MovieSceneVisibilityTrack
    sections [{
      type=MovieSceneVisibilitySection
      channels [None {type=MovieSceneScriptingBoolChannel, default=true, keys [[2298,false],[2304,true],[2463,false]]}]
     }]
   }
  ]
 }
 SkeletalMeshComponent0 {
  id=#6
  parent_id=#7
  possessed_class=SkeletalMeshComponent
  tracks [{
    type=MovieSceneObjectPropertyTrack
    display_name=Overlay Material
    sections [{
      type=MovieSceneObjectPropertySection
      channels [None {
        type=MovieSceneScriptingObjectPathChannel
        default="<Object '/GrillMoon/Assets/TechArt/Character_ScaleUp/MI_GrillMoon_Overlay_UnluckyRoll_Weapon' (0x85051d19800) Class 'MaterialInstanceConstant'>"
        const [
         "<Object '&MI_GrillMoon_Overlay_UnluckyRoll_Weapon' (0x85051d19800) Class 'MaterialInstanceConstant'>"
         1
        ]
       }]
     }]
   }]
 }
 StaticMeshComponent0 {
  id=#8
  parent_id=#9
  possessed_class=BaseBuildingStaticMeshComponent
  tracks [{
    type=MovieSceneCustomPrimitiveDataTrack
    display_name=Custom Primitive Data
    sections [{
      type=MovieSceneCustomPrimitiveDataSection
      end_frame=7727
      start_frame=-303
      channels [1 {type=MovieSceneScriptingFloatChannel, default=0, const [0,1]}]
     }]
   }]
 }
 BP_GrillMoon_Countdown_SplinterBlade_A_L {
  id=#10
  possessed_class=BP_GrillMoon_Countdown_SplinterBlade_A_C
  tracks [
   {type=MovieSceneSpawnTrack, sections [{type=MovieSceneSpawnSection, channels [None {type=MovieSceneScriptingBoolChannel, default=true}]}]}
   {
    type=MovieScene3DTransformTrack
    sections [{
      type=MovieScene3DTransformSection
      channels [
       Location.X {type=MovieSceneScriptingDoubleChannel, default=0}
       Location.Y {type=MovieSceneScriptingDoubleChannel, default=0}
       Location.Z {type=MovieSceneScriptingDoubleChannel, default=0}
       Rotation.X {type=MovieSceneScriptingDoubleChannel, default=0}
       Rotation.Y {type=MovieSceneScriptingDoubleChannel, default=0}
       Rotation.Z {type=MovieSceneScriptingDoubleChannel, default=0}
       Scale.X {type=MovieSceneScriptingDoubleChannel, default=1}
       Scale.Y {type=MovieSceneScriptingDoubleChannel, default=1}
       Scale.Z {type=MovieSceneScriptingDoubleChannel, default=1}
      ]
     }]
   }
   {type=MovieScene3DAttachTrack, sections [{type=MovieScene3DAttachSection}]}
  ]
 }
 UnluckyRoll_Axe {
  id=#7
  possessed_class=FortSkeletalMeshActor
  tracks [
   {type=MovieSceneSpawnTrack, sections [{type=MovieSceneSpawnSection, channels [None {type=MovieSceneScriptingBoolChannel, default=true}]}]}
   {
    type=MovieScene3DTransformTrack
    sections [{
      type=MovieScene3DTransformSection
      channels [
       Location.X {type=MovieSceneScriptingDoubleChannel, default=-12.5451, const [0,3]}
       Location.Y {type=MovieSceneScriptingDoubleChannel, default=4.7349, const [0,3]}
       Location.Z {type=MovieSceneScriptingDoubleChannel, default=4.8322, const [0,3]}
       Rotation.X {type=MovieSceneScriptingDoubleChannel, default=-1.0289, const [0,3]}
       Rotation.Y {type=MovieSceneScriptingDoubleChannel, default=38.5806, const [0,3]}
       Rotation.Z {type=MovieSceneScriptingDoubleChannel, default=270, keys [[1903,270],[1912,90],[1922,90]]}
       Scale.X {type=MovieSceneScriptingDoubleChannel, default=1, const [1,3]}
       Scale.Y {type=MovieSceneScriptingDoubleChannel, default=1, const [1,3]}
       Scale.Z {type=MovieSceneScriptingDoubleChannel, default=1, const [1,3]}
      ]
     }]
   }
   {type=MovieSceneSkeletalAnimationTrack, sections []}
   {type=MovieScene3DAttachTrack, sections [{type=MovieScene3DAttachSection}]}
   {
    type=MovieSceneVisibilityTrack
    sections [{
      type=MovieSceneVisibilitySection
      channels [None {type=MovieSceneScriptingBoolChannel, default=false, keys [[-538,false],[-342,true],[2541,false],[2602,true],[6430,false]]}]
     }]
   }
  ]
 }
 BP_GrillMoon_MistMylar_Orb {
  id=#9
  possessed_class=BP_MystMylar_Orb_C
  tracks [{
    type=MovieSceneVisibilityTrack
    sections [{type=MovieSceneVisibilitySection, channels [None {type=MovieSceneScriptingBoolChannel, default=false}]}]
   }]
 }
 GrillMoon_MistMylar {
  id=#11
  possessed_class=FortPlayerMannequin
  tracks [
   {type=MovieSceneSpawnTrack, sections [{type=MovieSceneSpawnSection, channels [None {type=MovieSceneScriptingBoolChannel, default=true}]}]}
   {
    type=MovieScene3DTransformTrack
    sections [{
      type=MovieScene3DTransformSection
      channels [
       Location.X {type=MovieSceneScriptingDoubleChannel, default=0, const [0,2]}
       Location.Y {type=MovieSceneScriptingDoubleChannel, default=0, const [0,2]}
       Location.Z {type=MovieSceneScriptingDoubleChannel, default=0, const [0,2]}
       Rotation.X {type=MovieSceneScriptingDoubleChannel, default=0, const [0,2]}
       Rotation.Y {type=MovieSceneScriptingDoubleChannel, default=0, const [0,2]}
       Rotation.Z {type=MovieSceneScriptingDoubleChannel, default=0, const [0,2]}
       Scale.X {type=MovieSceneScriptingDoubleChannel, default=1, const [1,2]}
       Scale.Y {type=MovieSceneScriptingDoubleChannel, default=1, const [1,2]}
       Scale.Z {type=MovieSceneScriptingDoubleChannel, default=1, const [1,2]}
      ]
     }]
   }
   {
    type=MovieSceneSkeletalAnimationTrack
    sections [
     {type=MovieSceneSkeletalAnimationSection, end_frame=2100, start_frame=-303}
     {type=MovieSceneSkeletalAnimationSection, end_frame=3734, start_frame=2100}
     {type=MovieSceneSkeletalAnimationSection, end_frame=4082, start_frame=3733}
     {type=MovieSceneSkeletalAnimationSection, end_frame=5410, start_frame=4082}
    ]
   }
   {type=MovieScene3DAttachTrack, sections [{type=MovieScene3DAttachSection}]}
  ]
 }
 Face {
  id=#12
  parent_id=#3
  possessed_class=SkeletalMeshComponentBudgeted
  tracks [
   {
    type=MovieScene3DTransformTrack
    sections [{
      type=MovieScene3DTransformSection
      channels [
       Location.X {type=MovieSceneScriptingDoubleChannel, default=0}
       Location.Y {type=MovieSceneScriptingDoubleChannel, default=0}
       Location.Z {type=MovieSceneScriptingDoubleChannel, default=0}
       Rotation.X {type=MovieSceneScriptingDoubleChannel, default=0}
       Rotation.Y {type=MovieSceneScriptingDoubleChannel, default=0}
       Rotation.Z {type=MovieSceneScriptingDoubleChannel, default=0}
       Scale.X {type=MovieSceneScriptingDoubleChannel, default=1}
       Scale.Y {type=MovieSceneScriptingDoubleChannel, default=1}
       Scale.Z {type=MovieSceneScriptingDoubleChannel, default=1}
      ]
     }]
   }
   {
    type=MovieSceneObjectPropertyTrack
    display_name=Overlay Material
    sections [{
      type=MovieSceneObjectPropertySection
      channels [None {
        type=MovieSceneScriptingObjectPathChannel
        default="<Object '/GrillMoon/Assets/TechArt/Character_ScaleUp/MI_GrillMoon_Overlay_UnluckyRoll_Acc' (0x85051d19000) Class 'MaterialInstanceConstant'>"
        const [
         "<Object '&MI_GrillMoon_Overlay_UnluckyRoll_Acc' (0x85051d19000) Class 'MaterialInstanceConstant'>"
         1
        ]
       }]
     }]
   }
  ]
 }
 _kfs1000_MoverCube_Scale {
  id=#13
  possessed_class=FortStaticMeshActor
  tracks [
   {
    type=MovieScene3DTransformTrack
    sections [{
      type=MovieScene3DTransformSection
      channels [
       Location.X {type=MovieSceneScriptingDoubleChannel, default=-55483.4353, const [-65486.8281,1]}
       Location.Y {type=MovieSceneScriptingDoubleChannel, default=24660, const [60478.625,1]}
       Location.Z {type=MovieSceneScriptingDoubleChannel, default=-2951, const [10514.4639,1]}
       Rotation.X {type=MovieSceneScriptingDoubleChannel, default=0, const [0,1]}
       Rotation.Y {type=MovieSceneScriptingDoubleChannel, default=0, const [0,1]}
       Rotation.Z {type=MovieSceneScriptingDoubleChannel, default=0, const [159.0275,1]}
       Scale.X {type=MovieSceneScriptingDoubleChannel, default=1, const [100,1]}
       Scale.Y {type=MovieSceneScriptingDoubleChannel, default=1, const [100,1]}
       Scale.Z {type=MovieSceneScriptingDoubleChannel, default=1, const [100,1]}
      ]
     }]
   }
   {type=MovieSceneSpawnTrack, sections [{type=MovieSceneSpawnSection, channels [None {type=MovieSceneScriptingBoolChannel, default=true}]}]}
  ]
 }
 Face {
  id=#14
  parent_id=#11
  possessed_class=SkeletalMeshComponentBudgeted
  tracks [{
    type=MovieScene3DTransformTrack
    sections [{
      type=MovieScene3DTransformSection
      channels [
       Location.X {type=MovieSceneScriptingDoubleChannel, default=0}
       Location.Y {type=MovieSceneScriptingDoubleChannel, default=0}
       Location.Z {type=MovieSceneScriptingDoubleChannel, default=0}
       Rotation.X {type=MovieSceneScriptingDoubleChannel, default=0}
       Rotation.Y {type=MovieSceneScriptingDoubleChannel, default=0}
       Rotation.Z {type=MovieSceneScriptingDoubleChannel, default=0}
       Scale.X {type=MovieSceneScriptingDoubleChannel, default=1}
       Scale.Y {type=MovieSceneScriptingDoubleChannel, default=1}
       Scale.Z {type=MovieSceneScriptingDoubleChannel, default=1}
      ]
     }]
   }]
 }
 SkeletalMeshComponent0 {
  id=#15
  parent_id=#1
  possessed_class=SkeletalMeshComponent
  tracks [{
    type=MovieSceneObjectPropertyTrack
    display_name=Overlay Material
    sections [{
      type=MovieSceneObjectPropertySection
      channels [None {
        type=MovieSceneScriptingObjectPathChannel
        default="<Object '&MI_GrillMoon_Overlay_UnluckyRoll_Weapon' (0x85051d19800) Class 'MaterialInstanceConstant'>"
        const [
         "<Object '&MI_GrillMoon_Overlay_UnluckyRoll_Weapon' (0x85051d19800) Class 'MaterialInstanceConstant'>"
         1
        ]
       }]
     }]
   }]
 }
]
display_rate {denominator=1, numerator=60}
tick_resolution {denominator=1, numerator=24000}
tracks []
