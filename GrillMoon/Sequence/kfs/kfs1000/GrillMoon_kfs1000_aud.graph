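# GrillMoon_kfs1000_aud is the audio Level Sequence for cinematic shot kfs1000 in GrillMoon.
# Separate audio tracks carry UnluckyRoll and MistMylar VO and SFX, a portal, player lasso with hit markers, and a sweetener.
# Each track holds one section over a shared range, with Volume, Pitch and AttachActor channels at defaults.
# Bindings for a dragon gate tile and three mannequin characters have no tracks; the lasso and sweetener sections carry attach-actor binding references.
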LevelSequence /GrillMoon/Sequence/kfs/kfs1000/GrillMoon_kfs1000_aud
bindings [
 BSG_ClayTile_DragonGate_B_GrillMoon {id=#1, possessed_class=BSG_ClayTile_DragonGate_GrillMoon_C, tracks []}
 GrillMoon_UnluckyRoll {id=#2, possessed_class=FortPlayerMannequin, tracks []}
 GrillMoon_MistMylar {id=#3, possessed_class=FortPlayerMannequin, tracks []}
 GrillMoon_Countdown_MistMylar {id=#4, possessed_class=FortPlayerMannequin, tracks []}
]
display_rate {denominator=1, numerator=60}
tick_resolution {denominator=1, numerator=24000}
tracks [
 {
  type=MovieSceneAudioTrack
  display_name=Player Lasso Hit Markers
  sections [{
    type=MovieSceneAudioSection
    end_frame=5884
    start_frame=-531
    channels [
     Volume {type=MovieSceneScriptingFloatChannel, default=1}
     Pitch {type=MovieSceneScriptingFloatChannel, default=1}
     AttachActor {type=MovieSceneScriptingActorReferenceChannel, default="<Struct 'MovieSceneObjectBindingID' (0x8509d96d400) {}>"}
    ]
   }]
 }
 {
  type=MovieSceneAudioTrack
  display_name=UnluckyRoll_VO
  sections [{
    type=MovieSceneAudioSection
    end_frame=5884
    start_frame=-531
    channels [
     Volume {type=MovieSceneScriptingFloatChannel, default=1}
     Pitch {type=MovieSceneScriptingFloatChannel, default=1}
     AttachActor {type=MovieSceneScriptingActorReferenceChannel}
    ]
   }]
 }
 {
  type=MovieSceneAudioTrack
  display_name=MistMylar_VO
  sections [{
    type=MovieSceneAudioSection
    end_frame=5884
    start_frame=-531
    channels [
     Volume {type=MovieSceneScriptingFloatChannel, default=1}
     Pitch {type=MovieSceneScriptingFloatChannel, default=1}
     AttachActor {type=MovieSceneScriptingActorReferenceChannel}
    ]
   }]
 }
 {
  type=MovieSceneAudioTrack
  display_name=Portal
  sections [{
    type=MovieSceneAudioSection
    end_frame=5884
    start_frame=-531
    channels [
     Volume {type=MovieSceneScriptingFloatChannel, default=1}
     Pitch {type=MovieSceneScriptingFloatChannel, default=1}
     AttachActor {type=MovieSceneScriptingActorReferenceChannel}
    ]
   }]
 }
 {
  type=MovieSceneAudioTrack
  display_name=Sweetener
  sections [{
    type=MovieSceneAudioSection
    end_frame=5884
    start_frame=-531
    channels [
     Volume {type=MovieSceneScriptingFloatChannel, default=1}
     Pitch {type=MovieSceneScriptingFloatChannel, default=1}
     AttachActor {type=MovieSceneScriptingActorReferenceChannel, default="<Struct 'MovieSceneObjectBindingID' (0x8509d96de20) {}>"}
    ]
   }]
 }
 {
  type=MovieSceneAudioTrack
  display_name=MistMylar_SFX
  sections [{
    type=MovieSceneAudioSection
    end_frame=5884
    start_frame=-531
    channels [
     Volume {type=MovieSceneScriptingFloatChannel, default=1}
     Pitch {type=MovieSceneScriptingFloatChannel, default=1}
     AttachActor {type=MovieSceneScriptingActorReferenceChannel}
    ]
   }]
 }
 {
  type=MovieSceneAudioTrack
  display_name=UnluckyRoll_SFX
  sections [{
    type=MovieSceneAudioSection
    end_frame=5884
    start_frame=-531
    channels [
     Volume {type=MovieSceneScriptingFloatChannel, default=1}
     Pitch {type=MovieSceneScriptingFloatChannel, default=1}
     AttachActor {type=MovieSceneScriptingActorReferenceChannel}
    ]
   }]
 }
 {
  type=MovieSceneAudioTrack
  display_name=Player Lasso
  sections [{
    type=MovieSceneAudioSection
    end_frame=5884
    start_frame=-531
    channels [
     Volume {type=MovieSceneScriptingFloatChannel, default=1}
     Pitch {type=MovieSceneScriptingFloatChannel, default=1}
     AttachActor {type=MovieSceneScriptingActorReferenceChannel, default="<Struct 'MovieSceneObjectBindingID' (0x8509d96d260) {}>"}
    ]
   }]
 }
]
