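# Level Sequence that times camera shake and controller rumble for the GrillMoon kfs1000 sequence; it holds no cameras or cuts.
# A single Events_Shared_cameraForce_BP binding, with spawn and transform tracks, carries everything.
# Keyed object-property tracks swap the Force Effect and Cam Shake Base assets, while Lobby event triggers keyed at mostly matching times fire them.
# The Radius trigger tracks have no timed keys, so timing runs through the Lobby events.
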LevelSequence /GrillMoon/Sequence/kfs/kfs1000/GrillMoon_kfs1000_cam
bindings [Events_Shared_cameraForce_BP {
  id=#1
  possessed_class=Events_Shared_cameraForce_BP_C
  tracks [
   {
    type=MovieSceneObjectPropertyTrack
    display_name=Force Effect
    sections [{
      type=MovieSceneObjectPropertySection
      channels [None {
        type=MovieSceneScriptingObjectPathChannel
        default="<Object '/GrillMoon/ForcedFeedback/GrillMoon_kfs1000_DistantClashBump' (0x84c3c705100) Class 'ForceFeedbackEffect'>"
        keys [
         [-858, "<Object '&Events_Shared_small_FF' (0x84c3c703480) Class 'ForceFeedbackEffect'>"]
         [143, "<Object '/GrillMoon/ForcedFeedback/GrillMoon_FF_kfs1000_PortalPull_1of2' (0x84c3c704800) Class 'ForceFeedbackEffect'>"]
         [459, "<Object '/Game/Weapons/ForceFeedback/FF_Generic_MeleeImpact' (0x84c4dc66a50) Class 'ForceFeedbackEffect'>"]
         [508, "<Object '/Game/Weapons/ForceFeedback/FF_ShoppingCartBump' (0x84c3c705080) Class 'ForceFeedbackEffect'>"]
         [588, "<Object '&FF_ShoppingCartBump' (0x84c3c705080) Class 'ForceFeedbackEffect'>"]
         [686, "<Object '&FF_ShoppingCartBump' (0x84c3c705080) Class 'ForceFeedbackEffect'>"]
         [715, "<Object '&FF_ShoppingCartBump' (0x84c3c705080) Class 'ForceFeedbackEffect'>"]
         [733, "<Object '&FF_ShoppingCartBump' (0x84c3c705080) Class 'ForceFeedbackEffect'>"]
         [759, "<Object '&FF_ShoppingCartBump' (0x84c3c705080) Class 'ForceFeedbackEffect'>"]
         [804, "<Object '&FF_ShoppingCartBump' (0x84c3c705080) Class 'ForceFeedbackEffect'>"]
         [856, "<Object '&FF_ShoppingCartBump' (0x84c3c705080) Class 'ForceFeedbackEffect'>"]
         [901, "<Object '&FF_ShoppingCartBump' (0x84c3c705080) Class 'ForceFeedbackEffect'>"]
         [933, "<Object '&FF_ShoppingCartBump' (0x84c3c705080) Class 'ForceFeedbackEffect'>"]
         [980, "<Object '&FF_ShoppingCartBump' (0x84c3c705080) Class 'ForceFeedbackEffect'>"]
         [1069, "<Object '&FF_Generic_MeleeImpact' (0x84c4dc66a50) Class 'ForceFeedbackEffect'>"]
         [1249, "<Object '&FF_Generic_MeleeImpact' (0x84c4dc66a50) Class 'ForceFeedbackEffect'>"]
         [1342, "<Object '&FF_ShoppingCartBump' (0x84c3c705080) Class 'ForceFeedbackEffect'>"]
         [1377, "<Object '&GrillMoon_kfs1000_DistantClashBump' (0x84c3c705100) Class 'ForceFeedbackEffect'>"]
         [1436, "<Object '&GrillMoon_kfs1000_DistantClashBump' (0x84c3c705100) Class 'ForceFeedbackEffect'>"]
         [1481, "<Object '&FF_ShoppingCartBump' (0x84c3c705080) Class 'ForceFeedbackEffect'>"]
         [1634, "<Object '&FF_Generic_MeleeImpact' (0x84c4dc66a50) Class 'ForceFeedbackEffect'>"]
         [1805, "<Object '&FF_Generic_MeleeImpact' (0x84c4dc66a50) Class 'ForceFeedbackEffect'>"]
         [1829, "<Object '&FF_Generic_MeleeImpact' (0x84c4dc66a50) Class 'ForceFeedbackEffect'>"]
         [2031, "<Object '&GrillMoon_FF_kfs1000_PortalPull_1of2' (0x84c3c704800) Class 'ForceFeedbackEffect'>"]
         [2479, "<Object '&FF_ShoppingCartBump' (0x84c3c705080) Class 'ForceFeedbackEffect'>"]
         [2556, "<Object '/Game/Weapons/ForceFeedback/FF_ATVBump' (0x84c4dc669e0) Class 'ForceFeedbackEffect'>"]
         [2580, "<Object '&FF_ATVBump' (0x84c4dc669e0) Class 'ForceFeedbackEffect'>"]
         [2612, "<Object '&FF_ShoppingCartBump' (0x84c3c705080) Class 'ForceFeedbackEffect'>"]
         [2722, "<Object '/Game/Weapons/ForceFeedback/FF_Generic_Impact_Near' (0x84c4dc66350) Class 'ForceFeedbackEffect'>"]
         [2820, "<Object '/Game/Weapons/ForceFeedback/FF_Fire_Med_R' (0x84c4dc66900) Class 'ForceFeedbackEffect'>"]
         [2960, "<Object '/Game/Weapons/ForceFeedback/FF_Fire_Burst_LR' (0x84c4dc663c0) Class 'ForceFeedbackEffect'>"]
         [3076, "<Object '/Game/Weapons/ForceFeedback/FF_Generic_Melee' (0x84c4dc66dd0) Class 'ForceFeedbackEffect'>"]
         [3118, "<Object '&FF_ShoppingCartBump' (0x84c3c705080) Class 'ForceFeedbackEffect'>"]
         [3179, "<Object '&FF_ShoppingCartBump' (0x84c3c705080) Class 'ForceFeedbackEffect'>"]
         [3229, "<Object '&FF_ShoppingCartBump' (0x84c3c705080) Class 'ForceFeedbackEffect'>"]
         [3275, "<Object '/Game/Weapons/ForceFeedback/FF_Fire_VeryHeavy_LR' (0x84c4dc66430) Class 'ForceFeedbackEffect'>"]
         [3310, "<Object '&FF_Generic_Impact_Near' (0x84c4dc66350) Class 'ForceFeedbackEffect'>"]
         [3537, "<Object '&GrillMoon_kfs1000_DistantClashBump' (0x84c3c705100) Class 'ForceFeedbackEffect'>"]
         [3563, "<Object '&FF_ShoppingCartBump' (0x84c3c705080) Class 'ForceFeedbackEffect'>"]
         [3689, "<Object '/Game/Weapons/ForceFeedback/FF_Shotgun_ShellIn' (0x84c3c705180) Class 'ForceFeedbackEffect'>"]
         [3785, "<Object '/Game/Weapons/ForceFeedback/FF_Generic_MeleeImpact_Weakspot' (0x84c4dc66890) Class 'ForceFeedbackEffect'>"]
         [4237, "<Object '/GrillMoon/ForcedFeedback/GrillMoon_FF_kfs1000_PortalPull_2of2' (0x84c3c704700) Class 'ForceFeedbackEffect'>"]
         [4553, "<Object '/Game/Weapons/ForceFeedback/FF_Generic_Impact_Far' (0x84c4dc66d60) Class 'ForceFeedbackEffect'>"]
         [4613, "<Object '&FF_Generic_Impact_Near' (0x84c4dc66350) Class 'ForceFeedbackEffect'>"]
         [4883, "<Object '/FNGameplayCues/Weapons/ForceFeedback/FF_Generic_Whoosh_Small' (0x84c4dc693e0) Class 'ForceFeedbackEffect'>"]
         [4972, "<Object '&FF_Generic_Whoosh_Small' (0x84c4dc693e0) Class 'ForceFeedbackEffect'>"]
         [5032, "<Object '/Game/Weapons/ForceFeedback/FF_GolfCartBump' (0x84c3c705000) Class 'ForceFeedbackEffect'>"]
         [5111, "<Object '&FF_Generic_Whoosh_Small' (0x84c4dc693e0) Class 'ForceFeedbackEffect'>"]
        ]
       }]
     }]
   }
   {
    type=MovieSceneObjectPropertyTrack
    display_name=Cam Shake Base
    sections [{
      type=MovieSceneObjectPropertySection
      channels [None {
        type=MovieSceneScriptingObjectPathChannel
        default="<Object '&Events_Shared_00small_csb_C' (0x84c3c6a8a00) Class 'BlueprintGeneratedClass'>"
        keys [
         [-858, "<Object '/GrillMoon/CameraShake/GrillMoon_CharTele_csb.GrillMoon_CharTele_csb_C' (0x84c3c6ade00) Class 'BlueprintGeneratedClass'>"]
         [
          143
          "<Object '/GrillMoon/CameraShake/GrillMoon_kfs1000_pullfov_csb.GrillMoon_kfs1000_pullfov_csb_C' (0x85051a30600) Class 'BlueprintGeneratedClass'>"
         ]
         [
          459
          "<Object '/GrillMoon/CameraShake/GrillMoon_kfs1000_soft_csb.GrillMoon_kfs1000_soft_csb_C' (0x84c3c6aea00) Class 'BlueprintGeneratedClass'>"
         ]
         [
          1068
          "<Object '&GrillMoon_kfs1000_soft_csb_C' (0x84c3c6aea00) Class 'BlueprintGeneratedClass'>"
         ]
         [
          1249
          "<Object '/GrillMoon/CameraShake/GrillMoon_kfs1000_quickShake_csb.GrillMoon_kfs1000_quickShake_csb_C' (0x85051a32a00) Class 'BlueprintGeneratedClass'>"
         ]
         [
          1638
          "<Object '&GrillMoon_kfs1000_soft_csb_C' (0x84c3c6aea00) Class 'BlueprintGeneratedClass'>"
         ]
         [
          1797
          "<Object '&GrillMoon_kfs1000_quickShake_csb_C' (0x85051a32a00) Class 'BlueprintGeneratedClass'>"
         ]
         [
          1837
          "<Object '&GrillMoon_kfs1000_quickShake_csb_C' (0x85051a32a00) Class 'BlueprintGeneratedClass'>"
         ]
         [
          2037
          "<Object '/GrillMoon/CameraShake/GrillMoon_kfs1000_pullfov2_csb.GrillMoon_kfs1000_pullfov2_csb_C' (0x84e78c39600) Class 'BlueprintGeneratedClass'>"
         ]
         [
          2479
          "<Object '&GrillMoon_kfs1000_quickShake_csb_C' (0x85051a32a00) Class 'BlueprintGeneratedClass'>"
         ]
         [
          2724
          "<Object '/GrillMoon/CameraShake/GrillMoon_kfs1000_softBigger_csb.GrillMoon_kfs1000_softBigger_csb_C' (0x85051a31800) Class 'BlueprintGeneratedClass'>"
         ]
         [
          2823
          "<Object '&GrillMoon_kfs1000_quickShake_csb_C' (0x85051a32a00) Class 'BlueprintGeneratedClass'>"
         ]
         [
          2960
          "<Object '&GrillMoon_kfs1000_quickShake_csb_C' (0x85051a32a00) Class 'BlueprintGeneratedClass'>"
         ]
         [
          3076
          "<Object '&GrillMoon_kfs1000_soft_csb_C' (0x84c3c6aea00) Class 'BlueprintGeneratedClass'>"
         ]
         [
          3283
          "<Object '&GrillMoon_kfs1000_quickShake_csb_C' (0x85051a32a00) Class 'BlueprintGeneratedClass'>"
         ]
         [
          3310
          "<Object '&GrillMoon_kfs1000_soft_csb_C' (0x84c3c6aea00) Class 'BlueprintGeneratedClass'>"
         ]
         [
          3689
          "<Object '&GrillMoon_kfs1000_soft_csb_C' (0x84c3c6aea00) Class 'BlueprintGeneratedClass'>"
         ]
         [
          4237
          "<Object '/GrillMoon/CameraShake/GrillMoon_kfs1000_pullfov3_csb.GrillMoon_kfs1000_pullfov3_csb_C' (0x85051a33600) Class 'BlueprintGeneratedClass'>"
         ]
         [
          4613
          "<Object '/GrillMoon/CameraShake/GrillMoon_kfs1000_demonSwordSlash_csb.GrillMoon_kfs1000_demonSwordSlash_csb_C' (0x84c3c6af600) Class 'BlueprintGeneratedClass'>"
         ]
         [
          4959
          "<Object '/GrillMoon/CameraShake/GrillMoon_pjw1000_end_csb.GrillMoon_pjw1000_end_csb_C' (0x85051a35400) Class 'BlueprintGeneratedClass'>"
         ]
        ]
       }]
     }]
   }
   {
    type=MovieSceneEventTrack
    display_name=Cam Shake Radius Event Trigger
    sections [{
      type=MovieSceneEventTriggerSection
      channels [None {type=MovieSceneScriptingEventChannel, const ["<Struct 'MovieSceneEvent' (0x85024458ac0) {}>", 1]}]
     }]
   }
   {type=MovieSceneSpawnTrack, sections [{type=MovieSceneSpawnSection, channels [None {type=MovieSceneScriptingBoolChannel, default=true}]}]}
   {
    type=MovieSceneEventTrack
    display_name=Force Feedback Radius Event Trigger
    sections [{
      type=MovieSceneEventTriggerSection
      channels [None {type=MovieSceneScriptingEventChannel, const ["<Struct 'MovieSceneEvent' (0x85024457bc0) {}>", 1]}]
     }]
   }
   {
    type=MovieScene3DTransformTrack
    sections [{
      type=MovieScene3DTransformSection
      channels [
       Location.X {type=MovieSceneScriptingDoubleChannel, default=3696}
       Location.Y {type=MovieSceneScriptingDoubleChannel, default=27040}
       Location.Z {type=MovieSceneScriptingDoubleChannel, default=0}
       Rotation.X {type=MovieSceneScriptingDoubleChannel, default=0}
       Rotation.Y {type=MovieSceneScriptingDoubleChannel, default=0}
       Rotation.Z {type=MovieSceneScriptingDoubleChannel, default=0}
       Scale.X {type=MovieSceneScriptingDoubleChannel, default=1}
       Scale.Y {type=MovieSceneScriptingDoubleChannel, default=1}
       Scale.Z {type=MovieSceneScriptingDoubleChannel, default=1}
      ]
     }]
   }
   {
    type=MovieSceneIntegerTrack
    sections [{type=MovieSceneIntegerSection, channels [None {type=MovieSceneScriptingIntegerChannel, default=2, const [0,1]}]}]
   }
   {
    type=MovieSceneEventTrack
    display_name=Cam Shake Lobby Event Trigger
    sections [{
      type=MovieSceneEventTriggerSection
      channels [None {
        type=MovieSceneScriptingEventChannel
        keys [
         [-858, "<Struct 'MovieSceneEvent' (0x8502445e060) {}>"]
         [143, "<Struct 'MovieSceneEvent' (0x85024452940) {}>"]
         [459, "<Struct 'MovieSceneEvent' (0x850244515e0) {}>"]
         [1069, "<Struct 'MovieSceneEvent' (0x85024458c00) {}>"]
         [1249, "<Struct 'MovieSceneEvent' (0x85024452bc0) {}>"]
         [1638, "<Struct 'MovieSceneEvent' (0x850244521c0) {}>"]
         [1797, "<Struct 'MovieSceneEvent' (0x85024452620) {}>"]
         [1837, "<Struct 'MovieSceneEvent' (0x85024453520) {}>"]
         [2037, "<Struct 'MovieSceneEvent' (0x85024451c20) {}>"]
         [2479, "<Struct 'MovieSceneEvent' (0x8502445d2a0) {}>"]
         [2563, "<Struct 'MovieSceneEvent' (0x85024459f60) {}>"]
         [2724, "<Struct 'MovieSceneEvent' (0x8502445b540) {}>"]
         [2823, "<Struct 'MovieSceneEvent' (0x85024453b60) {}>"]
         [2960, "<Struct 'MovieSceneEvent' (0x85024455140) {}>"]
         [3077, "<Struct 'MovieSceneEvent' (0x850244550a0) {}>"]
         [3283, "<Struct 'MovieSceneEvent' (0x85024454ce0) {}>"]
         [3310, "<Struct 'MovieSceneEvent' (0x85024454f60) {}>"]
         [3689, "<Struct 'MovieSceneEvent' (0x85024454ba0) {}>"]
         [4237, "<Struct 'MovieSceneEvent' (0x85024454b00) {}>"]
         [4613, "<Struct 'MovieSceneEvent' (0x85024454740) {}>"]
         [4959, "<Struct 'MovieSceneEvent' (0x850244549c0) {}>"]
        ]
       }]
     }]
   }
   {
    type=MovieSceneEventTrack
    display_name=Force Feedback Lobby Event Trigger
    sections [{
      type=MovieSceneEventTriggerSection
      channels [None {
        type=MovieSceneScriptingEventChannel
        keys [
         [-858, "<Struct 'MovieSceneEvent' (0x85024452440) {}>"]
         [143, "<Struct 'MovieSceneEvent' (0x8502445d520) {}>"]
         [459, "<Struct 'MovieSceneEvent' (0x850244574e0) {}>"]
         [508, "<Struct 'MovieSceneEvent' (0x8502445bea0) {}>"]
         [588, "<Struct 'MovieSceneEvent' (0x8502445d700) {}>"]
         [686, "<Struct 'MovieSceneEvent' (0x85024458de0) {}>"]
         [715, "<Struct 'MovieSceneEvent' (0x85024451540) {}>"]
         [733, "<Struct 'MovieSceneEvent' (0x850244533e0) {}>"]
         [759, "<Struct 'MovieSceneEvent' (0x85024451cc0) {}>"]
         [804, "<Struct 'MovieSceneEvent' (0x850244529e0) {}>"]
         [856, "<Struct 'MovieSceneEvent' (0x85024459ce0) {}>"]
         [901, "<Struct 'MovieSceneEvent' (0x85024451fe0) {}>"]
         [933, "<Struct 'MovieSceneEvent' (0x85024452a80) {}>"]
         [980, "<Struct 'MovieSceneEvent' (0x8502445ac80) {}>"]
         [1069, "<Struct 'MovieSceneEvent' (0x850244591a0) {}>"]
         [1249, "<Struct 'MovieSceneEvent' (0x8502445bb80) {}>"]
         [1342, "<Struct 'MovieSceneEvent' (0x85024457ee0) {}>"]
         [1377, "<Struct 'MovieSceneEvent' (0x85024458660) {}>"]
         [1436, "<Struct 'MovieSceneEvent' (0x850244580c0) {}>"]
         [1481, "<Struct 'MovieSceneEvent' (0x8502445a500) {}>"]
         [1634, "<Struct 'MovieSceneEvent' (0x8502445a6e0) {}>"]
         [1805, "<Struct 'MovieSceneEvent' (0x850244544c0) {}>"]
         [1829, "<Struct 'MovieSceneEvent' (0x8502445d340) {}>"]
         [2031, "<Struct 'MovieSceneEvent' (0x85024453e80) {}>"]
         [2479, "<Struct 'MovieSceneEvent' (0x8502445adc0) {}>"]
         [2556, "<Struct 'MovieSceneEvent' (0x85024459880) {}>"]
         [2580, "<Struct 'MovieSceneEvent' (0x8502445f6e0) {}>"]
         [2612, "<Struct 'MovieSceneEvent' (0x8502445e600) {}>"]
         [2722, "<Struct 'MovieSceneEvent' (0x85024453f20) {}>"]
         [2820, "<Struct 'MovieSceneEvent' (0x8502445afa0) {}>"]
         [2960, "<Struct 'MovieSceneEvent' (0x85024454420) {}>"]
         [3076, "<Struct 'MovieSceneEvent' (0x8502445e100) {}>"]
         [3118, "<Struct 'MovieSceneEvent' (0x8502445ca80) {}>"]
         [3179, "<Struct 'MovieSceneEvent' (0x85024457440) {}>"]
         [3229, "<Struct 'MovieSceneEvent' (0x8502445ea60) {}>"]
         [3275, "<Struct 'MovieSceneEvent' (0x85024452580) {}>"]
         [3310, "<Struct 'MovieSceneEvent' (0x85024456ae0) {}>"]
         [3537, "<Struct 'MovieSceneEvent' (0x85024457300) {}>"]
         [3563, "<Struct 'MovieSceneEvent' (0x85024455c80) {}>"]
         [3689, "<Struct 'MovieSceneEvent' (0x85024455000) {}>"]
         [3785, "<Struct 'MovieSceneEvent' (0x85024455be0) {}>"]
         [4116, "<Struct 'MovieSceneEvent' (0x85024455dc0) {}>"]
         [4237, "<Struct 'MovieSceneEvent' (0x85024456360) {}>"]
         [4553, "<Struct 'MovieSceneEvent' (0x85024456400) {}>"]
         [4613, "<Struct 'MovieSceneEvent' (0x85024456540) {}>"]
         [4883, "<Struct 'MovieSceneEvent' (0x850244565e0) {}>"]
         [4972, "<Struct 'MovieSceneEvent' (0x85024456d60) {}>"]
         [5032, "<Struct 'MovieSceneEvent' (0x850244514a0) {}>"]
         [5111, "<Struct 'MovieSceneEvent' (0x850244508c0) {}>"]
        ]
       }]
     }]
   }
  ]
 }]
display_rate {denominator=1, numerator=60}
tick_resolution {denominator=1, numerator=24000}
tracks []
