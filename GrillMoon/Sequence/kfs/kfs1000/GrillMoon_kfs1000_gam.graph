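# Level Sequence for GrillMoon scene kfs1000, apparently its gameplay ("gam") layer, with no camera or animation tracks.
# A possessed BP_GrillMoon_NoFightBarrier actor has a visibility track that controls whether the barrier shows during the sequence.
# A root event track fires nine trigger events along the timeline, the first keyed before frame zero; their target functions are not identified.
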LevelSequence /GrillMoon/Sequence/kfs/kfs1000/GrillMoon_kfs1000_gam
bindings [BP_GrillMoon_NoFightBarrier {
  id=#1
  possessed_class=BP_GrillMoon_NoFightBarrier_C
  tracks [{
    type=MovieSceneVisibilityTrack
    sections [{type=MovieSceneVisibilitySection, channels [None {type=MovieSceneScriptingBoolChannel, default=true, const [false,1]}]}]
   }]
 }]
display_rate {denominator=1, numerator=60}
tick_resolution {denominator=1, numerator=24000}
tracks [{
  type=MovieSceneEventTrack
  sections [{
    type=MovieSceneEventTriggerSection
    channels [None {
      type=MovieSceneScriptingEventChannel
      keys [
       [-301, "<Struct 'MovieSceneEvent' (0x85024458980) {}>"]
       [132, "<Struct 'MovieSceneEvent' (0x85024458f20) {}>"]
       [451, "<Struct 'MovieSceneEvent' (0x85024459060) {}>"]
       [2037, "<Struct 'MovieSceneEvent' (0x850244594c0) {}>"]
       [2436, "<Struct 'MovieSceneEvent' (0x85024459600) {}>"]
       [4237, "<Struct 'MovieSceneEvent' (0x85024459ba0) {}>"]
       [4606, "<Struct 'MovieSceneEvent' (0x8502445a000) {}>"]
       [5276, "<Struct 'MovieSceneEvent' (0x8502445a140) {}>"]
       [5294, "<Struct 'MovieSceneEvent' (0x8502445a640) {}>"]
      ]
     }]
   }]
 }]
