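# Level Sequence that drives post-process effects and supporting spawns for the GrillMoon kfs1000 cinematic.
# A brief slash beat spawns BP_PostControl, attached blade and UnluckyRoll meshes and a Niagara line effect, while MPC_PostProcress keys tints, emissive and slash offsets.
# BP_PostControlDrain appears three times with repeated MPC_PostProcress_Drain curves for blur, vignette, desaturation, magenta overlay and gate texture, plus a repeater event.
# A later BP_PostCombinedDaggerDagger spawn pulses its post-process blend weight.
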LevelSequence /GrillMoon/Sequence/kfs/kfs1000/GrillMoon_kfs1000_post
bindings [
 SM_UnluckyRoll_Low {
  id=#1
  possessed_class=FortStaticMeshActor
  tracks [
   {
    type=MovieSceneSpawnTrack
    sections [{
      type=MovieSceneSpawnSection
      channels [None {type=MovieSceneScriptingBoolChannel, default=false, keys [[1870,false],[1871,true],[1944,false]]}]
     }]
   }
   {
    type=MovieScene3DTransformTrack
    sections [{
      type=MovieScene3DTransformSection
      channels [
       Location.X {type=MovieSceneScriptingDoubleChannel, default=0}
       Location.Y {type=MovieSceneScriptingDoubleChannel, default=0}
       Location.Z {type=MovieSceneScriptingDoubleChannel, default=0}
       Rotation.X {type=MovieSceneScriptingDoubleChannel, default=0}
       Rotation.Y {type=MovieSceneScriptingDoubleChannel, default=0}
       Rotation.Z {type=MovieSceneScriptingDoubleChannel, default=0}
       Scale.X {type=MovieSceneScriptingDoubleChannel, default=1}
       Scale.Y {type=MovieSceneScriptingDoubleChannel, default=1}
       Scale.Z {type=MovieSceneScriptingDoubleChannel, default=1}
      ]
     }]
   }
   {type=MovieScene3DAttachTrack, sections [{type=MovieScene3DAttachSection, end_frame=2477, start_frame=1869}]}
  ]
 }
 BP_PostControl {
  id=#2
  possessed_class=BP_PostControl_C
  tracks [
   {
    type=MovieSceneSpawnTrack
    sections [{
      type=MovieSceneSpawnSection
      channels [None {type=MovieSceneScriptingBoolChannel, default=false, keys [[1870,false],[1871.2412,true],[1939.415,true],[1940.6562,false]]}]
     }]
   }
   {
    type=MovieScene3DTransformTrack
    sections [{
      type=MovieScene3DTransformSection
      channels [
       Location.X {type=MovieSceneScriptingDoubleChannel, default=0}
       Location.Y {type=MovieSceneScriptingDoubleChannel, default=0}
       Location.Z {type=MovieSceneScriptingDoubleChannel, default=0}
       Rotation.X {type=MovieSceneScriptingDoubleChannel, default=0}
       Rotation.Y {type=MovieSceneScriptingDoubleChannel, default=0}
       Rotation.Z {type=MovieSceneScriptingDoubleChannel, default=0}
       Scale.X {type=MovieSceneScriptingDoubleChannel, default=1}
       Scale.Y {type=MovieSceneScriptingDoubleChannel, default=1}
       Scale.Z {type=MovieSceneScriptingDoubleChannel, default=1}
      ]
     }]
   }
  ]
 }
 SM_GrillMoon_Countdown_SplinterBlade_Low {
  id=#3
  possessed_class=FortStaticMeshActor
  tracks [
   {
    type=MovieSceneSpawnTrack
    sections [{
      type=MovieSceneSpawnSection
      channels [None {type=MovieSceneScriptingBoolChannel, default=false, keys [[1870,false],[1871,true],[1944,false]]}]
     }]
   }
   {
    type=MovieScene3DTransformTrack
    sections [{
      type=MovieScene3DTransformSection
      channels [
       Location.X {type=MovieSceneScriptingDoubleChannel, default=0}
       Location.Y {type=MovieSceneScriptingDoubleChannel, default=0}
       Location.Z {type=MovieSceneScriptingDoubleChannel, default=0}
       Rotation.X {type=MovieSceneScriptingDoubleChannel, default=0}
       Rotation.Y {type=MovieSceneScriptingDoubleChannel, default=0}
       Rotation.Z {type=MovieSceneScriptingDoubleChannel, default=0}
       Scale.X {type=MovieSceneScriptingDoubleChannel, default=1}
       Scale.Y {type=MovieSceneScriptingDoubleChannel, default=1}
       Scale.Z {type=MovieSceneScriptingDoubleChannel, default=1}
      ]
     }]
   }
   {type=MovieScene3DAttachTrack, sections [{type=MovieScene3DAttachSection, end_frame=2489, start_frame=1870}]}
  ]
 }
 "SM_GrillMoon_Countdown_SplinterBlade_Low (2)" {
  id=#4
  possessed_class=FortStaticMeshActor
  tracks [
   {
    type=MovieSceneSpawnTrack
    sections [{
      type=MovieSceneSpawnSection
      channels [None {type=MovieSceneScriptingBoolChannel, default=false, keys [[1870,false],[1871,true],[1944,false]]}]
     }]
   }
   {
    type=MovieScene3DTransformTrack
    sections [{
      type=MovieScene3DTransformSection
      channels [
       Location.X {type=MovieSceneScriptingDoubleChannel, default=0}
       Location.Y {type=MovieSceneScriptingDoubleChannel, default=0}
       Location.Z {type=MovieSceneScriptingDoubleChannel, default=0}
       Rotation.X {type=MovieSceneScriptingDoubleChannel, default=0}
       Rotation.Y {type=MovieSceneScriptingDoubleChannel, default=0}
       Rotation.Z {type=MovieSceneScriptingDoubleChannel, default=0}
       Scale.X {type=MovieSceneScriptingDoubleChannel, default=1}
       Scale.Y {type=MovieSceneScriptingDoubleChannel, default=1}
       Scale.Z {type=MovieSceneScriptingDoubleChannel, default=1}
      ]
     }]
   }
   {type=MovieScene3DAttachTrack, sections [{type=MovieScene3DAttachSection, end_frame=2484, start_frame=1869}]}
  ]
 }
 BP_PostCombinedDaggerDagger {
  id=#5
  possessed_class=BP_PostCombinedDaggerDagger_C
  tracks [
   {
    type=MovieSceneSpawnTrack
    sections [{
      type=MovieSceneSpawnSection
      channels [None {type=MovieSceneScriptingBoolChannel, default=false, keys [[2178,false],[2192,true],[2237,false]]}]
     }]
   }
   {
    type=MovieScene3DTransformTrack
    sections [{
      type=MovieScene3DTransformSection
      channels [
       Location.X {type=MovieSceneScriptingDoubleChannel, default=0}
       Location.Y {type=MovieSceneScriptingDoubleChannel, default=0}
       Location.Z {type=MovieSceneScriptingDoubleChannel, default=0}
       Rotation.X {type=MovieSceneScriptingDoubleChannel, default=0}
       Rotation.Y {type=MovieSceneScriptingDoubleChannel, default=0}
       Rotation.Z {type=MovieSceneScriptingDoubleChannel, default=0}
       Scale.X {type=MovieSceneScriptingDoubleChannel, default=1}
       Scale.Y {type=MovieSceneScriptingDoubleChannel, default=1}
       Scale.Z {type=MovieSceneScriptingDoubleChannel, default=1}
      ]
     }]
   }
  ]
 }
 BP_PostControlDrain {
  id=#6
  possessed_class=BP_PostControlDrain_C
  tracks [
   {
    type=MovieSceneSpawnTrack
    sections [{
      type=MovieSceneSpawnSection
      channels [None {type=MovieSceneScriptingBoolChannel, default=false, keys [[30.785,false],[65,true],[340.9,false],[985,true],[1263.0587,false],[2101,true]]}]
     }]
   }
   {
    type=MovieScene3DTransformTrack
    sections [{
      type=MovieScene3DTransformSection
      channels [
       Location.X {type=MovieSceneScriptingDoubleChannel, default=0}
       Location.Y {type=MovieSceneScriptingDoubleChannel, default=0}
       Location.Z {type=MovieSceneScriptingDoubleChannel, default=0}
       Rotation.X {type=MovieSceneScriptingDoubleChannel, default=0}
       Rotation.Y {type=MovieSceneScriptingDoubleChannel, default=0}
       Rotation.Z {type=MovieSceneScriptingDoubleChannel, default=0}
       Scale.X {type=MovieSceneScriptingDoubleChannel, default=1}
       Scale.Y {type=MovieSceneScriptingDoubleChannel, default=1}
       Scale.Z {type=MovieSceneScriptingDoubleChannel, default=1}
      ]
     }]
   }
   {type=MovieSceneEventTrack, sections [{type=MovieSceneEventRepeaterSection, end_frame=2395, start_frame=0}]}
  ]
 }
 PostProcess {
  id=#7
  parent_id=#5
  possessed_class=PostProcessComponent
  tracks [{
    type=MovieSceneFloatTrack
    display_name=Blend Weight
    sections [{type=MovieSceneFloatSection, channels [None {type=MovieSceneScriptingFloatChannel, default=0.75, keys [[2195,0],[2198,0.75],[2234,0]]}]}]
   }]
 }
 BP_GrillMoon_Countdown_SplinterBlade_R {id=#8, possessed_class=BP_GrillMoon_Countdown_SplinterBlade_A_C, tracks []}
 PostProcess {
  id=#9
  parent_id=#2
  possessed_class=PostProcessComponent
  tracks [{
    type=MovieSceneFloatTrack
    display_name="Intensity (Settings)"
    sections [{type=MovieSceneFloatSection, channels [None {type=MovieSceneScriptingFloatChannel, default=0.65, keys [[1928,3.8773],[1930,0.65]]}]}]
   }]
 }
 BP_GrillMoon_Countdown_SplinterBlade_L {id=#10, possessed_class=BP_GrillMoon_Countdown_SplinterBlade_A_C, tracks []}
 NiagaraComponent0 {
  id=#11
  parent_id=#12
  possessed_class=NiagaraComponent
  tracks [{
    type=MovieSceneParticleTrack
    display_name=FX System
    sections [{type=MovieSceneParticleSection, channels [None {type=MovieSceneScriptingParticleChannel, const ["<ParticleKey.TRIGGER: 2>", 1]}]}]
   }]
 }
 NS_LineEffect {
  id=#12
  possessed_class=NiagaraActor
  tracks [
   {
    type=MovieSceneSpawnTrack
    sections [{
      type=MovieSceneSpawnSection
      channels [None {type=MovieSceneScriptingBoolChannel, default=false, keys [[1823.8537,false],[1882,true],[2048,false]]}]
     }]
   }
   {
    type=MovieScene3DTransformTrack
    sections [{
      type=MovieScene3DTransformSection
      channels [
       Location.X {type=MovieSceneScriptingDoubleChannel, default=-21116}
       Location.Y {type=MovieSceneScriptingDoubleChannel, default=32736}
       Location.Z {type=MovieSceneScriptingDoubleChannel, default=23992}
       Rotation.X {type=MovieSceneScriptingDoubleChannel, default=19}
       Rotation.Y {type=MovieSceneScriptingDoubleChannel, default=-15}
       Rotation.Z {type=MovieSceneScriptingDoubleChannel, default=0}
       Scale.X {type=MovieSceneScriptingDoubleChannel, default=1}
       Scale.Y {type=MovieSceneScriptingDoubleChannel, default=1}
       Scale.Z {type=MovieSceneScriptingDoubleChannel, default=1}
      ]
     }]
   }
  ]
 }
]
display_rate {denominator=1, numerator=30}
tick_resolution {denominator=1, numerator=24000}
tracks [
 {
  type=MovieSceneMaterialParameterCollectionTrack
  display_name=MPC_PostProcress
  sections [{
    type=MovieSceneParameterSection
    channels [
     PostSwitch {type=MovieSceneScriptingFloatChannel, default=0, thinned_from=5, keys [[1870,0],[1871,1],[1927,1],[1944,0]]}
     PostEmissiveMultiply {type=MovieSceneScriptingFloatChannel, default=0.1, keys [[1871,0.1],[1882,0.1],[1889,0.01],[1896,0.01],[1925,0.1],[1927,0.1],[1928,1]]}
     Invert {type=MovieSceneScriptingFloatChannel, default=0}
     EmissiveOvrrideR {type=MovieSceneScriptingFloatChannel, default=1, const [50,1]}
     EmissiveOvrrideG {type=MovieSceneScriptingFloatChannel, default=1, const [1,1]}
     EmissiveOvrrideB {type=MovieSceneScriptingFloatChannel, default=1, const [1,1]}
     EmissiveOvrrideA {type=MovieSceneScriptingFloatChannel, const [1,1]}
     SlashOffsetR {type=MovieSceneScriptingFloatChannel, default=0.4227, keys [[1890,-0.1453],[1893,0.4227],[1928,0.3033]]}
     SlashOffsetG {type=MovieSceneScriptingFloatChannel, default=0}
     SlashOffsetB {type=MovieSceneScriptingFloatChannel, default=0.98, keys [[1890,0.98],[1893,0.98],[1897,0.998]]}
     SlashOffsetA {
      type=MovieSceneScriptingFloatChannel
      default=0.1
      keys [[1889,0],[1890,0.1],[1893,0.1],[1894,1],[1897,0.01],[1928,0],[1929,1],[1931,0.1],[1940,0]]
     }
     SceneBG_TintR {type=MovieSceneScriptingFloatChannel, default=0, keys [[1871,0.2139],[1882,0.2139],[1889,0],[1927,0],[1928,1]]}
     SceneBG_TintG {type=MovieSceneScriptingFloatChannel, default=0, keys [[1871,0.0938],[1882,0.0938],[1889,0],[1927,0],[1928,1]]}
     SceneBG_TintB {type=MovieSceneScriptingFloatChannel, default=0, keys [[1871,0.3264],[1882,0.3264],[1889,0],[1927,0],[1928,1]]}
     SceneBG_TintA {type=MovieSceneScriptingFloatChannel, const [1,5]}
     WordColorR {
      type=MovieSceneScriptingFloatChannel
      default=-5.6588
      keys [[1893,0],[1894,0.7562],[1895,-23.0188],[1897,-5.6588],[1898,-0.033],[1899.9325,0.01],[1927,0],[1928,1],[1936,0]]
     }
     WordColorG {
      type=MovieSceneScriptingFloatChannel
      default=-2.2488
      thinned_from=9
      keys [[1893,0],[1894,0.9188],[1895,-5.1039],[1897,-2.2488],[1898,0],[1927,0],[1928,-0.29],[1936,0]]
     }
     WordColorB {
      type=MovieSceneScriptingFloatChannel
      default=-0.1807
      thinned_from=9
      keys [[1893,0],[1894,1],[1895,0.6212],[1897,-0.1807],[1898,0],[1927,0],[1928,-0.42],[1936,0]]
     }
     WordColorA {type=MovieSceneScriptingFloatChannel, keys [[1893,1],[1894,1],[1895,0.9854],[1897,0.8913],[1898,1],[1899.9325,0],[1927,0],[1928,1],[1936,1]]}
    ]
   }]
 }
 {
  type=MovieSceneMaterialParameterCollectionTrack
  display_name=MPC_PostProcress_Drain
  sections [{
    type=MovieSceneParameterSection
    channels [
     PreFrameRandomBP {
      type=MovieSceneScriptingFloatChannel
      default=0
      thinned_from=15
      keys [[46,0],[82,0],[121,0.629],[223,0.629],[228,0],[1007.6788,0],[1046.6788,0.629],[1209.1325,0.629],[1214.1325,0],[2117.8587,0],[2156.8587,0.629],[2260.2,0.629],[2265.2,0]]
     }
     TextureSpeedBP {
      type=MovieSceneScriptingFloatChannel
      default=1.4838
      keys [[46,-1.507],[73,0.402],[89,2.7357],[151,1.4838],[228,1.4838],[302,-1.5575],[971.6788,-1.507],[998.6788,0.402],[1014.6788,2.7357],[1076.6788,1.4838],[1214.1325,1.4838],[1260.6988,-1.5575],[2081.8587,-1.507],[2108.8587,0.402],[2124.8587,2.7357],[2186.8587,1.4838],[2265.2,1.4838],[2300.06,-1.5575]]
     }
     BlurIntensity_A {
      type=MovieSceneScriptingFloatChannel
      default=1.0006
      thinned_from=24
      keys [[46,1.0011],[66,1.0011],[68,1.4445],[78,0.3369],[120,0.5586],[223,0.5586],[229,1.1965],[261,1],[991.6788,1.0011],[994,1.1484],[1004,0.6009],[1046,0.8811],[1209.4438,0.8811],[1215.4438,1.0913],[1228.4438,1],[2101.8587,1.0011],[2104.18,1.1484],[2114.18,0.6009],[2156.18,0.8811],[2260.5113,0.8811],[2266.5113,1.0913],[2279.5113,1]]
     }
     BlurIntensity_B {
      type=MovieSceneScriptingFloatChannel
      default=1
      thinned_from=18
      keys [[46,0.9991],[66,0.9991],[68,1.1601],[120,0.9983],[223,0.9983],[228,1],[991.6788,0.9991],[994,1.0625],[1045.6788,0.9983],[1209.1325,0.9983],[1214.1325,1],[2101.8587,0.9991],[2104.18,1.0625],[2155.8587,0.9983],[2260.2,0.9983],[2265.2,1]]
     }
     Vignette {
      type=MovieSceneScriptingFloatChannel
      default=0
      thinned_from=24
      keys [[46,0.001],[65,0.001],[66,-0.5011],[74,0.4807],[99,0.8776],[220,0.8776],[223,-0.5694],[282,0],[990.6788,0.001],[991.6788,-0.5011],[999.6788,0.4807],[1024.6788,0.8776],[1206.1325,0.8776],[1209.1325,-0.5694],[1242.6775,0],[2100.8587,0.001],[2101.8587,-0.5011],[2109.8587,0.4807],[2134.8587,0.8776],[2257.2,0.8776],[2260.2,-0.5694],[2293.745,0]]
     }
     OverlayColorR {
      type=MovieSceneScriptingFloatChannel
      default=0.0287
      thinned_from=18
      keys [[46,0],[78,0],[89,1.2085],[224,1.2085],[229,0.7264],[330,0],[1003.6788,0],[1014.6788,1.2085],[1210.1325,1.2085],[1215.1325,0.7264],[1261.265,0],[2113.8587,0],[2124.8587,1.2085],[2261.2,1.2085],[2266.2,0.7264],[2287.165,0]]
     }
     OverlayColorG {
      type=MovieSceneScriptingFloatChannel
      thinned_from=18
      keys [[46,0],[78,0],[89,-0.4598],[224,-0.4598],[229,-0.2002],[330,0],[1003.6788,0],[1014.6788,-0.4598],[1210.1325,-0.4598],[1215.1325,-0.2002],[1261.265,0],[2113.8587,0],[2124.8587,-0.4598],[2261.2,-0.4598],[2266.2,-0.2002],[2287.165,0]]
     }
     OverlayColorB {
      type=MovieSceneScriptingFloatChannel
      thinned_from=18
      keys [[46,0],[78,0],[89,1.0664],[224,1.0664],[229,0.5843],[330,0],[1003.6788,0],[1014.6788,1.0664],[1210.1325,1.0664],[1215.1325,0.5843],[1261.265,0],[2113.8587,0],[2124.8587,1.0664],[2261.2,1.0664],[2266.2,0.5843],[2287.165,0]]
     }
     OverlayColorA {type=MovieSceneScriptingFloatChannel, default=1, const [1,22]}
     GhostColorR {
      type=MovieSceneScriptingFloatChannel
      default=2
      thinned_from=15
      keys [[46,0],[78,0],[90,2],[223,2],[228,0],[1004,0],[1015.6788,2],[1209.1325,2],[1214,0],[2114,0],[2125.8587,2],[2260.2,2],[2265,0]]
     }
     GhostColorG {
      type=MovieSceneScriptingFloatChannel
      default=1
      thinned_from=15
      keys [[46,0],[78,0],[90,0.7621],[223,0.7621],[228,0],[1004,0],[1015.6788,0.7621],[1209.1325,0.7621],[1214,0],[2114,0],[2125.8587,0.7621],[2260.2,0.7621],[2265,0]]
     }
     GhostColorB {
      type=MovieSceneScriptingFloatChannel
      default=1
      thinned_from=15
      keys [[46,0],[78,0],[90,1.7527],[223,1.7527],[228,0],[1004,0],[1015.6788,1.7527],[1209.1325,1.7527],[1214,0],[2114,0],[2125.8587,1.7527],[2260.2,1.7527],[2265,0]]
     }
     GhostColorA {type=MovieSceneScriptingFloatChannel, default=1, const [0,12]}
     DesaturationR {type=MovieSceneScriptingFloatChannel, default=0, const [0.2126,6]}
     DesaturationG {type=MovieSceneScriptingFloatChannel, default=0, const [0.7152,6]}
     DesaturationB {type=MovieSceneScriptingFloatChannel, default=0, const [0.0722,6]}
     DesaturationA {
      type=MovieSceneScriptingFloatChannel
      default=0
      keys [[46,0],[82,-0.0014],[90,1.0133],[223,1.0133],[228,0],[971.6788,0],[1007.6788,-0.0014],[1015.6788,1.0133],[1209.1325,1.0133],[1214.1325,0],[2081.8587,0],[2117.8587,-0.0014],[2125.8587,1.0133],[2260.2,1.0133],[2265.2,0]]
     }
     MainTintR {
      type=MovieSceneScriptingFloatChannel
      thinned_from=27
      keys [[46,1],[65,1],[66,1.4206],[69,0.8413],[72,1],[91,1],[217,0.6566],[223,1.1978],[285,1],[990.6788,1],[991.6788,1.4206],[994.6788,0.8413],[997.6788,1],[1016.6788,1],[1203.1325,0.6566],[1209.1325,1.1978],[1251.8263,1],[2100.8587,1],[2101.8587,1.4206],[2104.8587,0.8413],[2107.8587,1],[2126.8587,1],[2254.1475,0.9062],[2260.2,1.1978],[2279.4825,1]]
     }
     MainTintG {
      type=MovieSceneScriptingFloatChannel
      thinned_from=27
      keys [[46,1],[65,1],[66,1.3363],[69,0.2107],[72,1],[91,1],[217,0.3109],[223,1.1978],[285,1],[990.6788,1],[991.6788,1.3363],[994.6788,0.2107],[997.6788,1],[1016.6788,1],[1203.1325,0.3109],[1209.1325,1.1978],[1251.8263,1],[2100.8587,1],[2101.8587,1.3363],[2104.8587,0.2107],[2107.8587,1],[2126.8587,1],[2254.1475,0.8712],[2260.2,1.1978],[2279.4825,1]]
     }
     MainTintB {
      type=MovieSceneScriptingFloatChannel
      thinned_from=27
      keys [[46,1],[65,1],[66,1.381],[69,0.4709],[72,1],[91,1],[217,0.3109],[223,1.1978],[285,1],[990.6788,1],[991.6788,1.381],[994.6788,0.4709],[997.6788,1],[1016.6788,1],[1203.1325,0.3109],[1209.1325,1.1978],[1251.8263,1],[2100.8587,1],[2101.8587,1.381],[2104.8587,0.4709],[2107.8587,1],[2126.8587,1],[2254.1475,0.8712],[2260.2,1.1978],[2279.4825,1]]
     }
     MainTintA {
      type=MovieSceneScriptingFloatChannel
      thinned_from=44
      keys [[46,0],[66,-0.0012],[72,-0.0012],[91,0.2035],[103,0.7781],[117,0.2103],[144,0.2031],[152,0.5098],[160,0.2007],[200,0.2031],[208,0.4562],[215,0.2031],[223,0.2031],[285,-0.0005],[971.6788,0],[991.6788,-0.0012],[997.6788,-0.0012],[1016.6788,0.2035],[1028.6788,0.7781],[1042.6788,0.2103],[1137.245,0.2031],[1145.245,0.5098],[1153.245,0.2007],[1186.1325,0.2031],[1194.1325,0.4562],[1201.1325,0.2031],[1209.1325,0.2031],[1251.8263,-0.0005],[2081.8587,0],[2101.8587,-0.0012],[2107.8587,-0.0012],[2126.8587,0.2035],[2138.8587,0.7781],[2152.8587,0.2103],[2237.2,0.2031],[2260.2,0.2031],[2279.4825,-0.0005]]
     }
     GateTextureR {type=MovieSceneScriptingFloatChannel, const [1,3]}
     GateTextureG {type=MovieSceneScriptingFloatChannel, const [0,3]}
     GateTextureB {type=MovieSceneScriptingFloatChannel, const [0.5833,3]}
     GateTextureA {
      type=MovieSceneScriptingFloatChannel
      default=0
      keys [[64.5938,0],[78.1413,1],[215,1],[329.8338,0],[990.2725,0],[1003.82,1],[1201.1325,1],[1259.575,0],[2100.4525,0],[2114,1],[2252.2,1],[2301.2763,0]]
     }
    ]
   }]
 }
]
