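# Tech-art Level Sequence for the GrillMoon kfs1000 cinematic, dressing it with clouds, cloud lighting and character overlays.
# Spawns statically placed, scaled low-altitude cloud Blueprints and three CloudMPCLightControllers, whose repeating MPC_Updater events run alongside Color, Radius and Intensity tracks.
# FX1 animates position, color and intensity pulses, while the Demon and MistMylar lights use attach tracks.
# Mystic overlay materials cover both MistMylar mannequins, and a subsequence is nested.
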LevelSequence /GrillMoon/Sequence/kfs/kfs1000/GrillMoon_kfs1000_tec
bindings [
 "BP_LowAltitudeClouds_GrillMoon (10)" {
  id=#1
  possessed_class=BP_LowAltitudeClouds_GrillMoon_C
  tracks [
   {type=MovieSceneSpawnTrack, sections [{type=MovieSceneSpawnSection, channels [None {type=MovieSceneScriptingBoolChannel, default=true}]}]}
   {
    type=MovieScene3DTransformTrack
    sections [{
      type=MovieScene3DTransformSection
      channels [
       Location.X {type=MovieSceneScriptingDoubleChannel, default=-26083.3394}
       Location.Y {type=MovieSceneScriptingDoubleChannel, default=45705.3644}
       Location.Z {type=MovieSceneScriptingDoubleChannel, default=10859.054}
       Rotation.X {type=MovieSceneScriptingDoubleChannel, default=0}
       Rotation.Y {type=MovieSceneScriptingDoubleChannel, default=0}
       Rotation.Z {type=MovieSceneScriptingDoubleChannel, default=0}
       Scale.X {type=MovieSceneScriptingDoubleChannel, default=40}
       Scale.Y {type=MovieSceneScriptingDoubleChannel, default=40}
       Scale.Z {type=MovieSceneScriptingDoubleChannel, default=44.5167}
      ]
     }]
   }
  ]
 }
 BP_CloudMPCLightController_FX1 {
  id=#2
  possessed_class=BP_CloudMPCLightController_GrillMoon_C
  tracks [
   {type=MovieSceneSpawnTrack, sections [{type=MovieSceneSpawnSection, channels [None {type=MovieSceneScriptingBoolChannel, default=true}]}]}
   {
    type=MovieScene3DTransformTrack
    sections [{
      type=MovieScene3DTransformSection
      channels [
       Location.X {
        type=MovieSceneScriptingDoubleChannel
        default=-57011.5748
        keys [[2092,-57011.5742],[2455,-57011.5742],[4187,-53025.2656],[4681,-53025.2656]]
       }
       Location.Y {type=MovieSceneScriptingDoubleChannel, default=47019.141, keys [[2092,47019.1406],[2455,47019.1406],[4187,41554.0273],[4681,39034.3555]]}
       Location.Z {type=MovieSceneScriptingDoubleChannel, default=17460.3843, keys [[2092,15691.0527],[2455,15691.0527],[4187,17118.4648],[4681,17118.4648]]}
       Rotation.X {type=MovieSceneScriptingDoubleChannel, default=0, const [0,4]}
       Rotation.Y {type=MovieSceneScriptingDoubleChannel, default=0, const [0,4]}
       Rotation.Z {type=MovieSceneScriptingDoubleChannel, default=0, const [0,4]}
       Scale.X {type=MovieSceneScriptingDoubleChannel, default=1, const [1,4]}
       Scale.Y {type=MovieSceneScriptingDoubleChannel, default=1, const [1,4]}
       Scale.Z {type=MovieSceneScriptingDoubleChannel, default=1, const [1,4]}
      ]
     }]
   }
   {
    type=MovieSceneEventTrack
    sections [{
      type=MovieSceneEventTriggerSection
      channels [None {type=MovieSceneScriptingEventChannel, const ["<Struct 'MovieSceneEvent' (0x8502445c1c0) {}>", 1]}]
     }]
   }
   {type=MovieSceneEventTrack, display_name=MPC_Updater, sections [{type=MovieSceneEventRepeaterSection, end_frame=13569, start_frame=-301}]}
   {
    type=MovieSceneDoubleTrack
    display_name=Radius
    sections [{
      type=MovieSceneDoubleSection
      channels [None {type=MovieSceneScriptingDoubleChannel, default=5000, keys [[2092,5000],[2399,5000],[4187,8000],[4598,8000],[4681,10000]]}]
     }]
   }
   {
    type=MovieSceneDoubleTrack
    display_name=Intensity
    sections [{
      type=MovieSceneDoubleSection
      channels [None {
        type=MovieSceneScriptingDoubleChannel
        default=5
        keys [[2032,0],[2092,1],[2411,1],[2455,0],[4187,0],[4247,1],[4554,1],[4598,0],[4681,0],[4709,1],[4808,1],[5128,5]]
       }]
     }]
   }
   {
    type=MovieSceneColorTrack
    display_name=Color
    sections [{
      type=MovieSceneColorSection
      channels [
       Color.R {type=MovieSceneScriptingFloatChannel, default=0.4771, keys [[0,0.4771],[4598,0.4771],[4615,1]]}
       Color.G {type=MovieSceneScriptingFloatChannel, default=0, keys [[0,0],[4598,0],[4615,0.6562]]}
       Color.B {type=MovieSceneScriptingFloatChannel, default=1, keys [[0,1],[4598,1],[4615,0]]}
       Color.A {type=MovieSceneScriptingFloatChannel, default=1, const [1,3]}
      ]
     }]
   }
  ]
 }
 BP_LowAltitudeClouds_GrillMoon {
  id=#3
  possessed_class=BP_LowAltitudeClouds_GrillMoon_C
  tracks [
   {type=MovieSceneSpawnTrack, sections [{type=MovieSceneSpawnSection, channels [None {type=MovieSceneScriptingBoolChannel, default=true}]}]}
   {
    type=MovieScene3DTransformTrack
    sections [{
      type=MovieScene3DTransformSection
      channels [
       Location.X {type=MovieSceneScriptingDoubleChannel, default=-46644.2158}
       Location.Y {type=MovieSceneScriptingDoubleChannel, default=45003.0684}
       Location.Z {type=MovieSceneScriptingDoubleChannel, default=12095.5727}
       Rotation.X {type=MovieSceneScriptingDoubleChannel, default=0}
       Rotation.Y {type=MovieSceneScriptingDoubleChannel, default=0}
       Rotation.Z {type=MovieSceneScriptingDoubleChannel, default=0}
       Scale.X {type=MovieSceneScriptingDoubleChannel, default=30}
       Scale.Y {type=MovieSceneScriptingDoubleChannel, default=35}
       Scale.Z {type=MovieSceneScriptingDoubleChannel, default=35}
      ]
     }]
   }
  ]
 }
 "BP_LowAltitudeClouds_GrillMoon (2)" {
  id=#4
  possessed_class=BP_LowAltitudeClouds_GrillMoon_C
  tracks [
   {type=MovieSceneSpawnTrack, sections [{type=MovieSceneSpawnSection, channels [None {type=MovieSceneScriptingBoolChannel, default=true}]}]}
   {
    type=MovieScene3DTransformTrack
    sections [{
      type=MovieScene3DTransformSection
      channels [
       Location.X {type=MovieSceneScriptingDoubleChannel, default=-32071.2266}
       Location.Y {type=MovieSceneScriptingDoubleChannel, default=55625.2938}
       Location.Z {type=MovieSceneScriptingDoubleChannel, default=12541.3105}
       Rotation.X {type=MovieSceneScriptingDoubleChannel, default=0}
       Rotation.Y {type=MovieSceneScriptingDoubleChannel, default=0}
       Rotation.Z {type=MovieSceneScriptingDoubleChannel, default=0}
       Scale.X {type=MovieSceneScriptingDoubleChannel, default=29.238}
       Scale.Y {type=MovieSceneScriptingDoubleChannel, default=29.238}
       Scale.Z {type=MovieSceneScriptingDoubleChannel, default=29.238}
      ]
     }]
   }
  ]
 }
 "BP_LowAltitudeClouds_GrillMoon (5)" {
  id=#5
  possessed_class=BP_LowAltitudeClouds_GrillMoon_C
  tracks [
   {type=MovieSceneSpawnTrack, sections [{type=MovieSceneSpawnSection, channels [None {type=MovieSceneScriptingBoolChannel, default=true}]}]}
   {
    type=MovieScene3DTransformTrack
    sections [{
      type=MovieScene3DTransformSection
      channels [
       Location.X {type=MovieSceneScriptingDoubleChannel, default=-80094.2605}
       Location.Y {type=MovieSceneScriptingDoubleChannel, default=106657.866}
       Location.Z {type=MovieSceneScriptingDoubleChannel, default=12772.3184}
       Rotation.X {type=MovieSceneScriptingDoubleChannel, default=0}
       Rotation.Y {type=MovieSceneScriptingDoubleChannel, default=0}
       Rotation.Z {type=MovieSceneScriptingDoubleChannel, default=0}
       Scale.X {type=MovieSceneScriptingDoubleChannel, default=40}
       Scale.Y {type=MovieSceneScriptingDoubleChannel, default=40}
       Scale.Z {type=MovieSceneScriptingDoubleChannel, default=40}
      ]
     }]
   }
  ]
 }
 "BP_LowAltitudeClouds_GrillMoon (11)" {
  id=#6
  possessed_class=BP_LowAltitudeClouds_GrillMoon_C
  tracks [
   {type=MovieSceneSpawnTrack, sections [{type=MovieSceneSpawnSection, channels [None {type=MovieSceneScriptingBoolChannel, default=true}]}]}
   {
    type=MovieScene3DTransformTrack
    sections [{
      type=MovieScene3DTransformSection
      channels [
       Location.X {type=MovieSceneScriptingDoubleChannel, default=-57677.1145}
       Location.Y {type=MovieSceneScriptingDoubleChannel, default=92412.2287}
       Location.Z {type=MovieSceneScriptingDoubleChannel, default=11954.186}
       Rotation.X {type=MovieSceneScriptingDoubleChannel, default=0}
       Rotation.Y {type=MovieSceneScriptingDoubleChannel, default=0}
       Rotation.Z {type=MovieSceneScriptingDoubleChannel, default=0}
       Scale.X {type=MovieSceneScriptingDoubleChannel, default=91.8794}
       Scale.Y {type=MovieSceneScriptingDoubleChannel, default=54.8314}
       Scale.Z {type=MovieSceneScriptingDoubleChannel, default=54.8314}
      ]
     }]
   }
  ]
 }
 Face {
  id=#7
  parent_id=#8
  possessed_class=SkeletalMeshComponentBudgeted
  tracks [{
    type=MovieSceneObjectPropertyTrack
    display_name=Overlay Material
    sections [{
      type=MovieSceneObjectPropertySection
      channels [None {
        type=MovieSceneScriptingObjectPathChannel
        default="<Object '/GrillMoon/Assets/TechArt/Character_ScaleUp/MI_GrillMoon_Overlay_Mystic_Acc' (0x85051d1b000) Class 'MaterialInstanceConstant'>"
        const ["<Object '&MI_GrillMoon_Overlay_Mystic_Acc' (0x85051d1b000) Class 'MaterialInstanceConstant'>", 1]
       }]
     }]
   }]
 }
 "BP_LowAltitudeClouds_GrillMoon (13)" {
  id=#9
  possessed_class=BP_LowAltitudeClouds_GrillMoon_C
  tracks [
   {type=MovieSceneSpawnTrack, sections [{type=MovieSceneSpawnSection, channels [None {type=MovieSceneScriptingBoolChannel, default=true}]}]}
   {
    type=MovieScene3DTransformTrack
    sections [{
      type=MovieScene3DTransformSection
      channels [
       Location.X {type=MovieSceneScriptingDoubleChannel, default=-29549.619}
       Location.Y {type=MovieSceneScriptingDoubleChannel, default=27181.0195}
       Location.Z {type=MovieSceneScriptingDoubleChannel, default=12381.808}
       Rotation.X {type=MovieSceneScriptingDoubleChannel, default=0}
       Rotation.Y {type=MovieSceneScriptingDoubleChannel, default=0}
       Rotation.Z {type=MovieSceneScriptingDoubleChannel, default=0}
       Scale.X {type=MovieSceneScriptingDoubleChannel, default=30.4201}
       Scale.Y {type=MovieSceneScriptingDoubleChannel, default=40}
       Scale.Z {type=MovieSceneScriptingDoubleChannel, default=40}
      ]
     }]
   }
  ]
 }
 Head {
  id=#10
  parent_id=#8
  possessed_class=SkeletalMeshComponentBudgeted
  tracks [{
    type=MovieSceneObjectPropertyTrack
    display_name=Overlay Material
    sections [{
      type=MovieSceneObjectPropertySection
      channels [None {
        type=MovieSceneScriptingObjectPathChannel
        default="<Object '/GrillMoon/Assets/TechArt/Character_ScaleUp/MI_GrillMoon_Overlay_Mystic_Head' (0x85051d1a800) Class 'MaterialInstanceConstant'>"
        const [
         "<Object '&MI_GrillMoon_Overlay_Mystic_Head' (0x85051d1a800) Class 'MaterialInstanceConstant'>"
         1
        ]
       }]
     }]
   }]
 }
 "BP_LowAltitudeClouds_GrillMoon (6)" {
  id=#11
  possessed_class=BP_LowAltitudeClouds_GrillMoon_C
  tracks [
   {type=MovieSceneSpawnTrack, sections [{type=MovieSceneSpawnSection, channels [None {type=MovieSceneScriptingBoolChannel, default=true}]}]}
   {
    type=MovieScene3DTransformTrack
    sections [{
      type=MovieScene3DTransformSection
      channels [
       Location.X {type=MovieSceneScriptingDoubleChannel, default=-90826.7253}
       Location.Y {type=MovieSceneScriptingDoubleChannel, default=106315.2849}
       Location.Z {type=MovieSceneScriptingDoubleChannel, default=12090.5983}
       Rotation.X {type=MovieSceneScriptingDoubleChannel, default=0}
       Rotation.Y {type=MovieSceneScriptingDoubleChannel, default=0}
       Rotation.Z {type=MovieSceneScriptingDoubleChannel, default=0}
       Scale.X {type=MovieSceneScriptingDoubleChannel, default=30}
       Scale.Y {type=MovieSceneScriptingDoubleChannel, default=30}
       Scale.Z {type=MovieSceneScriptingDoubleChannel, default=30}
      ]
     }]
   }
  ]
 }
 "BP_LowAltitudeClouds_GrillMoon (9)" {
  id=#12
  possessed_class=BP_LowAltitudeClouds_GrillMoon_C
  tracks [
   {type=MovieSceneSpawnTrack, sections [{type=MovieSceneSpawnSection, channels [None {type=MovieSceneScriptingBoolChannel, default=true}]}]}
   {
    type=MovieScene3DTransformTrack
    sections [{
      type=MovieScene3DTransformSection
      channels [
       Location.X {type=MovieSceneScriptingDoubleChannel, default=-30954.8503}
       Location.Y {type=MovieSceneScriptingDoubleChannel, default=72284.5408}
       Location.Z {type=MovieSceneScriptingDoubleChannel, default=13137.2323}
       Rotation.X {type=MovieSceneScriptingDoubleChannel, default=0}
       Rotation.Y {type=MovieSceneScriptingDoubleChannel, default=0}
       Rotation.Z {type=MovieSceneScriptingDoubleChannel, default=0}
       Scale.X {type=MovieSceneScriptingDoubleChannel, default=35}
       Scale.Y {type=MovieSceneScriptingDoubleChannel, default=50}
       Scale.Z {type=MovieSceneScriptingDoubleChannel, default=50}
      ]
     }]
   }
  ]
 }
 GrillMoon_Countdown_MistMylar {id=#13, possessed_class=FortPlayerMannequin, tracks []}
 "BP_LowAltitudeClouds_GrillMoon (8)" {
  id=#14
  possessed_class=BP_LowAltitudeClouds_GrillMoon_C
  tracks [
   {type=MovieSceneSpawnTrack, sections [{type=MovieSceneSpawnSection, channels [None {type=MovieSceneScriptingBoolChannel, default=true}]}]}
   {
    type=MovieScene3DTransformTrack
    sections [{
      type=MovieScene3DTransformSection
      channels [
       Location.X {type=MovieSceneScriptingDoubleChannel, default=-71858.599}
       Location.Y {type=MovieSceneScriptingDoubleChannel, default=72288.5258}
       Location.Z {type=MovieSceneScriptingDoubleChannel, default=11215.113}
       Rotation.X {type=MovieSceneScriptingDoubleChannel, default=0}
       Rotation.Y {type=MovieSceneScriptingDoubleChannel, default=0}
       Rotation.Z {type=MovieSceneScriptingDoubleChannel, default=0}
       Scale.X {type=MovieSceneScriptingDoubleChannel, default=52.4177}
       Scale.Y {type=MovieSceneScriptingDoubleChannel, default=47.6186}
       Scale.Z {type=MovieSceneScriptingDoubleChannel, default=42.7932}
      ]
     }]
   }
  ]
 }
 "BP_LowAltitudeClouds_GrillMoon (7)3" {
  id=#15
  possessed_class=BP_LowAltitudeClouds_GrillMoon_C
  tracks [
   {type=MovieSceneSpawnTrack, sections [{type=MovieSceneSpawnSection, channels [None {type=MovieSceneScriptingBoolChannel, default=true}]}]}
   {
    type=MovieScene3DTransformTrack
    sections [{
      type=MovieScene3DTransformSection
      channels [
       Location.X {type=MovieSceneScriptingDoubleChannel, default=-72034.5343}
       Location.Y {type=MovieSceneScriptingDoubleChannel, default=57039.8793}
       Location.Z {type=MovieSceneScriptingDoubleChannel, default=11290.8805}
       Rotation.X {type=MovieSceneScriptingDoubleChannel, default=0}
       Rotation.Y {type=MovieSceneScriptingDoubleChannel, default=0}
       Rotation.Z {type=MovieSceneScriptingDoubleChannel, default=-49.2711}
       Scale.X {type=MovieSceneScriptingDoubleChannel, default=13.3035}
       Scale.Y {type=MovieSceneScriptingDoubleChannel, default=19.5812}
       Scale.Z {type=MovieSceneScriptingDoubleChannel, default=8.3007}
      ]
     }]
   }
  ]
 }
 BP_CloudMPCLightController_Demon {
  id=#16
  possessed_class=BP_CloudMPCLightController_GrillMoon_C
  tracks [
   {type=MovieSceneSpawnTrack, sections [{type=MovieSceneSpawnSection, channels [None {type=MovieSceneScriptingBoolChannel, default=true}]}]}
   {
    type=MovieScene3DTransformTrack
    sections [{
      type=MovieScene3DTransformSection
      channels [
       Location.X {type=MovieSceneScriptingDoubleChannel, default=-30}
       Location.Y {type=MovieSceneScriptingDoubleChannel, default=0}
       Location.Z {type=MovieSceneScriptingDoubleChannel, default=0}
       Rotation.X {type=MovieSceneScriptingDoubleChannel, default=0}
       Rotation.Y {type=MovieSceneScriptingDoubleChannel, default=0}
       Rotation.Z {type=MovieSceneScriptingDoubleChannel, default=0}
       Scale.X {type=MovieSceneScriptingDoubleChannel, default=1}
       Scale.Y {type=MovieSceneScriptingDoubleChannel, default=1}
       Scale.Z {type=MovieSceneScriptingDoubleChannel, default=1}
      ]
     }]
   }
   {type=MovieSceneEventTrack, display_name=MPC_Updater, sections [{type=MovieSceneEventRepeaterSection, end_frame=13569, start_frame=-305}]}
   {type=MovieScene3DAttachTrack, sections [{type=MovieScene3DAttachSection, end_frame=12912, start_frame=-305}]}
   {
    type=MovieSceneDoubleTrack
    display_name=Intensity
    sections [{type=MovieSceneDoubleSection, channels [None {type=MovieSceneScriptingDoubleChannel, default=0, const [0.75,1]}]}]
   }
   {
    type=MovieSceneDoubleTrack
    display_name=Radius
    sections [{type=MovieSceneDoubleSection, channels [None {type=MovieSceneScriptingDoubleChannel, default=10000, const [10000,1]}]}]
   }
   {
    type=MovieSceneColorTrack
    display_name=Color
    sections [{
      type=MovieSceneColorSection
      channels [
       Color.R {type=MovieSceneScriptingFloatChannel, default=1, const [1,1]}
       Color.G {type=MovieSceneScriptingFloatChannel, default=0, const [0,1]}
       Color.B {type=MovieSceneScriptingFloatChannel, default=0.2505, const [0.0088,1]}
       Color.A {type=MovieSceneScriptingFloatChannel, default=1, const [1,1]}
      ]
     }]
   }
   {
    type=MovieSceneEventTrack
    sections [{
      type=MovieSceneEventTriggerSection
      channels [None {type=MovieSceneScriptingEventChannel, const ["<Struct 'MovieSceneEvent' (0x850244501e0) {}>", 1]}]
     }]
   }
  ]
 }
 Body {
  id=#17
  parent_id=#8
  possessed_class=SkeletalMeshComponentBudgeted
  tracks [{
    type=MovieSceneObjectPropertyTrack
    display_name=Overlay Material
    sections [{
      type=MovieSceneObjectPropertySection
      channels [None {
        type=MovieSceneScriptingObjectPathChannel
        default="<Object '/GrillMoon/Assets/TechArt/Character_ScaleUp/MI_GrillMoon_Overlay_Mystic_Body' (0x85051d1a000) Class 'MaterialInstanceConstant'>"
        const [
         "<Object '&MI_GrillMoon_Overlay_Mystic_Body' (0x85051d1a000) Class 'MaterialInstanceConstant'>"
         1
        ]
       }]
     }]
   }]
 }
 Head {
  id=#18
  parent_id=#13
  possessed_class=SkeletalMeshComponentBudgeted
  tracks [{
    type=MovieSceneObjectPropertyTrack
    display_name=Overlay Material
    sections [{
      type=MovieSceneObjectPropertySection
      channels [None {
        type=MovieSceneScriptingObjectPathChannel
        default="<Object '&MI_GrillMoon_Overlay_Mystic_Head' (0x85051d1a800) Class 'MaterialInstanceConstant'>"
        const [
         "<Object '&MI_GrillMoon_Overlay_Mystic_Head' (0x85051d1a800) Class 'MaterialInstanceConstant'>"
         1
        ]
       }]
     }]
   }]
 }
 GrillMoon_MistMylar {id=#8, possessed_class=FortPlayerMannequin, tracks []}
 "BP_LowAltitudeClouds_GrillMoon (2)2" {
  id=#19
  possessed_class=BP_LowAltitudeClouds_GrillMoon_C
  tracks [
   {type=MovieSceneSpawnTrack, sections [{type=MovieSceneSpawnSection, channels [None {type=MovieSceneScriptingBoolChannel, default=true}]}]}
   {
    type=MovieScene3DTransformTrack
    sections [{
      type=MovieScene3DTransformSection
      channels [
       Location.X {type=MovieSceneScriptingDoubleChannel, default=-18427.5588}
       Location.Y {type=MovieSceneScriptingDoubleChannel, default=44673.5564}
       Location.Z {type=MovieSceneScriptingDoubleChannel, default=13087.054}
       Rotation.X {type=MovieSceneScriptingDoubleChannel, default=0}
       Rotation.Y {type=MovieSceneScriptingDoubleChannel, default=0}
       Rotation.Z {type=MovieSceneScriptingDoubleChannel, default=0}
       Scale.X {type=MovieSceneScriptingDoubleChannel, default=30}
       Scale.Y {type=MovieSceneScriptingDoubleChannel, default=30}
       Scale.Z {type=MovieSceneScriptingDoubleChannel, default=30}
      ]
     }]
   }
  ]
 }
 BP_CloudMPCLightController_MistMylar {
  id=#20
  possessed_class=BP_CloudMPCLightController_GrillMoon_C
  tracks [
   {type=MovieSceneSpawnTrack, sections [{type=MovieSceneSpawnSection, channels [None {type=MovieSceneScriptingBoolChannel, default=true}]}]}
   {
    type=MovieScene3DTransformTrack
    sections [{
      type=MovieScene3DTransformSection
      channels [
       Location.X {type=MovieSceneScriptingDoubleChannel, default=-30}
       Location.Y {type=MovieSceneScriptingDoubleChannel, default=0}
       Location.Z {type=MovieSceneScriptingDoubleChannel, default=0}
       Rotation.X {type=MovieSceneScriptingDoubleChannel, default=0}
       Rotation.Y {type=MovieSceneScriptingDoubleChannel, default=0}
       Rotation.Z {type=MovieSceneScriptingDoubleChannel, default=0}
       Scale.X {type=MovieSceneScriptingDoubleChannel, default=1}
       Scale.Y {type=MovieSceneScriptingDoubleChannel, default=1}
       Scale.Z {type=MovieSceneScriptingDoubleChannel, default=1}
      ]
     }]
   }
   {type=MovieScene3DAttachTrack, sections [{type=MovieScene3DAttachSection, end_frame=9776, start_frame=-305}]}
   {type=MovieSceneEventTrack, display_name=MPC_Updater, sections [{type=MovieSceneEventRepeaterSection, end_frame=13569, start_frame=-305}]}
   {
    type=MovieSceneColorTrack
    display_name=Color
    sections [{
      type=MovieSceneColorSection
      channels [
       Color.R {type=MovieSceneScriptingFloatChannel, default=0.9925, const [0.9925,1]}
       Color.G {type=MovieSceneScriptingFloatChannel, default=1, const [1,1]}
       Color.B {type=MovieSceneScriptingFloatChannel, default=0.4031, const [0.4031,1]}
       Color.A {type=MovieSceneScriptingFloatChannel, default=1, const [1,1]}
      ]
     }]
   }
   {
    type=MovieSceneDoubleTrack
    display_name=Radius
    sections [{type=MovieSceneDoubleSection, channels [None {type=MovieSceneScriptingDoubleChannel, default=10000, const [10000,1]}]}]
   }
   {
    type=MovieSceneDoubleTrack
    display_name=Intensity
    sections [{type=MovieSceneDoubleSection, channels [None {type=MovieSceneScriptingDoubleChannel, default=0.75, const [0.75,1]}]}]
   }
   {
    type=MovieSceneEventTrack
    sections [{
      type=MovieSceneEventTriggerSection
      channels [None {type=MovieSceneScriptingEventChannel, const ["<Struct 'MovieSceneEvent' (0x8502445c800) {}>", 1]}]
     }]
   }
  ]
 }
 "BP_LowAltitudeClouds_GrillMoon (7)4" {
  id=#21
  possessed_class=BP_LowAltitudeClouds_GrillMoon_C
  tracks [
   {type=MovieSceneSpawnTrack, sections [{type=MovieSceneSpawnSection, channels [None {type=MovieSceneScriptingBoolChannel, default=true}]}]}
   {
    type=MovieScene3DTransformTrack
    sections [{
      type=MovieScene3DTransformSection
      channels [
       Location.X {type=MovieSceneScriptingDoubleChannel, default=-80106.5343}
       Location.Y {type=MovieSceneScriptingDoubleChannel, default=55551.8793}
       Location.Z {type=MovieSceneScriptingDoubleChannel, default=9646.8805}
       Rotation.X {type=MovieSceneScriptingDoubleChannel, default=0}
       Rotation.Y {type=MovieSceneScriptingDoubleChannel, default=0}
       Rotation.Z {type=MovieSceneScriptingDoubleChannel, default=118.7545}
       Scale.X {type=MovieSceneScriptingDoubleChannel, default=9.0813}
       Scale.Y {type=MovieSceneScriptingDoubleChannel, default=12.6686}
       Scale.Z {type=MovieSceneScriptingDoubleChannel, default=12.9753}
      ]
     }]
   }
  ]
 }
 Body {
  id=#22
  parent_id=#13
  possessed_class=SkeletalMeshComponentBudgeted
  tracks [{
    type=MovieSceneObjectPropertyTrack
    display_name=Overlay Material
    sections [{
      type=MovieSceneObjectPropertySection
      channels [None {
        type=MovieSceneScriptingObjectPathChannel
        default="<Object '&MI_GrillMoon_Overlay_Mystic_Body' (0x85051d1a000) Class 'MaterialInstanceConstant'>"
        const [
         "<Object '&MI_GrillMoon_Overlay_Mystic_Body' (0x85051d1a000) Class 'MaterialInstanceConstant'>"
         1
        ]
       }]
     }]
   }]
 }
 "BP_LowAltitudeClouds_GrillMoon (7)" {
  id=#23
  possessed_class=BP_LowAltitudeClouds_GrillMoon_C
  tracks [
   {type=MovieSceneSpawnTrack, sections [{type=MovieSceneSpawnSection, channels [None {type=MovieSceneScriptingBoolChannel, default=true}]}]}
   {
    type=MovieScene3DTransformTrack
    sections [{
      type=MovieScene3DTransformSection
      channels [
       Location.X {type=MovieSceneScriptingDoubleChannel, default=-93298.5343}
       Location.Y {type=MovieSceneScriptingDoubleChannel, default=66727.8793}
       Location.Z {type=MovieSceneScriptingDoubleChannel, default=11822.8805}
       Rotation.X {type=MovieSceneScriptingDoubleChannel, default=0}
       Rotation.Y {type=MovieSceneScriptingDoubleChannel, default=0}
       Rotation.Z {type=MovieSceneScriptingDoubleChannel, default=0}
       Scale.X {type=MovieSceneScriptingDoubleChannel, default=20}
       Scale.Y {type=MovieSceneScriptingDoubleChannel, default=20}
       Scale.Z {type=MovieSceneScriptingDoubleChannel, default=20}
      ]
     }]
   }
  ]
 }
 "BP_LowAltitudeClouds_GrillMoon (3)" {
  id=#24
  possessed_class=BP_LowAltitudeClouds_GrillMoon_C
  tracks [
   {type=MovieSceneSpawnTrack, sections [{type=MovieSceneSpawnSection, channels [None {type=MovieSceneScriptingBoolChannel, default=true}]}]}
   {
    type=MovieScene3DTransformTrack
    sections [{
      type=MovieScene3DTransformSection
      channels [
       Location.X {type=MovieSceneScriptingDoubleChannel, default=-87355.3463}
       Location.Y {type=MovieSceneScriptingDoubleChannel, default=71553.3763}
       Location.Z {type=MovieSceneScriptingDoubleChannel, default=12587.1946}
       Rotation.X {type=MovieSceneScriptingDoubleChannel, default=0}
       Rotation.Y {type=MovieSceneScriptingDoubleChannel, default=0}
       Rotation.Z {type=MovieSceneScriptingDoubleChannel, default=0}
       Scale.X {type=MovieSceneScriptingDoubleChannel, default=45}
       Scale.Y {type=MovieSceneScriptingDoubleChannel, default=45}
       Scale.Z {type=MovieSceneScriptingDoubleChannel, default=45}
      ]
     }]
   }
  ]
 }
 "BP_LowAltitudeClouds_GrillMoon (12)" {
  id=#25
  possessed_class=BP_LowAltitudeClouds_GrillMoon_C
  tracks [
   {type=MovieSceneSpawnTrack, sections [{type=MovieSceneSpawnSection, channels [None {type=MovieSceneScriptingBoolChannel, default=true}]}]}
   {
    type=MovieScene3DTransformTrack
    sections [{
      type=MovieScene3DTransformSection
      channels [
       Location.X {type=MovieSceneScriptingDoubleChannel, default=-3749.493}
       Location.Y {type=MovieSceneScriptingDoubleChannel, default=38513.2319}
       Location.Z {type=MovieSceneScriptingDoubleChannel, default=13437.4746}
       Rotation.X {type=MovieSceneScriptingDoubleChannel, default=0}
       Rotation.Y {type=MovieSceneScriptingDoubleChannel, default=0}
       Rotation.Z {type=MovieSceneScriptingDoubleChannel, default=0}
       Scale.X {type=MovieSceneScriptingDoubleChannel, default=60.6694}
       Scale.Y {type=MovieSceneScriptingDoubleChannel, default=70.6694}
       Scale.Z {type=MovieSceneScriptingDoubleChannel, default=52.6706}
      ]
     }]
   }
  ]
 }
 "BP_LowAltitudeClouds_GrillMoon (4)" {
  id=#26
  possessed_class=BP_LowAltitudeClouds_GrillMoon_C
  tracks [
   {type=MovieSceneSpawnTrack, sections [{type=MovieSceneSpawnSection, channels [None {type=MovieSceneScriptingBoolChannel, default=true}]}]}
   {
    type=MovieScene3DTransformTrack
    sections [{
      type=MovieScene3DTransformSection
      channels [
       Location.X {type=MovieSceneScriptingDoubleChannel, default=-58574.5591}
       Location.Y {type=MovieSceneScriptingDoubleChannel, default=51771.386}
       Location.Z {type=MovieSceneScriptingDoubleChannel, default=10910.1252}
       Rotation.X {type=MovieSceneScriptingDoubleChannel, default=0}
       Rotation.Y {type=MovieSceneScriptingDoubleChannel, default=0}
       Rotation.Z {type=MovieSceneScriptingDoubleChannel, default=0.8908}
       Scale.X {type=MovieSceneScriptingDoubleChannel, default=73.2187}
       Scale.Y {type=MovieSceneScriptingDoubleChannel, default=17.2036}
       Scale.Z {type=MovieSceneScriptingDoubleChannel, default=17.2036}
      ]
     }]
   }
  ]
 }
 Face {
  id=#27
  parent_id=#13
  possessed_class=SkeletalMeshComponentBudgeted
  tracks [{
    type=MovieSceneObjectPropertyTrack
    display_name=Overlay Material
    sections [{
      type=MovieSceneObjectPropertySection
      channels [None {
        type=MovieSceneScriptingObjectPathChannel
        default="<Object '&MI_GrillMoon_Overlay_Mystic_Acc' (0x85051d1b000) Class 'MaterialInstanceConstant'>"
        const ["<Object '&MI_GrillMoon_Overlay_Mystic_Acc' (0x85051d1b000) Class 'MaterialInstanceConstant'>", 1]
       }]
     }]
   }]
 }
 "BP_LowAltitudeClouds_GrillMoon (7)2" {
  id=#28
  possessed_class=BP_LowAltitudeClouds_GrillMoon_C
  tracks [
   {type=MovieSceneSpawnTrack, sections [{type=MovieSceneSpawnSection, channels [None {type=MovieSceneScriptingBoolChannel, default=true}]}]}
   {
    type=MovieScene3DTransformTrack
    sections [{
      type=MovieScene3DTransformSection
      channels [
       Location.X {type=MovieSceneScriptingDoubleChannel, default=-75242.5343}
       Location.Y {type=MovieSceneScriptingDoubleChannel, default=56015.8793}
       Location.Z {type=MovieSceneScriptingDoubleChannel, default=11890.8805}
       Rotation.X {type=MovieSceneScriptingDoubleChannel, default=0}
       Rotation.Y {type=MovieSceneScriptingDoubleChannel, default=0}
       Rotation.Z {type=MovieSceneScriptingDoubleChannel, default=-109.7087}
       Scale.X {type=MovieSceneScriptingDoubleChannel, default=19.5226}
       Scale.Y {type=MovieSceneScriptingDoubleChannel, default=19.5226}
       Scale.Z {type=MovieSceneScriptingDoubleChannel, default=19.5226}
      ]
     }]
   }
  ]
 }
]
display_rate {denominator=1, numerator=60}
tick_resolution {denominator=1, numerator=24000}
tracks [{type=MovieSceneSubTrack, display_name=Subsequences, sections [{type=MovieSceneSubSection, end_frame=13880, start_frame=-301}]}]
