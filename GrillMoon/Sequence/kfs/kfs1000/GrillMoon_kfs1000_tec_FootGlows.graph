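# Level Sequence for GrillMoon shot kfs1000 that drives foot and body glows for the Demon and MaskMaker on a huge fight plane.
# It spawns BP_FootGlowFightPlane and keys its mesh material's per-foot overrides, body overrides and radii.
# An MPC_GrillMoon track keys body glow positions, alongside a repeating Event_MPCUpdater event on the plane.
# Trigger events also fire on the plane and two mannequins, MistMylar and UnluckyRoll, the latter via Event_AddTag.
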LevelSequence /GrillMoon/Sequence/kfs/kfs1000/GrillMoon_kfs1000_tec_FootGlows
bindings [
 BP_FootGlowFightPlane {
  id=#1
  possessed_class=BP_FootGlowFightPlane_C
  tracks [
   {type=MovieSceneSpawnTrack, sections [{type=MovieSceneSpawnSection, channels [None {type=MovieSceneScriptingBoolChannel, default=true}]}]}
   {
    type=MovieScene3DTransformTrack
    sections [{
      type=MovieScene3DTransformSection
      channels [
       Location.X {type=MovieSceneScriptingDoubleChannel, default=-55385.9037}
       Location.Y {type=MovieSceneScriptingDoubleChannel, default=63250.4813}
       Location.Z {type=MovieSceneScriptingDoubleChannel, default=0}
       Rotation.X {type=MovieSceneScriptingDoubleChannel, default=0}
       Rotation.Y {type=MovieSceneScriptingDoubleChannel, default=0}
       Rotation.Z {type=MovieSceneScriptingDoubleChannel, default=-18}
       Scale.X {type=MovieSceneScriptingDoubleChannel, default=1}
       Scale.Y {type=MovieSceneScriptingDoubleChannel, default=1}
       Scale.Z {type=MovieSceneScriptingDoubleChannel, default=1}
      ]
     }]
   }
   {
    type=MovieSceneEventTrack
    sections [{
      type=MovieSceneEventTriggerSection
      channels [None {type=MovieSceneScriptingEventChannel, const ["<Struct 'MovieSceneEvent' (0x8502445cda0) {}>", 1]}]
     }]
   }
   {type=MovieSceneEventTrack, display_name=Event_MPCUpdater, sections [{type=MovieSceneEventRepeaterSection, end_frame=5630, start_frame=-301}]}
  ]
 }
 Plane {
  id=#2
  parent_id=#1
  possessed_class=StaticMeshComponent
  tracks [
   {
    type=MovieScene3DTransformTrack
    sections [{
      type=MovieScene3DTransformSection
      channels [
       Location.X {type=MovieSceneScriptingDoubleChannel, default=0, const [0,1]}
       Location.Y {type=MovieSceneScriptingDoubleChannel, default=0, const [0,1]}
       Location.Z {type=MovieSceneScriptingDoubleChannel, default=10500, const [10500,1]}
       Rotation.X {type=MovieSceneScriptingDoubleChannel, default=0, const [0,1]}
       Rotation.Y {type=MovieSceneScriptingDoubleChannel, default=0, const [0,1]}
       Rotation.Z {type=MovieSceneScriptingDoubleChannel, default=0, const [0,1]}
       Scale.X {type=MovieSceneScriptingDoubleChannel, default=1000, const [1000,1]}
       Scale.Y {type=MovieSceneScriptingDoubleChannel, default=700, const [700,1]}
       Scale.Z {type=MovieSceneScriptingDoubleChannel, default=1000, const [1000,1]}
      ]
     }]
   }
   {
    type=MovieSceneComponentMaterialTrack
    display_name=Material Slot: lambert1
    sections [{
      type=MovieSceneComponentMaterialParameterSection
      end_frame=13872
      start_frame=-305
      channels [
       Demon.FootOverride.Left {
        type=MovieSceneScriptingFloatChannel
        default=-1
        thinned_from=25
        keys [[-278,-1],[-40,0],[1558,0],[1577,-1],[1629,-1],[1641,0],[1800,0],[1807,-1],[1844,-1],[1854,0],[2519,0],[2545,-1],[2594,-1],[2599,0],[3228,0],[3232,-1],[3256,-1],[3267,0],[3624,0],[3641,-1],[3671,-1],[3682,0],[4608,0],[5034,-1]]
       }
       Demon.FootOverride.Right {
        type=MovieSceneScriptingFloatChannel
        default=-1
        thinned_from=29
        keys [[-278,-1],[-40,0],[1558,0],[1577,-1],[1629,-1],[1641,0],[1774,0],[1780,-1],[1791,-1],[1795,0],[2519,0],[2545,-1],[2594,-1],[2599,0],[2931,0],[2936,-1],[2990,-1],[2998,0],[3245,0],[3253,-1],[3285,-1],[3293,0],[3604,0],[3624,-1],[3671,-1],[3682,0],[4608,0],[5034,-1]]
       }
       MaskMaker.FootOverride.Left {
        type=MovieSceneScriptingFloatChannel
        default=-1
        thinned_from=23
        keys [[-278,-1],[-40,0],[371,0],[385,-1],[416,-1],[429,0],[1774,0],[1791,-1],[1864,-1],[2668,0],[2688,-1],[2716,-1],[2740,0],[3046,0],[3052,-1],[3078,-1],[3082,0],[3293,0],[3303,-1],[3318,-1],[3323,0]]
       }
       MaskMaker.FootOverride.Right {
        type=MovieSceneScriptingFloatChannel
        default=-1
        thinned_from=25
        keys [[-278,-1],[-40,0],[371,0],[385,-1],[416,-1],[429,0],[1139,0],[1153,-1],[1187,0],[1774,0],[1780,-1],[1864,-1],[2668,0],[2688,-1],[2716,-1],[2740,0],[3044,0],[3052,-1],[3078,-1],[3082,0],[3293,0],[3303,-1],[3318,-1],[3323,0]]
       }
       Demon.BodyOverride {type=MovieSceneScriptingFloatChannel, default=0, keys [[-278,0],[1606,0],[1619,1],[1629,1],[1641,0],[1765,0],[1774,0.5],[1780,0.5],[1791,0]]}
       MaskMaker.BodyOverride {
        type=MovieSceneScriptingFloatChannel
        default=0.7
        thinned_from=14
        keys [[-278,0],[1827,0],[1842,1],[2242,1],[2302,0],[2688,0],[2716,1],[2752,1],[2782,0],[3279,0],[3299,1],[3314,0.7],[3327,0]]
       }
       Demon.Body.Radius {type=MovieSceneScriptingFloatChannel, default=0, keys [[1606,0],[1619,0],[1629,4000],[1637,4000],[1640,0]]}
       MaskMaker.Body.Radius {
        type=MovieSceneScriptingFloatChannel
        default=0
        keys [[1606,0],[1827,0],[1842,4000],[2302,4000],[2400,0],[2668,0],[2688,4000],[2782,4000],[2800,0],[3276,0],[3299,4000],[3323,4000],[3330,0]]
       }
      ]
     }]
   }
  ]
 }
 GrillMoon_MistMylar {
  id=#3
  possessed_class=FortPlayerMannequin
  tracks [{
    type=MovieSceneEventTrack
    sections [{
      type=MovieSceneEventTriggerSection
      channels [None {type=MovieSceneScriptingEventChannel, const ["<Struct 'MovieSceneEvent' (0x85024451a40) {}>", 1]}]
     }]
   }]
 }
 GrillMoon_UnluckyRoll {
  id=#4
  possessed_class=FortPlayerMannequin
  tracks [{
    type=MovieSceneEventTrack
    display_name=Event_AddTag
    sections [{
      type=MovieSceneEventTriggerSection
      channels [None {type=MovieSceneScriptingEventChannel, const ["<Struct 'MovieSceneEvent' (0x850244573a0) {}>", 1]}]
     }]
   }]
 }
]
display_rate {denominator=1, numerator=60}
tick_resolution {denominator=1, numerator=24000}
tracks [{
  type=MovieSceneMaterialParameterCollectionTrack
  display_name=MPC_GrillMoon
  sections [{
    type=MovieSceneParameterSection
    channels [
     DemonMaker_Body_PosR {
      type=MovieSceneScriptingFloatChannel
      default=-48067.1875
      keys [[603,0],[1577,0],[1615,-63612.3125],[1641,-57673.9648],[1763,-57673.9648],[1765,-48067.1875],[1774,-48067.1875]]
     }
     DemonMaker_Body_PosG {
      type=MovieSceneScriptingFloatChannel
      default=67232.3906
      keys [[603,0],[1577,0],[1615,86419.6953],[1641,82274.7266],[1763,82274.7266],[1765,67232.3906],[1774,67232.3906]]
     }
     DemonMaker_Body_PosB {type=MovieSceneScriptingFloatChannel, default=10500, const [10500,1]}
     DemonMaker_Body_PosA {type=MovieSceneScriptingFloatChannel, const [0,1]}
     MaskMaker_Body_PosR {
      type=MovieSceneScriptingFloatChannel
      default=-77086.5547
      thinned_from=20
      keys [[-301,0],[603,0],[1827,-68574.9297],[1842,-72015.3047],[1883,-77155.9688],[2302,-77155.9688],[2684,-65539.3125],[2688,-70342.4766],[2752,-70342.4766],[2782,-68942.6641],[2786,0],[3279,0],[3282,-33488.5898],[3290,-33488.5898],[3299,-29740.2363],[3310,-27719.8438],[3327,-21759.2832]]
     }
     MaskMaker_Body_PosG {
      type=MovieSceneScriptingFloatChannel
      default=82483.9688
      thinned_from=19
      keys [[-301,0],[603,0],[1827,80996.2891],[1842,80971.9922],[1883,85059.8359],[2062,85059.8359],[2183,82483.9688],[2302,85059.8359],[2688,57814.0508],[2716,60342.4219],[2782,60342.4219],[2786,0],[3279,0],[3282,40965.2578],[3290,40965.2578],[3299,39386.4961],[3310,38136.5781],[3327,33715.9062]]
     }
     MaskMaker_Body_PosB {type=MovieSceneScriptingFloatChannel, default=10500, const [10500,1]}
     MaskMaker_Body_PosA {type=MovieSceneScriptingFloatChannel, const [0,1]}
    ]
   }]
 }]
