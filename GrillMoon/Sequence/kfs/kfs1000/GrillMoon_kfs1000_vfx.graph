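# VFX Level Sequence for GrillMoon cinematic shot kfs1000, timing Niagara effects for what appears to be a demon-themed fight.
# Spawn tracks window each effect actor: slash ribbons, dagger trails, spin flips, fireballs, teleport, dash, sparks and weapon impacts.
# Transform keys place or move the actors, while some trails attach to other objects.
# Particle tracks trigger one-shot impacts or toggle reused emitters in bursts, alongside Niagara parameter tracks for spark direction and scale.
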LevelSequence /GrillMoon/Sequence/kfs/kfs1000/GrillMoon_kfs1000_vfx
bindings [
 NS_DM_DemonMask_Fire_Prj_Impact {
  id=#1
  possessed_class=NiagaraActor
  tracks [
   {
    type=MovieSceneSpawnTrack
    sections [{
      type=MovieSceneSpawnSection
      channels [None {type=MovieSceneScriptingBoolChannel, default=true, keys [[2447,false],[2461,true],[2601,false],[3487,false],[3508,true],[3810,false]]}]
     }]
   }
   {
    type=MovieScene3DTransformTrack
    sections [{
      type=MovieScene3DTransformSection
      channels [
       Location.X {
        type=MovieSceneScriptingDoubleChannel
        default=-46396.0273
        keys [[2469,-59284],[3517,-42199.3516],[3548,-46396.0273],[3677,-39316.4688],[3730,-217597.8125]]
       }
       Location.Y {
        type=MovieSceneScriptingDoubleChannel
        default=59789.9023
        keys [[2469,66629.1562],[3517,60472.7891],[3548,59789.9023],[3677,62131.707],[3730,140516.0938]]
       }
       Location.Z {
        type=MovieSceneScriptingDoubleChannel
        default=20470.6387
        keys [[2469,21658.9961],[3517,21356.3164],[3548,20470.6387],[3677,18691.7949],[3730,20187.5957]]
       }
       Rotation.X {type=MovieSceneScriptingDoubleChannel, default=0, const [0,2]}
       Rotation.Y {type=MovieSceneScriptingDoubleChannel, default=0, const [0,2]}
       Rotation.Z {type=MovieSceneScriptingDoubleChannel, default=0, const [0,2]}
       Scale.X {type=MovieSceneScriptingDoubleChannel, default=1, const [1,2]}
       Scale.Y {type=MovieSceneScriptingDoubleChannel, default=1, const [1,2]}
       Scale.Z {type=MovieSceneScriptingDoubleChannel, default=1, const [1,2]}
      ]
     }]
   }
  ]
 }
 "NS_GrillMoonFireBall (2)" {
  id=#2
  possessed_class=NiagaraActor
  tracks [
   {
    type=MovieSceneSpawnTrack
    sections [{
      type=MovieSceneSpawnSection
      channels [None {type=MovieSceneScriptingBoolChannel, default=true, keys [[3398,false],[3496,true],[3588,false]]}]
     }]
   }
   {
    type=MovieScene3DTransformTrack
    sections [{
      type=MovieScene3DTransformSection
      channels [
       Location.X {type=MovieSceneScriptingDoubleChannel, default=-29493.4471, keys [[3508,-29948.3945],[3518,-41801.793]]}
       Location.Y {type=MovieSceneScriptingDoubleChannel, default=41723.152, keys [[3508,43476.4922],[3518,60677.1133]]}
       Location.Z {type=MovieSceneScriptingDoubleChannel, default=20423.2333, keys [[3508,20531.168],[3518,21095.7188]]}
       Rotation.X {type=MovieSceneScriptingDoubleChannel, default=0, keys [[3508,0],[3518,0]]}
       Rotation.Y {type=MovieSceneScriptingDoubleChannel, default=-7.2424, const [-7.2424,2]}
       Rotation.Z {type=MovieSceneScriptingDoubleChannel, default=138.6439, const [138.6439,2]}
       Scale.X {type=MovieSceneScriptingDoubleChannel, default=1, const [1,2]}
       Scale.Y {type=MovieSceneScriptingDoubleChannel, default=1, const [1,2]}
       Scale.Z {type=MovieSceneScriptingDoubleChannel, default=1, const [1,2]}
      ]
     }]
   }
  ]
 }
 NiagaraComponent0 {
  id=#3
  parent_id=#4
  possessed_class=NiagaraComponent
  tracks [
   {
    type=MovieSceneParticleTrack
    display_name=FX System
    sections [{
      type=MovieSceneParticleSection
      channels [None {
        type=MovieSceneScriptingParticleChannel
        keys [
         [662, "<ParticleKey.ACTIVATE: 0>"]
         [685, "<ParticleKey.DEACTIVATE: 1>"]
         [726, "<ParticleKey.ACTIVATE: 0>"]
         [749, "<ParticleKey.DEACTIVATE: 1>"]
         [1733, "<ParticleKey.ACTIVATE: 0>"]
         [1756, "<ParticleKey.DEACTIVATE: 1>"]
         [2636, "<ParticleKey.ACTIVATE: 0>"]
         [2659, "<ParticleKey.DEACTIVATE: 1>"]
         [2774, "<ParticleKey.ACTIVATE: 0>"]
         [2797, "<ParticleKey.DEACTIVATE: 1>"]
        ]
       }]
     }]
   }
   {
    type=MovieSceneNiagaraFloatParameterTrack
    display_name=UseFinalSlash
    sections [{type=MovieSceneFloatSection, channels [None {type=MovieSceneScriptingFloatChannel, default=0, const [0,2]}]}]
   }
  ]
 }
 NiagaraComponent0 {
  id=#5
  parent_id=#6
  possessed_class=NiagaraComponent
  tracks [
   {
    type=MovieSceneNiagaraFloatParameterTrack
    display_name=System_FX Scale
    sections [{type=MovieSceneFloatSection, channels [None {type=MovieSceneScriptingFloatChannel, default=100}]}]
   }
   {
    type=MovieSceneNiagaraFloatParameterTrack
    display_name=System_Effect Speed
    sections [{type=MovieSceneFloatSection, channels [None {type=MovieSceneScriptingFloatChannel, default=1}]}]
   }
   {
    type=MovieSceneParticleTrack
    display_name=FX System
    sections [{
      type=MovieSceneParticleSection
      channels [None {
        type=MovieSceneScriptingParticleChannel
        keys [
         [-9, "<ParticleKey.DEACTIVATE: 1>"]
         [339, "<ParticleKey.DEACTIVATE: 1>"]
         [366, "<ParticleKey.ACTIVATE: 0>"]
         [426, "<ParticleKey.DEACTIVATE: 1>"]
         [1580, "<ParticleKey.ACTIVATE: 0>"]
         [1623, "<ParticleKey.DEACTIVATE: 1>"]
         [2669, "<ParticleKey.ACTIVATE: 0>"]
         [2719, "<ParticleKey.DEACTIVATE: 1>"]
         [3010, "<ParticleKey.ACTIVATE: 0>"]
         [3071, "<ParticleKey.DEACTIVATE: 1>"]
         [3282, "<ParticleKey.ACTIVATE: 0>"]
         [3361, "<ParticleKey.DEACTIVATE: 1>"]
        ]
       }]
     }]
   }
  ]
 }
 NS_GrillMoon_KaijuSparkStream {
  id=#7
  possessed_class=NiagaraActor
  tracks [
   {
    type=MovieSceneSpawnTrack
    sections [{type=MovieSceneSpawnSection, channels [None {type=MovieSceneScriptingBoolChannel, default=false, keys [[626,false],[638,true],[790,false]]}]}]
   }
   {
    type=MovieScene3DTransformTrack
    sections [{
      type=MovieScene3DTransformSection
      channels [
       Location.X {type=MovieSceneScriptingDoubleChannel, default=-67742.5704, thinned_from=3, keys [[640,-67742.5703],[654,-67574.2969]]}
       Location.Y {type=MovieSceneScriptingDoubleChannel, default=59932.6885, keys [[640,59932.6875],[647,56996.5898],[654,52381.3984]]}
       Location.Z {type=MovieSceneScriptingDoubleChannel, default=23680.8043, keys [[640,23680.8047],[647,23862.6484],[654,24642.0117]]}
       Rotation.X {type=MovieSceneScriptingDoubleChannel, default=0, thinned_from=3, keys [[640,0],[654,0]]}
       Rotation.Y {type=MovieSceneScriptingDoubleChannel, default=28.0485, thinned_from=3, keys [[640,28.0485],[654,11.0828]]}
       Rotation.Z {type=MovieSceneScriptingDoubleChannel, default=-103.4027, thinned_from=3, keys [[640,-103.4027],[654,-70.8236]]}
       Scale.X {type=MovieSceneScriptingDoubleChannel, default=1, const [1,3]}
       Scale.Y {type=MovieSceneScriptingDoubleChannel, default=1, const [1,3]}
       Scale.Z {type=MovieSceneScriptingDoubleChannel, default=1, const [1,3]}
      ]
     }]
   }
  ]
 }
 NiagaraComponent0 {
  id=#8
  parent_id=#9
  possessed_class=NiagaraComponent
  tracks [
   {
    type=MovieSceneParticleTrack
    display_name=FX System
    sections [{type=MovieSceneParticleSection, channels [None {type=MovieSceneScriptingParticleChannel, const ["<ParticleKey.TRIGGER: 2>", 1]}]}]
   }
   {
    type=MovieSceneNiagaraVectorParameterTrack
    display_name=SparkDirA
    sections [{
      type=MovieSceneFloatVectorSection
      channels [
       Vector.X {type=MovieSceneScriptingFloatChannel, default=100}
       Vector.Y {type=MovieSceneScriptingFloatChannel, default=100}
       Vector.Z {type=MovieSceneScriptingFloatChannel, default=0}
      ]
     }]
   }
   {
    type=MovieSceneNiagaraVectorParameterTrack
    display_name=SparkDirB
    sections [{
      type=MovieSceneFloatVectorSection
      channels [
       Vector.X {type=MovieSceneScriptingFloatChannel, default=-100}
       Vector.Y {type=MovieSceneScriptingFloatChannel, default=0}
       Vector.Z {type=MovieSceneScriptingFloatChannel, default=50}
      ]
     }]
   }
  ]
 }
 NiagaraComponent0 {
  id=#10
  parent_id=#11
  possessed_class=NiagaraComponent
  tracks [
   {
    type=MovieSceneParticleTrack
    display_name=FX System
    sections [{type=MovieSceneParticleSection, channels [None {type=MovieSceneScriptingParticleChannel, const ["<ParticleKey.TRIGGER: 2>", 1]}]}]
   }
   {
    type=MovieSceneNiagaraVectorParameterTrack
    display_name=SparkDirA
    sections [{
      type=MovieSceneFloatVectorSection
      channels [
       Vector.X {type=MovieSceneScriptingFloatChannel, default=100}
       Vector.Y {type=MovieSceneScriptingFloatChannel, default=100}
       Vector.Z {type=MovieSceneScriptingFloatChannel, default=0}
      ]
     }]
   }
   {
    type=MovieSceneNiagaraVectorParameterTrack
    display_name=SparkDirB
    sections [{
      type=MovieSceneFloatVectorSection
      channels [
       Vector.X {type=MovieSceneScriptingFloatChannel, default=-100}
       Vector.Y {type=MovieSceneScriptingFloatChannel, default=0}
       Vector.Z {type=MovieSceneScriptingFloatChannel, default=50}
      ]
     }]
   }
  ]
 }
 "NS_GrillMoon_SmallWeaponsWeaponImpact (2) (2) (2) (3) (2) (2) (2)" {
  id=#12
  possessed_class=NiagaraActor
  tracks [
   {
    type=MovieSceneSpawnTrack
    sections [{
      type=MovieSceneSpawnSection
      channels [None {type=MovieSceneScriptingBoolChannel, default=true, keys [[3161,false],[3176,true],[3254,false]]}]
     }]
   }
   {
    type=MovieScene3DTransformTrack
    sections [{
      type=MovieScene3DTransformSection
      channels [
       Location.X {type=MovieSceneScriptingDoubleChannel, default=0, const [-53490.0977,1]}
       Location.Y {type=MovieSceneScriptingDoubleChannel, default=0, const [57535.2305,1]}
       Location.Z {type=MovieSceneScriptingDoubleChannel, default=0, const [24789.5488,1]}
       Rotation.X {type=MovieSceneScriptingDoubleChannel, default=0, const [0,1]}
       Rotation.Y {type=MovieSceneScriptingDoubleChannel, default=0, const [0,1]}
       Rotation.Z {type=MovieSceneScriptingDoubleChannel, default=0, const [0,1]}
       Scale.X {type=MovieSceneScriptingDoubleChannel, default=1, const [1,1]}
       Scale.Y {type=MovieSceneScriptingDoubleChannel, default=1, const [1,1]}
       Scale.Z {type=MovieSceneScriptingDoubleChannel, default=1, const [1,1]}
      ]
     }]
   }
  ]
 }
 NE_SystemShare {
  id=#13
  parent_id=#14
  possessed_class=NiagaraComponent
  tracks [{
    type=MovieSceneVisibilityTrack
    sections [{type=MovieSceneVisibilitySection, channels [None {type=MovieSceneScriptingBoolChannel, default=true}]}]
   }]
 }
 NiagaraComponent0 {
  id=#15
  parent_id=#16
  possessed_class=NiagaraComponent
  tracks [{
    type=MovieSceneParticleTrack
    display_name=FX System
    sections [{type=MovieSceneParticleSection, channels [None {type=MovieSceneScriptingParticleChannel, const ["<ParticleKey.TRIGGER: 2>", 1]}]}]
   }]
 }
 Asteria_HLOD0/Hermes_Terrain_MainGrid_L0_X-1_Y-2 {id=#17, possessed_class=FortMainHLOD, tracks []}
 NiagaraComponent0 {
  id=#18
  parent_id=#12
  possessed_class=NiagaraComponent
  tracks [
   {
    type=MovieSceneParticleTrack
    display_name=FX System
    sections [{type=MovieSceneParticleSection, channels [None {type=MovieSceneScriptingParticleChannel, const ["<ParticleKey.TRIGGER: 2>", 1]}]}]
   }
   {
    type=MovieSceneNiagaraVectorParameterTrack
    display_name=SparkDirA
    sections [{
      type=MovieSceneFloatVectorSection
      channels [
       Vector.X {type=MovieSceneScriptingFloatChannel, default=100}
       Vector.Y {type=MovieSceneScriptingFloatChannel, default=100}
       Vector.Z {type=MovieSceneScriptingFloatChannel, default=0}
      ]
     }]
   }
   {
    type=MovieSceneNiagaraVectorParameterTrack
    display_name=SparkDirB
    sections [{
      type=MovieSceneFloatVectorSection
      channels [
       Vector.X {type=MovieSceneScriptingFloatChannel, default=-100}
       Vector.Y {type=MovieSceneScriptingFloatChannel, default=0}
       Vector.Z {type=MovieSceneScriptingFloatChannel, default=50}
      ]
     }]
   }
  ]
 }
 NS_GrillMoon_DaggerImpactLong {
  id=#19
  possessed_class=NiagaraActor
  tracks [
   {
    type=MovieSceneSpawnTrack
    sections [{
      type=MovieSceneSpawnSection
      channels [None {type=MovieSceneScriptingBoolChannel, default=false, keys [[4547,false],[4572,true],[4790,false]]}]
     }]
   }
   {
    type=MovieScene3DTransformTrack
    sections [{
      type=MovieScene3DTransformSection
      channels [
       Location.X {type=MovieSceneScriptingDoubleChannel, default=-28924.0543}
       Location.Y {type=MovieSceneScriptingDoubleChannel, default=36660.814}
       Location.Z {type=MovieSceneScriptingDoubleChannel, default=22558.1619}
       Rotation.X {type=MovieSceneScriptingDoubleChannel, default=27.9635}
       Rotation.Y {type=MovieSceneScriptingDoubleChannel, default=-17.0788}
       Rotation.Z {type=MovieSceneScriptingDoubleChannel, default=-176.3823}
       Scale.X {type=MovieSceneScriptingDoubleChannel, default=1}
       Scale.Y {type=MovieSceneScriptingDoubleChannel, default=1}
       Scale.Z {type=MovieSceneScriptingDoubleChannel, default=1}
      ]
     }]
   }
  ]
 }
 "NS_GrillMoon_MediumWeaponsWeaponImpact (2) (2) (2) (2) (2)" {
  id=#20
  possessed_class=NiagaraActor
  tracks [
   {
    type=MovieSceneSpawnTrack
    sections [{
      type=MovieSceneSpawnSection
      channels [None {type=MovieSceneScriptingBoolChannel, default=true, keys [[2551,false],[2556,true],[2644,false]]}]
     }]
   }
   {
    type=MovieScene3DTransformTrack
    sections [{
      type=MovieScene3DTransformSection
      channels [
       Location.X {type=MovieSceneScriptingDoubleChannel, default=0, const [-75563.4062,1]}
       Location.Y {type=MovieSceneScriptingDoubleChannel, default=0, const [68254.6016,1]}
       Location.Z {type=MovieSceneScriptingDoubleChannel, default=0, const [26243.5605,1]}
       Rotation.X {type=MovieSceneScriptingDoubleChannel, default=0, const [0,1]}
       Rotation.Y {type=MovieSceneScriptingDoubleChannel, default=0, const [0,1]}
       Rotation.Z {type=MovieSceneScriptingDoubleChannel, default=0, const [0,1]}
       Scale.X {type=MovieSceneScriptingDoubleChannel, default=1, const [1,1]}
       Scale.Y {type=MovieSceneScriptingDoubleChannel, default=1, const [1,1]}
       Scale.Z {type=MovieSceneScriptingDoubleChannel, default=1, const [1,1]}
      ]
     }]
   }
  ]
 }
 "NS_GrillMoonFireBall (2) (3)" {
  id=#21
  possessed_class=NiagaraActor
  tracks [
   {
    type=MovieSceneSpawnTrack
    sections [{
      type=MovieSceneSpawnSection
      channels [None {type=MovieSceneScriptingBoolChannel, default=true, keys [[3569,false],[3667,true],[3723,false]]}]
     }]
   }
   {
    type=MovieScene3DTransformTrack
    sections [{
      type=MovieScene3DTransformSection
      channels [
       Location.X {type=MovieSceneScriptingDoubleChannel, default=-29493.4471, thinned_from=3, keys [[3675,-31518.0098],[3681,-37979.3242]]}
       Location.Y {type=MovieSceneScriptingDoubleChannel, default=41723.152, thinned_from=3, keys [[3675,40330.5703],[3681,61048.375]]}
       Location.Z {type=MovieSceneScriptingDoubleChannel, default=24416.4345, thinned_from=3, keys [[3675,20680.3066],[3681,18539.3672]]}
       Rotation.X {type=MovieSceneScriptingDoubleChannel, default=0}
       Rotation.Y {type=MovieSceneScriptingDoubleChannel, default=0}
       Rotation.Z {type=MovieSceneScriptingDoubleChannel, default=105.9215}
       Scale.X {type=MovieSceneScriptingDoubleChannel, default=1}
       Scale.Y {type=MovieSceneScriptingDoubleChannel, default=1}
       Scale.Z {type=MovieSceneScriptingDoubleChannel, default=1}
      ]
     }]
   }
  ]
 }
 NiagaraComponent0 {
  id=#22
  parent_id=#23
  possessed_class=NiagaraComponent
  tracks [
   {
    type=MovieSceneParticleTrack
    display_name=FX System
    sections [{type=MovieSceneParticleSection, channels [None {type=MovieSceneScriptingParticleChannel, const ["<ParticleKey.TRIGGER: 2>", 1]}]}]
   }
   {
    type=MovieSceneNiagaraVectorParameterTrack
    display_name=SparkDirA
    sections [{
      type=MovieSceneFloatVectorSection
      channels [
       Vector.X {type=MovieSceneScriptingFloatChannel, default=100}
       Vector.Y {type=MovieSceneScriptingFloatChannel, default=100}
       Vector.Z {type=MovieSceneScriptingFloatChannel, default=0}
      ]
     }]
   }
   {
    type=MovieSceneNiagaraVectorParameterTrack
    display_name=SparkDirB
    sections [{
      type=MovieSceneFloatVectorSection
      channels [
       Vector.X {type=MovieSceneScriptingFloatChannel, default=-100}
       Vector.Y {type=MovieSceneScriptingFloatChannel, default=0}
       Vector.Z {type=MovieSceneScriptingFloatChannel, default=50}
      ]
     }]
   }
  ]
 }
 NiagaraComponent0 {
  id=#24
  parent_id=#25
  possessed_class=NiagaraComponent
  tracks [
   {
    type=MovieSceneNiagaraBoolParameterTrack
    display_name=UsingMeshSlash
    sections [{type=MovieSceneBoolSection, channels [None {type=MovieSceneScriptingBoolChannel, default=false}]}]
   }
   {
    type=MovieSceneParticleTrack
    display_name=FX System
    sections [{
      type=MovieSceneParticleSection
      channels [None {
        type=MovieSceneScriptingParticleChannel
        keys [
         [-9, "<ParticleKey.DEACTIVATE: 1>"]
         [643, "<ParticleKey.ACTIVATE: 0>"]
         [663, "<ParticleKey.DEACTIVATE: 1>"]
         [685, "<ParticleKey.ACTIVATE: 0>"]
         [716, "<ParticleKey.DEACTIVATE: 1>"]
         [733, "<ParticleKey.ACTIVATE: 0>"]
         [750, "<ParticleKey.DEACTIVATE: 1>"]
         [778, "<ParticleKey.ACTIVATE: 0>"]
         [812, "<ParticleKey.DEACTIVATE: 1>"]
         [875, "<ParticleKey.ACTIVATE: 0>"]
         [890, "<ParticleKey.DEACTIVATE: 1>"]
         [910, "<ParticleKey.ACTIVATE: 0>"]
         [926, "<ParticleKey.DEACTIVATE: 1>"]
         [935, "<ParticleKey.ACTIVATE: 0>"]
         [969, "<ParticleKey.DEACTIVATE: 1>"]
         [1003, "<ParticleKey.ACTIVATE: 0>"]
         [1065, "<ParticleKey.DEACTIVATE: 1>"]
         [1315, "<ParticleKey.ACTIVATE: 0>"]
         [1327, "<ParticleKey.DEACTIVATE: 1>"]
         [1350, "<ParticleKey.ACTIVATE: 0>"]
         [1363, "<ParticleKey.DEACTIVATE: 1>"]
         [1378, "<ParticleKey.ACTIVATE: 0>"]
         [1409, "<ParticleKey.DEACTIVATE: 1>"]
         [1443, "<ParticleKey.ACTIVATE: 0>"]
         [1470, "<ParticleKey.DEACTIVATE: 1>"]
         [1671, "<ParticleKey.ACTIVATE: 0>"]
         [1694, "<ParticleKey.DEACTIVATE: 1>"]
         [2593, "<ParticleKey.ACTIVATE: 0>"]
         [2610, "<ParticleKey.DEACTIVATE: 1>"]
         [2797, "<ParticleKey.ACTIVATE: 0>"]
         [2816, "<ParticleKey.DEACTIVATE: 1>"]
         [3046, "<ParticleKey.ACTIVATE: 0>"]
         [3065, "<ParticleKey.DEACTIVATE: 1>"]
         [3114, "<ParticleKey.ACTIVATE: 0>"]
         [3144, "<ParticleKey.DEACTIVATE: 1>"]
         [3159, "<ParticleKey.ACTIVATE: 0>"]
         [3189, "<ParticleKey.DEACTIVATE: 1>"]
         [3518, "<ParticleKey.ACTIVATE: 0>"]
         [3540, "<ParticleKey.DEACTIVATE: 1>"]
         [3549, "<ParticleKey.ACTIVATE: 0>"]
         [3578, "<ParticleKey.DEACTIVATE: 1>"]
         [3671, "<ParticleKey.ACTIVATE: 0>"]
         [3693, "<ParticleKey.DEACTIVATE: 1>"]
         [3751, "<ParticleKey.ACTIVATE: 0>"]
         [3792, "<ParticleKey.DEACTIVATE: 1>"]
        ]
       }]
     }]
   }
  ]
 }
 NS_GrillMoonTeleportStart {
  id=#16
  possessed_class=NiagaraActor
  tracks [
   {
    type=MovieSceneSpawnTrack
    sections [{
      type=MovieSceneSpawnSection
      channels [None {type=MovieSceneScriptingBoolChannel, default=true, keys [[2447,false],[2513,true],[2558,false]]}]
     }]
   }
   {
    type=MovieScene3DTransformTrack
    sections [{
      type=MovieScene3DTransformSection
      channels [
       Location.X {type=MovieSceneScriptingDoubleChannel, default=-50339.1969}
       Location.Y {type=MovieSceneScriptingDoubleChannel, default=67487.327}
       Location.Z {type=MovieSceneScriptingDoubleChannel, default=19709.1164}
       Rotation.X {type=MovieSceneScriptingDoubleChannel, default=0}
       Rotation.Y {type=MovieSceneScriptingDoubleChannel, default=0}
       Rotation.Z {type=MovieSceneScriptingDoubleChannel, default=0}
       Scale.X {type=MovieSceneScriptingDoubleChannel, default=1}
       Scale.Y {type=MovieSceneScriptingDoubleChannel, default=1}
       Scale.Z {type=MovieSceneScriptingDoubleChannel, default=1}
      ]
     }]
   }
  ]
 }
 "NS_GrillMoon_SmallWeaponsWeaponImpact (2) (2) (2) (3) (2)" {
  id=#26
  possessed_class=NiagaraActor
  tracks [
   {
    type=MovieSceneSpawnTrack
    sections [{
      type=MovieSceneSpawnSection
      channels [None {type=MovieSceneScriptingBoolChannel, default=true, keys [[3109,false],[3114,true],[3202,false]]}]
     }]
   }
   {
    type=MovieScene3DTransformTrack
    sections [{
      type=MovieScene3DTransformSection
      channels [
       Location.X {type=MovieSceneScriptingDoubleChannel, default=0, const [-58308.2656,1]}
       Location.Y {type=MovieSceneScriptingDoubleChannel, default=0, const [63437.7617,1]}
       Location.Z {type=MovieSceneScriptingDoubleChannel, default=0, const [23533.1797,1]}
       Rotation.X {type=MovieSceneScriptingDoubleChannel, default=0, const [0,1]}
       Rotation.Y {type=MovieSceneScriptingDoubleChannel, default=0, const [0,1]}
       Rotation.Z {type=MovieSceneScriptingDoubleChannel, default=0, const [0,1]}
       Scale.X {type=MovieSceneScriptingDoubleChannel, default=1, const [1,1]}
       Scale.Y {type=MovieSceneScriptingDoubleChannel, default=1, const [1,1]}
       Scale.Z {type=MovieSceneScriptingDoubleChannel, default=1, const [1,1]}
      ]
     }]
   }
  ]
 }
 NiagaraComponent0 {
  id=#27
  parent_id=#28
  possessed_class=NiagaraComponent
  tracks [{
    type=MovieSceneParticleTrack
    display_name=FX System
    sections [{
      type=MovieSceneParticleSection
      channels [None {
        type=MovieSceneScriptingParticleChannel
        keys [
         [-107, "<ParticleKey.DEACTIVATE: 1>"]
         [3508, "<ParticleKey.DEACTIVATE: 1>"]
         [3535, "<ParticleKey.ACTIVATE: 0>"]
         [3552, "<ParticleKey.DEACTIVATE: 1>"]
        ]
       }]
     }]
   }]
 }
 NiagaraComponent0 {
  id=#29
  parent_id=#30
  possessed_class=NiagaraComponent
  tracks [
   {
    type=MovieSceneParticleTrack
    display_name=FX System
    sections [{type=MovieSceneParticleSection, channels [None {type=MovieSceneScriptingParticleChannel, const ["<ParticleKey.TRIGGER: 2>", 1]}]}]
   }
   {
    type=MovieSceneNiagaraVectorParameterTrack
    display_name=SparkDirA
    sections [{
      type=MovieSceneFloatVectorSection
      channels [
       Vector.X {type=MovieSceneScriptingFloatChannel, default=100}
       Vector.Y {type=MovieSceneScriptingFloatChannel, default=100}
       Vector.Z {type=MovieSceneScriptingFloatChannel, default=0}
      ]
     }]
   }
   {
    type=MovieSceneNiagaraVectorParameterTrack
    display_name=SparkDirB
    sections [{
      type=MovieSceneFloatVectorSection
      channels [
       Vector.X {type=MovieSceneScriptingFloatChannel, default=-100}
       Vector.Y {type=MovieSceneScriptingFloatChannel, default=0}
       Vector.Z {type=MovieSceneScriptingFloatChannel, default=50}
      ]
     }]
   }
  ]
 }
 "NS_GrillMoon_SmallWeaponsWeaponImpact (2) (2) (2) (2)" {
  id=#31
  possessed_class=NiagaraActor
  tracks [
   {
    type=MovieSceneSpawnTrack
    sections [{
      type=MovieSceneSpawnSection
      channels [None {type=MovieSceneScriptingBoolChannel, default=true, keys [[1327,false],[1332,true],[1451,false]]}]
     }]
   }
   {
    type=MovieScene3DTransformTrack
    sections [{
      type=MovieScene3DTransformSection
      channels [
       Location.X {type=MovieSceneScriptingDoubleChannel, default=0, const [-71021.4922,1]}
       Location.Y {type=MovieSceneScriptingDoubleChannel, default=0, const [71736.7422,1]}
       Location.Z {type=MovieSceneScriptingDoubleChannel, default=0, const [24668.8184,1]}
       Rotation.X {type=MovieSceneScriptingDoubleChannel, default=0, const [0,1]}
       Rotation.Y {type=MovieSceneScriptingDoubleChannel, default=0, const [0,1]}
       Rotation.Z {type=MovieSceneScriptingDoubleChannel, default=0, const [0,1]}
       Scale.X {type=MovieSceneScriptingDoubleChannel, default=1, const [1,1]}
       Scale.Y {type=MovieSceneScriptingDoubleChannel, default=1, const [1,1]}
       Scale.Z {type=MovieSceneScriptingDoubleChannel, default=1, const [1,1]}
      ]
     }]
   }
  ]
 }
 NS_Demon_Slash_Ribbon {
  id=#25
  possessed_class=NiagaraActor
  tracks [
   {
    type=MovieSceneSpawnTrack
    sections [{
      type=MovieSceneSpawnSection
      channels [None {
        type=MovieSceneScriptingBoolChannel
        default=true
        keys [[561,false],[588,true],[1131,false],[1258,false],[1285,true],[1808,false],[2475,false],[2502,true],[3923,false]]
       }]
     }]
   }
   {
    type=MovieScene3DTransformTrack
    sections [{
      type=MovieScene3DTransformSection
      channels [
       Location.X {type=MovieSceneScriptingDoubleChannel, default=0}
       Location.Y {type=MovieSceneScriptingDoubleChannel, default=0}
       Location.Z {type=MovieSceneScriptingDoubleChannel, default=0}
       Rotation.X {type=MovieSceneScriptingDoubleChannel, default=0}
       Rotation.Y {type=MovieSceneScriptingDoubleChannel, default=0}
       Rotation.Z {type=MovieSceneScriptingDoubleChannel, default=0}
       Scale.X {type=MovieSceneScriptingDoubleChannel, default=1}
       Scale.Y {type=MovieSceneScriptingDoubleChannel, default=1}
       Scale.Z {type=MovieSceneScriptingDoubleChannel, default=1}
      ]
     }]
   }
   {type=MovieScene3DAttachTrack, sections [{type=MovieScene3DAttachSection, end_frame=9860, start_frame=-317}]}
  ]
 }
 NS_GrillmoonDagger_Trail Left {
  id=#4
  possessed_class=NiagaraActor
  tracks [
   {
    type=MovieSceneSpawnTrack
    sections [{
      type=MovieSceneSpawnSection
      channels [None {
        type=MovieSceneScriptingBoolChannel
        default=false
        keys [[176,false],[616,true],[832,false],[1707,true],[1799,false],[2584,true],[2891,false]]
       }]
     }]
   }
   {
    type=MovieScene3DTransformTrack
    sections [{
      type=MovieScene3DTransformSection
      channels [
       Location.X {type=MovieSceneScriptingDoubleChannel, default=0}
       Location.Y {type=MovieSceneScriptingDoubleChannel, default=0}
       Location.Z {type=MovieSceneScriptingDoubleChannel, default=0}
       Rotation.X {type=MovieSceneScriptingDoubleChannel, default=0}
       Rotation.Y {type=MovieSceneScriptingDoubleChannel, default=0}
       Rotation.Z {type=MovieSceneScriptingDoubleChannel, default=0}
       Scale.X {type=MovieSceneScriptingDoubleChannel, default=1}
       Scale.Y {type=MovieSceneScriptingDoubleChannel, default=1}
       Scale.Z {type=MovieSceneScriptingDoubleChannel, default=1}
      ]
     }]
   }
   {type=MovieScene3DAttachTrack, sections [{type=MovieScene3DAttachSection, end_frame=10784, start_frame=-1091}]}
  ]
 }
 NiagaraComponent0 {
  id=#32
  parent_id=#33
  possessed_class=NiagaraComponent
  tracks [{
    type=MovieSceneParticleTrack
    display_name=FX System
    sections [{
      type=MovieSceneParticleSection
      channels [None {
        type=MovieSceneScriptingParticleChannel
        keys [
         [-9, "<ParticleKey.DEACTIVATE: 1>"]
         [5, "<ParticleKey.ACTIVATE: 0>"]
         [3001, "<ParticleKey.ACTIVATE: 0>"]
         [3099, "<ParticleKey.ACTIVATE: 0>"]
         [3155, "<ParticleKey.ACTIVATE: 0>"]
         [3972, "<ParticleKey.DEACTIVATE: 1>"]
        ]
       }]
     }]
   }]
 }
 NiagaraComponent0 {
  id=#34
  parent_id=#2
  possessed_class=NiagaraComponent
  tracks [{
    type=MovieSceneParticleTrack
    display_name=FX System
    sections [{
      type=MovieSceneParticleSection
      channels [None {
        type=MovieSceneScriptingParticleChannel
        keys [
         [-107, "<ParticleKey.DEACTIVATE: 1>"]
         [3499, "<ParticleKey.DEACTIVATE: 1>"]
         [3508, "<ParticleKey.ACTIVATE: 0>"]
         [3518, "<ParticleKey.DEACTIVATE: 1>"]
        ]
       }]
     }]
   }]
 }
 NiagaraComponent0 {
  id=#35
  parent_id=#19
  possessed_class=NiagaraComponent
  tracks [{
    type=MovieSceneParticleTrack
    display_name=FX System
    sections [{type=MovieSceneParticleSection, channels [None {type=MovieSceneScriptingParticleChannel, const ["<ParticleKey.TRIGGER: 2>", 1]}]}]
   }]
 }
 "NS_GrillMoon_SmallWeaponsWeaponImpact (2) (2) (2)" {
  id=#23
  possessed_class=NiagaraActor
  tracks [
   {
    type=MovieSceneSpawnTrack
    sections [{
      type=MovieSceneSpawnSection
      channels [None {type=MovieSceneScriptingBoolChannel, default=true, keys [[1289,false],[1294,true],[1407,false]]}]
     }]
   }
   {
    type=MovieScene3DTransformTrack
    sections [{
      type=MovieScene3DTransformSection
      channels [
       Location.X {type=MovieSceneScriptingDoubleChannel, default=0, const [-66463.3906,1]}
       Location.Y {type=MovieSceneScriptingDoubleChannel, default=0, const [63500.6328,1]}
       Location.Z {type=MovieSceneScriptingDoubleChannel, default=0, const [23081.2422,1]}
       Rotation.X {type=MovieSceneScriptingDoubleChannel, default=0, const [0,1]}
       Rotation.Y {type=MovieSceneScriptingDoubleChannel, default=0, const [0,1]}
       Rotation.Z {type=MovieSceneScriptingDoubleChannel, default=0, const [0,1]}
       Scale.X {type=MovieSceneScriptingDoubleChannel, default=1, const [1,1]}
       Scale.Y {type=MovieSceneScriptingDoubleChannel, default=1, const [1,1]}
       Scale.Z {type=MovieSceneScriptingDoubleChannel, default=1, const [1,1]}
      ]
     }]
   }
  ]
 }
 SM_GrillMoon_Countdown_SplinterBlade_A {id=#36, parent_id=#37, possessed_class=StaticMeshComponent, tracks []}
 NS_GrillMoon_DemonDash {
  id=#38
  possessed_class=NiagaraActor
  tracks [
   {
    type=MovieSceneSpawnTrack
    sections [{
      type=MovieSceneSpawnSection
      channels [None {type=MovieSceneScriptingBoolChannel, default=false, keys [[3372,false],[3724,true],[3893.675,false]]}]
     }]
   }
   {
    type=MovieScene3DTransformTrack
    sections [{
      type=MovieScene3DTransformSection
      channels [
       Location.X {type=MovieSceneScriptingDoubleChannel, default=0}
       Location.Y {type=MovieSceneScriptingDoubleChannel, default=0}
       Location.Z {type=MovieSceneScriptingDoubleChannel, default=0}
       Rotation.X {type=MovieSceneScriptingDoubleChannel, default=0}
       Rotation.Y {type=MovieSceneScriptingDoubleChannel, default=0}
       Rotation.Z {type=MovieSceneScriptingDoubleChannel, default=0}
       Scale.X {type=MovieSceneScriptingDoubleChannel, default=1}
       Scale.Y {type=MovieSceneScriptingDoubleChannel, default=1}
       Scale.Z {type=MovieSceneScriptingDoubleChannel, default=1}
      ]
     }]
   }
  ]
 }
 NiagaraComponent0 {
  id=#39
  parent_id=#26
  possessed_class=NiagaraComponent
  tracks [
   {
    type=MovieSceneParticleTrack
    display_name=FX System
    sections [{type=MovieSceneParticleSection, channels [None {type=MovieSceneScriptingParticleChannel, const ["<ParticleKey.TRIGGER: 2>", 1]}]}]
   }
   {
    type=MovieSceneNiagaraVectorParameterTrack
    display_name=SparkDirA
    sections [{
      type=MovieSceneFloatVectorSection
      channels [
       Vector.X {type=MovieSceneScriptingFloatChannel, default=100}
       Vector.Y {type=MovieSceneScriptingFloatChannel, default=100}
       Vector.Z {type=MovieSceneScriptingFloatChannel, default=0}
      ]
     }]
   }
   {
    type=MovieSceneNiagaraVectorParameterTrack
    display_name=SparkDirB
    sections [{
      type=MovieSceneFloatVectorSection
      channels [
       Vector.X {type=MovieSceneScriptingFloatChannel, default=-100}
       Vector.Y {type=MovieSceneScriptingFloatChannel, default=0}
       Vector.Z {type=MovieSceneScriptingFloatChannel, default=50}
      ]
     }]
   }
  ]
 }
 NiagaraComponent0 {
  id=#40
  parent_id=#41
  possessed_class=NiagaraComponent
  tracks [
   {
    type=MovieSceneParticleTrack
    display_name=FX System
    sections [{type=MovieSceneParticleSection, channels [None {type=MovieSceneScriptingParticleChannel, const ["<ParticleKey.TRIGGER: 2>", 1]}]}]
   }
   {
    type=MovieSceneNiagaraVectorParameterTrack
    display_name=SparkDirA
    sections [{
      type=MovieSceneFloatVectorSection
      channels [
       Vector.X {type=MovieSceneScriptingFloatChannel, default=100}
       Vector.Y {type=MovieSceneScriptingFloatChannel, default=100}
       Vector.Z {type=MovieSceneScriptingFloatChannel, default=0}
      ]
     }]
   }
   {
    type=MovieSceneNiagaraVectorParameterTrack
    display_name=SparkDirB
    sections [{
      type=MovieSceneFloatVectorSection
      channels [
       Vector.X {type=MovieSceneScriptingFloatChannel, default=-100}
       Vector.Y {type=MovieSceneScriptingFloatChannel, default=0}
       Vector.Z {type=MovieSceneScriptingFloatChannel, default=50}
      ]
     }]
   }
  ]
 }
 NiagaraComponent0 {
  id=#42
  parent_id=#43
  possessed_class=NiagaraComponent
  tracks [
   {
    type=MovieSceneParticleTrack
    display_name=FX System
    sections [{type=MovieSceneParticleSection, channels [None {type=MovieSceneScriptingParticleChannel, const ["<ParticleKey.TRIGGER: 2>", 1]}]}]
   }
   {
    type=MovieSceneNiagaraVectorParameterTrack
    display_name=SparkDirA
    sections [{
      type=MovieSceneFloatVectorSection
      channels [
       Vector.X {type=MovieSceneScriptingFloatChannel, default=100}
       Vector.Y {type=MovieSceneScriptingFloatChannel, default=100}
       Vector.Z {type=MovieSceneScriptingFloatChannel, default=0}
      ]
     }]
   }
   {
    type=MovieSceneNiagaraVectorParameterTrack
    display_name=SparkDirB
    sections [{
      type=MovieSceneFloatVectorSection
      channels [
       Vector.X {type=MovieSceneScriptingFloatChannel, default=-100}
       Vector.Y {type=MovieSceneScriptingFloatChannel, default=0}
       Vector.Z {type=MovieSceneScriptingFloatChannel, default=50}
      ]
     }]
   }
  ]
 }
 NiagaraComponent0 {
  id=#44
  parent_id=#45
  possessed_class=NiagaraComponent
  tracks [
   {
    type=MovieSceneParticleTrack
    display_name=FX System
    sections [{
      type=MovieSceneParticleSection
      channels [None {
        type=MovieSceneScriptingParticleChannel
        keys [
         [-94, "<ParticleKey.DEACTIVATE: 1>"]
         [3600, "<ParticleKey.DEACTIVATE: 1>"]
         [3627, "<ParticleKey.ACTIVATE: 0>"]
         [3735, "<ParticleKey.DEACTIVATE: 1>"]
        ]
       }]
     }]
   }
   {
    type=MovieSceneNiagaraFloatParameterTrack
    display_name=System_FX scale
    sections [{type=MovieSceneFloatSection, channels [None {type=MovieSceneScriptingFloatChannel, default=220}]}]
   }
  ]
 }
 NiagaraComponent0 {
  id=#46
  parent_id=#47
  possessed_class=NiagaraComponent
  tracks [{
    type=MovieSceneParticleTrack
    display_name=FX System
    sections [{
      type=MovieSceneParticleSection
      channels [None {
        type=MovieSceneScriptingParticleChannel
        keys [
         [104.895, "<ParticleKey.ACTIVATE: 0>"]
         [410.7, "<ParticleKey.DEACTIVATE: 1>"]
         [1978.1225, "<ParticleKey.ACTIVATE: 0>"]
         [2404.1325, "<ParticleKey.DEACTIVATE: 1>"]
         [4168, "<ParticleKey.ACTIVATE: 0>"]
         [4563.265, "<ParticleKey.DEACTIVATE: 1>"]
        ]
       }]
     }]
   }]
 }
 NS_SpinFlip {
  id=#6
  possessed_class=NiagaraActor
  tracks [
   {
    type=MovieSceneSpawnTrack
    sections [{
      type=MovieSceneSpawnSection
      channels [None {
        type=MovieSceneScriptingBoolChannel
        default=false
        keys [[204,false],[331,true],[454,false],[1526,true],[1721,false],[2648,true],[3493,false]]
       }]
     }]
   }
   {
    type=MovieScene3DTransformTrack
    sections [{
      type=MovieScene3DTransformSection
      channels [
       Location.X {
        type=MovieSceneScriptingDoubleChannel
        default=-67448.5747
        keys [[353,-86071.6173],[366,-79066.125],[397,-65959.5859],[426,-59145.7148],[1580,-80087.9142],[1603,-68389.211],[1623,-60341.4688],[2669,-70405.9454],[2687,-69442.2267],[2719,-69406.8751],[3010,-53316.1837],[3051,-51712.1173],[3072,-50358.7306],[3123,-45800.5899],[3267,-30397.0938],[3283,-36351.2812],[3311,-23854.9648],[3361,-19937.7852]]
       }
       Location.Y {
        type=MovieSceneScriptingDoubleChannel
        default=85207.03
        keys [[353,68142.4295],[366,64450.1367],[397,59548.8711],[426,57253.7305],[1580,91473.2966],[1603,88641.367],[1623,87940.5076],[2669,66659.1795],[2687,65551.6756],[2719,65922.867],[3010,63875.0427],[3051,62850.6326],[3072,63080.6013],[3123,61904.1131],[3267,52295.9724],[3283,43059.1484],[3311,37255.1289],[3361,33842.832]]
       }
       Location.Z {
        type=MovieSceneScriptingDoubleChannel
        default=12673.2059
        keys [[353,18811.8792],[366,18920.7773],[397,17526.9277],[426,16052],[1580,19137.6009],[1603,23605.9867],[1623,16084.6185],[2669,20311.4203],[2687,24272.225],[2719,16177.7709],[3010,18583.0599],[3051,27297.9125],[3072,21743.8207],[3123,21283.6468],[3267,20247.7777],[3283,17124.6406],[3311,8094.9126],[3361,6347.3521]]
       }
       Rotation.X {
        type=MovieSceneScriptingDoubleChannel
        default=-8.6429
        keys [[366,0.0793],[397,0.0626],[426,0.0247],[1580,-8.6429],[1623,-8.6429],[2669,-10.8233],[3010,10.2945],[3072,10.1305],[3123,6.645],[3267,-10.1209],[3283,-10.9424],[3311,-10.3532],[3361,-10.4777]]
       }
       Rotation.Y {
        type=MovieSceneScriptingDoubleChannel
        default=6.6785
        thinned_from=13
        keys [[366,6.6785],[1623,6.6785],[2669,1.3569],[3010,-3.4107],[3072,-3.8749],[3123,-2.5259],[3267,4.087],[3283,-0.7821],[3311,3.6166],[3361,3.0468]]
       }
       Rotation.Z {
        type=MovieSceneScriptingDoubleChannel
        default=79.4442
        thinned_from=14
        keys [[353,60.0646],[426,60.2437],[1580,79.4442],[1623,79.4442],[2669,49.085],[3010,240.1479],[3072,242.7292],[3123,213.3127],[3267,63.7896],[3283,36.9567],[3311,60.0598],[3361,58.059]]
       }
       Scale.X {type=MovieSceneScriptingDoubleChannel, default=1, const [1,11]}
       Scale.Y {type=MovieSceneScriptingDoubleChannel, default=1, const [1,11]}
       Scale.Z {type=MovieSceneScriptingDoubleChannel, default=1, const [1,11]}
      ]
     }]
   }
  ]
 }
 NiagaraComponent0 {
  id=#48
  parent_id=#49
  possessed_class=NiagaraComponent
  tracks [
   {
    type=MovieSceneParticleTrack
    display_name=FX System
    sections [{type=MovieSceneParticleSection, channels [None {type=MovieSceneScriptingParticleChannel, const ["<ParticleKey.TRIGGER: 2>", 1]}]}]
   }
   {
    type=MovieSceneNiagaraVectorParameterTrack
    display_name=SparkDirA
    sections [{
      type=MovieSceneFloatVectorSection
      channels [
       Vector.X {type=MovieSceneScriptingFloatChannel, default=100}
       Vector.Y {type=MovieSceneScriptingFloatChannel, default=100}
       Vector.Z {type=MovieSceneScriptingFloatChannel, default=0}
      ]
     }]
   }
   {
    type=MovieSceneNiagaraVectorParameterTrack
    display_name=SparkDirB
    sections [{
      type=MovieSceneFloatVectorSection
      channels [
       Vector.X {type=MovieSceneScriptingFloatChannel, default=-100}
       Vector.Y {type=MovieSceneScriptingFloatChannel, default=0}
       Vector.Z {type=MovieSceneScriptingFloatChannel, default=50}
      ]
     }]
   }
   {
    type=MovieSceneNiagaraFloatParameterTrack
    display_name=FlareFXSacle
    sections [{type=MovieSceneFloatSection, channels [None {type=MovieSceneScriptingFloatChannel, default=0.5}]}]
   }
  ]
 }
 NiagaraComponent0 {
  id=#50
  parent_id=#31
  possessed_class=NiagaraComponent
  tracks [
   {
    type=MovieSceneParticleTrack
    display_name=FX System
    sections [{type=MovieSceneParticleSection, channels [None {type=MovieSceneScriptingParticleChannel, const ["<ParticleKey.TRIGGER: 2>", 1]}]}]
   }
   {
    type=MovieSceneNiagaraVectorParameterTrack
    display_name=SparkDirA
    sections [{
      type=MovieSceneFloatVectorSection
      channels [
       Vector.X {type=MovieSceneScriptingFloatChannel, default=100}
       Vector.Y {type=MovieSceneScriptingFloatChannel, default=100}
       Vector.Z {type=MovieSceneScriptingFloatChannel, default=0}
      ]
     }]
   }
   {
    type=MovieSceneNiagaraVectorParameterTrack
    display_name=SparkDirB
    sections [{
      type=MovieSceneFloatVectorSection
      channels [
       Vector.X {type=MovieSceneScriptingFloatChannel, default=-100}
       Vector.Y {type=MovieSceneScriptingFloatChannel, default=0}
       Vector.Z {type=MovieSceneScriptingFloatChannel, default=50}
      ]
     }]
   }
  ]
 }
 NiagaraComponent0 {
  id=#51
  parent_id=#52
  possessed_class=NiagaraComponent
  tracks [{
    type=MovieSceneParticleTrack
    display_name=FX System
    sections [{type=MovieSceneParticleSection, channels [None {type=MovieSceneScriptingParticleChannel, const ["<ParticleKey.TRIGGER: 2>", 5]}]}]
   }]
 }
 NiagaraComponent0 {
  id=#53
  parent_id=#54
  possessed_class=NiagaraComponent
  tracks [
   {
    type=MovieSceneParticleTrack
    display_name=FX System
    sections [{type=MovieSceneParticleSection, channels [None {type=MovieSceneScriptingParticleChannel, const ["<ParticleKey.TRIGGER: 2>", 4]}]}]
   }
   {
    type=MovieSceneNiagaraFloatParameterTrack
    display_name=CharTint
    sections [{type=MovieSceneFloatSection, channels [None {type=MovieSceneScriptingFloatChannel, default=1, keys [[1603,0],[1831,1]]}]}]
   }
  ]
 }
 SM_GrillMoon_Countdown_SplinterBlade_A {id=#55, parent_id=#56, possessed_class=StaticMeshComponent, tracks []}
 NiagaraComponent0 {
  id=#57
  parent_id=#58
  possessed_class=NiagaraComponent
  tracks [{
    type=MovieSceneParticleTrack
    display_name=FX System
    sections [{type=MovieSceneParticleSection, channels [None {type=MovieSceneScriptingParticleChannel, const ["<ParticleKey.TRIGGER: 2>", 1]}]}]
   }]
 }
 NiagaraComponent0 {
  id=#59
  parent_id=#1
  possessed_class=NiagaraComponent
  tracks [
   {
    type=MovieSceneParticleTrack
    display_name=FX System
    sections [{type=MovieSceneParticleSection, channels [None {type=MovieSceneScriptingParticleChannel, const ["<ParticleKey.TRIGGER: 2>", 5]}]}]
   }
   {
    type=MovieSceneNiagaraFloatParameterTrack
    display_name=System_FX Scale
    sections [{type=MovieSceneFloatSection, channels [None {type=MovieSceneScriptingFloatChannel, default=75, keys [[2469,25],[3724,75]]}]}]
   }
  ]
 }
 "NS_GrillMoon_MediumWeaponsWeaponImpact (2)" {
  id=#49
  possessed_class=NiagaraActor
  tracks [
   {
    type=MovieSceneSpawnTrack
    sections [{type=MovieSceneSpawnSection, channels [None {type=MovieSceneScriptingBoolChannel, default=true, keys [[731,false],[736,true],[824,false]]}]}]
   }
   {
    type=MovieScene3DTransformTrack
    sections [{
      type=MovieScene3DTransformSection
      channels [
       Location.X {type=MovieSceneScriptingDoubleChannel, default=0, const [-56749.6719,1]}
       Location.Y {type=MovieSceneScriptingDoubleChannel, default=0, const [59816.4141,1]}
       Location.Z {type=MovieSceneScriptingDoubleChannel, default=0, const [22861.1152,1]}
       Rotation.X {type=MovieSceneScriptingDoubleChannel, default=0, const [0,1]}
       Rotation.Y {type=MovieSceneScriptingDoubleChannel, default=0, const [0,1]}
       Rotation.Z {type=MovieSceneScriptingDoubleChannel, default=0, const [0,1]}
       Scale.X {type=MovieSceneScriptingDoubleChannel, default=1, const [1,1]}
       Scale.Y {type=MovieSceneScriptingDoubleChannel, default=1, const [1,1]}
       Scale.Z {type=MovieSceneScriptingDoubleChannel, default=1, const [1,1]}
      ]
     }]
   }
  ]
 }
 NiagaraComponent0 {
  id=#60
  parent_id=#61
  possessed_class=NiagaraComponent
  tracks [
   {
    type=MovieSceneParticleTrack
    display_name=FX System
    sections [{
      type=MovieSceneParticleSection
      channels [None {
        type=MovieSceneScriptingParticleChannel
        keys [
         [117, "<ParticleKey.ACTIVATE: 0>"]
         [449.935, "<ParticleKey.DEACTIVATE: 1>"]
         [1996.1625, "<ParticleKey.ACTIVATE: 0>"]
         [2443.455, "<ParticleKey.DEACTIVATE: 1>"]
         [4186.73, "<ParticleKey.ACTIVATE: 0>"]
         [4575.6825, "<ParticleKey.DEACTIVATE: 1>"]
        ]
       }]
     }]
   }
   {
    type=MovieSceneNiagaraBoolParameterTrack
    display_name=LeftHand
    sections [{
      type=MovieSceneBoolSection
      channels [None {type=MovieSceneScriptingBoolChannel, default=false, keys [[117,true],[1988.4025,true],[4179.7575,false]]}]
     }]
   }
   {
    type=MovieSceneNiagaraFloatParameterTrack
    display_name=Flip
    sections [{
      type=MovieSceneFloatSection
      channels [None {
        type=MovieSceneScriptingFloatChannel
        default=1
        keys [[117,0],[355.5525,1],[449.935,1],[1996.1625,0],[2361.855,1],[4179.7575,0],[4502.72,1]]
       }]
     }]
   }
   {
    type=MovieSceneNiagaraFloatParameterTrack
    display_name=Erosion
    sections [{
      type=MovieSceneFloatSection
      channels [None {
        type=MovieSceneScriptingFloatChannel
        default=0
        thinned_from=9
        keys [[382.66,0],[449.935,1],[1996.1625,0],[2392.0825,0],[2443.455,1],[4179.7575,0],[4549.05,0],[4575.6825,1]]
       }]
     }]
   }
   {
    type=MovieSceneNiagaraFloatParameterTrack
    display_name=SpawnRate
    sections [{
      type=MovieSceneFloatSection
      channels [None {
        type=MovieSceneScriptingFloatChannel
        default=0
        thinned_from=13
        keys [[117,0],[284.855,1],[382.66,0],[1996.1625,0],[2239.4625,1],[2392.0825,1],[2439.275,0],[4186.73,0],[4266.0625,1],[4541.265,1],[4564.195,0]]
       }]
     }]
   }
   {
    type=MovieSceneNiagaraFloatParameterTrack
    display_name=Growth
    sections [{
      type=MovieSceneFloatSection
      channels [None {
        type=MovieSceneScriptingFloatChannel
        default=1
        keys [[117,0],[284.855,1],[1996.1625,0],[2239.4625,1],[4179.7575,0],[4186.73,0],[4266.0625,1]]
       }]
     }]
   }
  ]
 }
 NiagaraComponent0 {
  id=#62
  parent_id=#63
  possessed_class=NiagaraComponent
  tracks [
   {
    type=MovieSceneParticleTrack
    display_name=FX System
    sections [{type=MovieSceneParticleSection, channels [None {type=MovieSceneScriptingParticleChannel, const ["<ParticleKey.TRIGGER: 2>", 4]}]}]
   }
   {
    type=MovieSceneNiagaraFloatParameterTrack
    display_name=System_FX Scale
    sections [{
      type=MovieSceneFloatSection
      channels [None {type=MovieSceneScriptingFloatChannel, default=50, keys [[3508,50],[3535,50],[3627,100],[3675,50]]}]
     }]
   }
  ]
 }
 NS_GrillMoon_LargeWeaponsWeaponImpact {
  id=#64
  possessed_class=NiagaraActor
  tracks [
   {
    type=MovieSceneSpawnTrack
    sections [{
      type=MovieSceneSpawnSection
      channels [None {type=MovieSceneScriptingBoolChannel, default=false, keys [[425,false],[1030,true],[1241,false]]}]
     }]
   }
   {
    type=MovieScene3DTransformTrack
    sections [{
      type=MovieScene3DTransformSection
      channels [
       Location.X {type=MovieSceneScriptingDoubleChannel, default=5000, const [-70564.8828,1]}
       Location.Y {type=MovieSceneScriptingDoubleChannel, default=5000, const [67750.4141,1]}
       Location.Z {type=MovieSceneScriptingDoubleChannel, default=5000, const [23569.8184,1]}
       Rotation.X {type=MovieSceneScriptingDoubleChannel, default=0, const [0,1]}
       Rotation.Y {type=MovieSceneScriptingDoubleChannel, default=0, const [0,1]}
       Rotation.Z {type=MovieSceneScriptingDoubleChannel, default=0, const [0,1]}
       Scale.X {type=MovieSceneScriptingDoubleChannel, default=1, const [1,1]}
       Scale.Y {type=MovieSceneScriptingDoubleChannel, default=1, const [1,1]}
       Scale.Z {type=MovieSceneScriptingDoubleChannel, default=1, const [1,1]}
      ]
     }]
   }
  ]
 }
 NS_GrillmoonDagger_Trail Right {
  id=#65
  possessed_class=NiagaraActor
  tracks [
   {
    type=MovieSceneSpawnTrack
    sections [{
      type=MovieSceneSpawnSection
      channels [None {
        type=MovieSceneScriptingBoolChannel
        default=false
        keys [[282,false],[387,true],[1856,false],[2584,true],[2891,false],[4468,true],[4637,false]]
       }]
     }]
   }
   {
    type=MovieScene3DTransformTrack
    sections [{
      type=MovieScene3DTransformSection
      channels [
       Location.X {type=MovieSceneScriptingDoubleChannel, default=0}
       Location.Y {type=MovieSceneScriptingDoubleChannel, default=0}
       Location.Z {type=MovieSceneScriptingDoubleChannel, default=0}
       Rotation.X {type=MovieSceneScriptingDoubleChannel, default=0}
       Rotation.Y {type=MovieSceneScriptingDoubleChannel, default=0}
       Rotation.Z {type=MovieSceneScriptingDoubleChannel, default=0}
       Scale.X {type=MovieSceneScriptingDoubleChannel, default=1}
       Scale.Y {type=MovieSceneScriptingDoubleChannel, default=1}
       Scale.Z {type=MovieSceneScriptingDoubleChannel, default=1}
      ]
     }]
   }
   {type=MovieScene3DAttachTrack, sections [{type=MovieScene3DAttachSection, end_frame=17065, start_frame=-609}]}
  ]
 }
 NS_GrillMoon_Demon_Absorb_Hand {
  id=#66
  possessed_class=NiagaraActor
  tracks [
   {
    type=MovieSceneSpawnTrack
    sections [{
      type=MovieSceneSpawnSection
      channels [None {
        type=MovieSceneScriptingBoolChannel
        default=false
        keys [[-9,false],[87,true],[497.6725,false],[1982,true],[2593.89,false],[4126.9775,true],[4703.78,false]]
       }]
     }]
   }
   {
    type=MovieScene3DTransformTrack
    sections [{
      type=MovieScene3DTransformSection
      channels [
       Location.X {type=MovieSceneScriptingDoubleChannel, default=-7.3247, keys [[2437.3325,-7.3247],[4161.8,9.7667]]}
       Location.Y {type=MovieSceneScriptingDoubleChannel, default=3.9344, keys [[2437.3325,3.9344],[4161.8,-3.9587]]}
       Location.Z {type=MovieSceneScriptingDoubleChannel, default=0, keys [[2437.3325,0],[4161.8,0]]}
       Rotation.X {type=MovieSceneScriptingDoubleChannel, default=0, keys [[2437.3325,0],[4161.8,0]]}
       Rotation.Y {type=MovieSceneScriptingDoubleChannel, default=0, keys [[2437.3325,0],[4161.8,0]]}
       Rotation.Z {type=MovieSceneScriptingDoubleChannel, default=60, keys [[2437.3325,60],[4161.8,-130]]}
       Scale.X {type=MovieSceneScriptingDoubleChannel, default=1, const [1,2]}
       Scale.Y {type=MovieSceneScriptingDoubleChannel, default=1, const [1,2]}
       Scale.Z {type=MovieSceneScriptingDoubleChannel, default=1, const [1,2]}
      ]
     }]
   }
   {
    type=MovieScene3DAttachTrack
    sections [{type=MovieScene3DAttachSection, end_frame=2814, start_frame=38}, {type=MovieScene3DAttachSection, end_frame=12533, start_frame=3920}]
   }
  ]
 }
 NS_GrillMoonDaggerCombine {
  id=#58
  possessed_class=NiagaraActor
  tracks [
   {
    type=MovieSceneSpawnTrack
    sections [{
      type=MovieSceneSpawnSection
      channels [None {type=MovieSceneScriptingBoolChannel, default=false, keys [[4367,false],[4383,true],[4529,false]]}]
     }]
   }
   {
    type=MovieScene3DTransformTrack
    sections [{
      type=MovieScene3DTransformSection
      channels [
       Location.X {type=MovieSceneScriptingDoubleChannel, default=0}
       Location.Y {type=MovieSceneScriptingDoubleChannel, default=0}
       Location.Z {type=MovieSceneScriptingDoubleChannel, default=0}
       Rotation.X {type=MovieSceneScriptingDoubleChannel, default=0}
       Rotation.Y {type=MovieSceneScriptingDoubleChannel, default=0}
       Rotation.Z {type=MovieSceneScriptingDoubleChannel, default=0}
       Scale.X {type=MovieSceneScriptingDoubleChannel, default=1}
       Scale.Y {type=MovieSceneScriptingDoubleChannel, default=1}
       Scale.Z {type=MovieSceneScriptingDoubleChannel, default=1}
      ]
     }]
   }
   {type=MovieScene3DAttachTrack, sections [{type=MovieScene3DAttachSection, end_frame=17065, start_frame=-609}]}
  ]
 }
 NiagaraComponent0 {
  id=#67
  parent_id=#68
  possessed_class=NiagaraComponent
  tracks [{
    type=MovieSceneParticleTrack
    display_name=FX System
    sections [{type=MovieSceneParticleSection, channels [None {type=MovieSceneScriptingParticleChannel, const ["<ParticleKey.TRIGGER: 2>", 1]}]}]
   }]
 }
 NiagaraComponent0 {
  id=#69
  parent_id=#70
  possessed_class=NiagaraComponent
  tracks [
   {
    type=MovieSceneNiagaraBoolParameterTrack
    display_name=UsingMeshSlash
    sections [{type=MovieSceneBoolSection, channels [None {type=MovieSceneScriptingBoolChannel, default=false}]}]
   }
   {
    type=MovieSceneParticleTrack
    display_name=FX System
    sections [{
      type=MovieSceneParticleSection
      channels [None {
        type=MovieSceneScriptingParticleChannel
        keys [[-9, "<ParticleKey.DEACTIVATE: 1>"], [3154, "<ParticleKey.ACTIVATE: 0>"], [3193, "<ParticleKey.DEACTIVATE: 1>"]]
       }]
     }]
   }
  ]
 }
 "NS_GrillMoonFireBall (2) (2)" {
  id=#28
  possessed_class=NiagaraActor
  tracks [
   {
    type=MovieSceneSpawnTrack
    sections [{
      type=MovieSceneSpawnSection
      channels [None {type=MovieSceneScriptingBoolChannel, default=true, keys [[3463,false],[3508,true],[3604,false]]}]
     }]
   }
   {
    type=MovieScene3DTransformTrack
    sections [{
      type=MovieScene3DTransformSection
      channels [
       Location.X {type=MovieSceneScriptingDoubleChannel, default=-46396.0273, keys [[3535,-32004.918],[3552,-46396.0273]]}
       Location.Y {type=MovieSceneScriptingDoubleChannel, default=41723.152, keys [[3535,39534.4492],[3552,59789.9023]]}
       Location.Z {type=MovieSceneScriptingDoubleChannel, default=24416.4345, keys [[3535,21555.2578],[3552,20470.6387]]}
       Rotation.X {type=MovieSceneScriptingDoubleChannel, default=0}
       Rotation.Y {type=MovieSceneScriptingDoubleChannel, default=0}
       Rotation.Z {type=MovieSceneScriptingDoubleChannel, default=136.7644}
       Scale.X {type=MovieSceneScriptingDoubleChannel, default=1}
       Scale.Y {type=MovieSceneScriptingDoubleChannel, default=1}
       Scale.Z {type=MovieSceneScriptingDoubleChannel, default=1}
      ]
     }]
   }
  ]
 }
 NiagaraComponent0 {
  id=#71
  parent_id=#72
  possessed_class=NiagaraComponent
  tracks [
   {
    type=MovieSceneParticleTrack
    display_name=FX System
    sections [{type=MovieSceneParticleSection, channels [None {type=MovieSceneScriptingParticleChannel, const ["<ParticleKey.TRIGGER: 2>", 1]}]}]
   }
   {
    type=MovieSceneNiagaraVectorParameterTrack
    display_name=SparkDirA
    sections [{
      type=MovieSceneFloatVectorSection
      channels [
       Vector.X {type=MovieSceneScriptingFloatChannel, default=100}
       Vector.Y {type=MovieSceneScriptingFloatChannel, default=100}
       Vector.Z {type=MovieSceneScriptingFloatChannel, default=0}
      ]
     }]
   }
   {
    type=MovieSceneNiagaraVectorParameterTrack
    display_name=SparkDirB
    sections [{
      type=MovieSceneFloatVectorSection
      channels [
       Vector.X {type=MovieSceneScriptingFloatChannel, default=-100}
       Vector.Y {type=MovieSceneScriptingFloatChannel, default=0}
       Vector.Z {type=MovieSceneScriptingFloatChannel, default=50}
      ]
     }]
   }
  ]
 }
 "NS_GrillMoon_SmallWeaponsWeaponImpact (2) (2) (2) (3) (2) (2)" {
  id=#41
  possessed_class=NiagaraActor
  tracks [
   {
    type=MovieSceneSpawnTrack
    sections [{
      type=MovieSceneSpawnSection
      channels [None {type=MovieSceneScriptingBoolChannel, default=true, keys [[3150,false],[3161,true],[3243,false]]}]
     }]
   }
   {
    type=MovieScene3DTransformTrack
    sections [{
      type=MovieScene3DTransformSection
      channels [
       Location.X {type=MovieSceneScriptingDoubleChannel, default=0, const [-56530.4961,1]}
       Location.Y {type=MovieSceneScriptingDoubleChannel, default=0, const [60798.9727,1]}
       Location.Z {type=MovieSceneScriptingDoubleChannel, default=0, const [24274.0938,1]}
       Rotation.X {type=MovieSceneScriptingDoubleChannel, default=0, const [0,1]}
       Rotation.Y {type=MovieSceneScriptingDoubleChannel, default=0, const [0,1]}
       Rotation.Z {type=MovieSceneScriptingDoubleChannel, default=0, const [0,1]}
       Scale.X {type=MovieSceneScriptingDoubleChannel, default=1, const [1,1]}
       Scale.Y {type=MovieSceneScriptingDoubleChannel, default=1, const [1,1]}
       Scale.Z {type=MovieSceneScriptingDoubleChannel, default=1, const [1,1]}
      ]
     }]
   }
  ]
 }
 NS_GrillMoonFireBall Big {
  id=#45
  possessed_class=NiagaraActor
  tracks [
   {
    type=MovieSceneSpawnTrack
    sections [{
      type=MovieSceneSpawnSection
      channels [None {type=MovieSceneScriptingBoolChannel, default=true, keys [[3569,false],[3611,true],[3792,false]]}]
     }]
   }
   {
    type=MovieScene3DTransformTrack
    sections [{
      type=MovieScene3DTransformSection
      channels [
       Location.X {type=MovieSceneScriptingDoubleChannel, default=-29493.4471, keys [[3627,-31659.252],[3735,-212162.2812]]}
       Location.Y {type=MovieSceneScriptingDoubleChannel, default=41723.152, keys [[3627,42042.7227],[3735,136814.4219]]}
       Location.Z {type=MovieSceneScriptingDoubleChannel, default=24416.4345, keys [[3627,21510.5762],[3735,20221.3984]]}
       Rotation.X {type=MovieSceneScriptingDoubleChannel, default=0}
       Rotation.Y {type=MovieSceneScriptingDoubleChannel, default=0}
       Rotation.Z {type=MovieSceneScriptingDoubleChannel, default=154.9852}
       Scale.X {type=MovieSceneScriptingDoubleChannel, default=1}
       Scale.Y {type=MovieSceneScriptingDoubleChannel, default=1}
       Scale.Z {type=MovieSceneScriptingDoubleChannel, default=1}
      ]
     }]
   }
  ]
 }
 "NS_GrillMoon_SmallWeaponsWeaponImpact (2) (2) (2) (3)" {
  id=#11
  possessed_class=NiagaraActor
  tracks [
   {
    type=MovieSceneSpawnTrack
    sections [{
      type=MovieSceneSpawnSection
      channels [None {type=MovieSceneScriptingBoolChannel, default=true, keys [[1363,false],[1368,true],[1512,false]]}]
     }]
   }
   {
    type=MovieScene3DTransformTrack
    sections [{
      type=MovieScene3DTransformSection
      channels [
       Location.X {type=MovieSceneScriptingDoubleChannel, default=0, const [-81926.1406,1]}
       Location.Y {type=MovieSceneScriptingDoubleChannel, default=0, const [83356.8828,1]}
       Location.Z {type=MovieSceneScriptingDoubleChannel, default=0, const [21033.0918,1]}
       Rotation.X {type=MovieSceneScriptingDoubleChannel, default=0, const [0,1]}
       Rotation.Y {type=MovieSceneScriptingDoubleChannel, default=0, const [0,1]}
       Rotation.Z {type=MovieSceneScriptingDoubleChannel, default=0, const [0,1]}
       Scale.X {type=MovieSceneScriptingDoubleChannel, default=1, const [1,1]}
       Scale.Y {type=MovieSceneScriptingDoubleChannel, default=1, const [1,1]}
       Scale.Z {type=MovieSceneScriptingDoubleChannel, default=1, const [1,1]}
      ]
     }]
   }
  ]
 }
 NiagaraComponent0 {
  id=#73
  parent_id=#74
  possessed_class=NiagaraComponent
  tracks [
   {
    type=MovieSceneNiagaraFloatParameterTrack
    display_name=DaggerCharge0to1
    sections [{type=MovieSceneFloatSection, channels [None {type=MovieSceneScriptingFloatChannel, default=1, keys [[4383,0.3555],[4390,5],[4432,1]]}]}]
   }
   {
    type=MovieSceneNiagaraFloatParameterTrack
    display_name=CoronaFlareScale
    sections [{type=MovieSceneFloatSection, channels [None {type=MovieSceneScriptingFloatChannel, default=1, keys [[4383,1],[4400,2],[4432,1]]}]}]
   }
   {
    type=MovieSceneNiagaraFloatParameterTrack
    display_name=UseCombinedFX
    sections [{type=MovieSceneFloatSection, channels [None {type=MovieSceneScriptingFloatChannel, default=0, keys [[4380,0],[4384,1]]}]}]
   }
  ]
 }
 NiagaraComponent0 {
  id=#75
  parent_id=#76
  possessed_class=NiagaraComponent
  tracks [
   {
    type=MovieSceneParticleTrack
    display_name=FX System
    sections [{type=MovieSceneParticleSection, channels [None {type=MovieSceneScriptingParticleChannel, const ["<ParticleKey.TRIGGER: 2>", 1]}]}]
   }
   {
    type=MovieSceneNiagaraVectorParameterTrack
    display_name=SparkDirA
    sections [{
      type=MovieSceneFloatVectorSection
      channels [
       Vector.X {type=MovieSceneScriptingFloatChannel, default=100}
       Vector.Y {type=MovieSceneScriptingFloatChannel, default=100}
       Vector.Z {type=MovieSceneScriptingFloatChannel, default=0}
      ]
     }]
   }
   {
    type=MovieSceneNiagaraVectorParameterTrack
    display_name=SparkDirB
    sections [{
      type=MovieSceneFloatVectorSection
      channels [
       Vector.X {type=MovieSceneScriptingFloatChannel, default=-100}
       Vector.Y {type=MovieSceneScriptingFloatChannel, default=0}
       Vector.Z {type=MovieSceneScriptingFloatChannel, default=50}
      ]
     }]
   }
  ]
 }
 NiagaraComponent0 {
  id=#77
  parent_id=#78
  possessed_class=NiagaraComponent
  tracks [{
    type=MovieSceneParticleTrack
    display_name=FX System
    sections [{
      type=MovieSceneParticleSection
      channels [None {
        type=MovieSceneScriptingParticleChannel
        keys [
         [-267, "<ParticleKey.DEACTIVATE: 1>"]
         [2443, "<ParticleKey.DEACTIVATE: 1>"]
         [2454, "<ParticleKey.ACTIVATE: 0>"]
         [2469, "<ParticleKey.DEACTIVATE: 1>"]
        ]
       }]
     }]
   }]
 }
 "NS_GrillMoon_MediumWeaponsWeaponImpact (2) (2) (2)" {
  id=#79
  possessed_class=NiagaraActor
  tracks [
   {
    type=MovieSceneSpawnTrack
    sections [{type=MovieSceneSpawnSection, channels [None {type=MovieSceneScriptingBoolChannel, default=true, keys [[947,false],[952,true],[1040,false]]}]}]
   }
   {
    type=MovieScene3DTransformTrack
    sections [{
      type=MovieScene3DTransformSection
      channels [
       Location.X {type=MovieSceneScriptingDoubleChannel, default=0, const [-70412.2188,1]}
       Location.Y {type=MovieSceneScriptingDoubleChannel, default=0, const [65610.1562,1]}
       Location.Z {type=MovieSceneScriptingDoubleChannel, default=0, const [28360.4355,1]}
       Rotation.X {type=MovieSceneScriptingDoubleChannel, default=0, const [0,1]}
       Rotation.Y {type=MovieSceneScriptingDoubleChannel, default=0, const [0,1]}
       Rotation.Z {type=MovieSceneScriptingDoubleChannel, default=0, const [0,1]}
       Scale.X {type=MovieSceneScriptingDoubleChannel, default=1, const [1,1]}
       Scale.Y {type=MovieSceneScriptingDoubleChannel, default=1, const [1,1]}
       Scale.Z {type=MovieSceneScriptingDoubleChannel, default=1, const [1,1]}
      ]
     }]
   }
  ]
 }
 "NS_GrillMoon_MediumWeaponsWeaponImpact (2) (2)" {
  id=#30
  possessed_class=NiagaraActor
  tracks [
   {
    type=MovieSceneSpawnTrack
    sections [{type=MovieSceneSpawnSection, channels [None {type=MovieSceneScriptingBoolChannel, default=true, keys [[909,false],[914,true],[1002,false]]}]}]
   }
   {
    type=MovieScene3DTransformTrack
    sections [{
      type=MovieScene3DTransformSection
      channels [
       Location.X {type=MovieSceneScriptingDoubleChannel, default=0, const [-69884.3594,1]}
       Location.Y {type=MovieSceneScriptingDoubleChannel, default=0, const [60137.1562,1]}
       Location.Z {type=MovieSceneScriptingDoubleChannel, default=0, const [18356.0547,1]}
       Rotation.X {type=MovieSceneScriptingDoubleChannel, default=0, const [0,1]}
       Rotation.Y {type=MovieSceneScriptingDoubleChannel, default=0, const [0,1]}
       Rotation.Z {type=MovieSceneScriptingDoubleChannel, default=0, const [0,1]}
       Scale.X {type=MovieSceneScriptingDoubleChannel, default=1, const [1,1]}
       Scale.Y {type=MovieSceneScriptingDoubleChannel, default=1, const [1,1]}
       Scale.Z {type=MovieSceneScriptingDoubleChannel, default=1, const [1,1]}
      ]
     }]
   }
  ]
 }
 NS_DemonSword_IdleLoop {
  id=#33
  possessed_class=NiagaraActor
  tracks [
   {
    type=MovieSceneSpawnTrack
    sections [{
      type=MovieSceneSpawnSection
      channels [None {type=MovieSceneScriptingBoolChannel, default=false, keys [[1841,true],[1861,false],[2455,true],[2532,false],[2593,true],[3982,false]]}]
     }]
   }
   {
    type=MovieScene3DTransformTrack
    sections [{
      type=MovieScene3DTransformSection
      channels [
       Location.X {type=MovieSceneScriptingDoubleChannel, default=0}
       Location.Y {type=MovieSceneScriptingDoubleChannel, default=0}
       Location.Z {type=MovieSceneScriptingDoubleChannel, default=0}
       Rotation.X {type=MovieSceneScriptingDoubleChannel, default=0}
       Rotation.Y {type=MovieSceneScriptingDoubleChannel, default=0}
       Rotation.Z {type=MovieSceneScriptingDoubleChannel, default=0}
       Scale.X {type=MovieSceneScriptingDoubleChannel, default=1}
       Scale.Y {type=MovieSceneScriptingDoubleChannel, default=1}
       Scale.Z {type=MovieSceneScriptingDoubleChannel, default=1}
      ]
     }]
   }
   {type=MovieScene3DAttachTrack, sections [{type=MovieScene3DAttachSection, end_frame=10229, start_frame=-317}]}
  ]
 }
 "NS_DemonSword_IdleLoop (2)" {
  id=#80
  possessed_class=NiagaraActor
  tracks [
   {
    type=MovieSceneSpawnTrack
    sections [{
      type=MovieSceneSpawnSection
      channels [None {type=MovieSceneScriptingBoolChannel, default=false, keys [[3001,false],[3028,true],[3376,false]]}]
     }]
   }
   {
    type=MovieScene3DTransformTrack
    sections [{
      type=MovieScene3DTransformSection
      channels [
       Location.X {type=MovieSceneScriptingDoubleChannel, default=0}
       Location.Y {type=MovieSceneScriptingDoubleChannel, default=0}
       Location.Z {type=MovieSceneScriptingDoubleChannel, default=0}
       Rotation.X {type=MovieSceneScriptingDoubleChannel, default=0}
       Rotation.Y {type=MovieSceneScriptingDoubleChannel, default=0}
       Rotation.Z {type=MovieSceneScriptingDoubleChannel, default=0}
       Scale.X {type=MovieSceneScriptingDoubleChannel, default=1}
       Scale.Y {type=MovieSceneScriptingDoubleChannel, default=1}
       Scale.Z {type=MovieSceneScriptingDoubleChannel, default=1}
      ]
     }]
   }
   {type=MovieScene3DAttachTrack, sections [{type=MovieScene3DAttachSection, end_frame=10229, start_frame=-317}]}
  ]
 }
 NiagaraComponent0 {
  id=#81
  parent_id=#79
  possessed_class=NiagaraComponent
  tracks [
   {
    type=MovieSceneParticleTrack
    display_name=FX System
    sections [{type=MovieSceneParticleSection, channels [None {type=MovieSceneScriptingParticleChannel, const ["<ParticleKey.TRIGGER: 2>", 1]}]}]
   }
   {
    type=MovieSceneNiagaraVectorParameterTrack
    display_name=SparkDirA
    sections [{
      type=MovieSceneFloatVectorSection
      channels [
       Vector.X {type=MovieSceneScriptingFloatChannel, default=100}
       Vector.Y {type=MovieSceneScriptingFloatChannel, default=100}
       Vector.Z {type=MovieSceneScriptingFloatChannel, default=0}
      ]
     }]
   }
   {
    type=MovieSceneNiagaraVectorParameterTrack
    display_name=SparkDirB
    sections [{
      type=MovieSceneFloatVectorSection
      channels [
       Vector.X {type=MovieSceneScriptingFloatChannel, default=-100}
       Vector.Y {type=MovieSceneScriptingFloatChannel, default=0}
       Vector.Z {type=MovieSceneScriptingFloatChannel, default=50}
      ]
     }]
   }
  ]
 }
 BP_GrillMoon_Countdown_SplinterBlade_A_L {
  id=#37
  possessed_class=BP_GrillMoon_Countdown_SplinterBlade_A_C
  tracks [{type=MovieSceneSpawnTrack, sections [{type=MovieSceneSpawnSection, channels [None {type=MovieSceneScriptingBoolChannel, default=true}]}]}]
 }
 NS_GrillMoon_DaggerImpact {
  id=#52
  possessed_class=NiagaraActor
  tracks [
   {
    type=MovieSceneSpawnTrack
    sections [{
      type=MovieSceneSpawnSection
      channels [None {type=MovieSceneScriptingBoolChannel, default=false, keys [[373,false],[431,true],[1414,false],[4505,true],[4626,false]]}]
     }]
   }
   {
    type=MovieScene3DTransformTrack
    sections [{
      type=MovieScene3DTransformSection
      channels [
       Location.X {
        type=MovieSceneScriptingDoubleChannel
        default=-52036.5981
        keys [[452,-47954.8398],[672,-62207.5898],[706,-61192.6523],[1236,-61415.2227],[4514,-26157.5293]]
       }
       Location.Y {
        type=MovieSceneScriptingDoubleChannel
        default=50038.6166
        keys [[452,50135.5898],[672,59443.8047],[706,57895.668],[1236,58225.5664],[4514,35380.793]]
       }
       Location.Z {
        type=MovieSceneScriptingDoubleChannel
        default=21404.4647
        keys [[452,21592.9453],[672,17797.4336],[706,18783.543],[1236,23939.291],[4514,24767.8242]]
       }
       Rotation.X {type=MovieSceneScriptingDoubleChannel, default=0, keys [[452,0],[672,0.1226],[706,16.9128],[1236,7.8158],[4514,78.6002]]}
       Rotation.Y {type=MovieSceneScriptingDoubleChannel, default=0, keys [[452,0],[672,24.5175],[706,13.9023],[1236,-18.8851],[4514,-70.4513]]}
       Rotation.Z {
        type=MovieSceneScriptingDoubleChannel
        default=-106.4126
        keys [[452,-106.4126],[672,-268.4849],[706,-124.7989],[1236,-100.7261],[4514,-43.7128]]
       }
       Scale.X {type=MovieSceneScriptingDoubleChannel, default=1, const [1,5]}
       Scale.Y {type=MovieSceneScriptingDoubleChannel, default=1, const [1,5]}
       Scale.Z {type=MovieSceneScriptingDoubleChannel, default=1, const [1,5]}
      ]
     }]
   }
   {
    type=MovieSceneEventTrack
    sections [{
      type=MovieSceneEventTriggerSection
      channels [None {type=MovieSceneScriptingEventChannel, const ["<Struct 'MovieSceneEvent' (0x8502445da20) {}>", 1]}]
     }]
   }
  ]
 }
 NS_GrillMoon_WindImpact {
  id=#82
  possessed_class=NiagaraActor
  tracks [
   {
    type=MovieSceneSpawnTrack
    sections [{
      type=MovieSceneSpawnSection
      channels [None {type=MovieSceneScriptingBoolChannel, default=false, keys [[1081,false],[1702,true],[3466,false]]}]
     }]
   }
   {
    type=MovieScene3DTransformTrack
    sections [{
      type=MovieScene3DTransformSection
      channels [
       Location.X {
        type=MovieSceneScriptingDoubleChannel
        default=-63095.7613
        keys [[1776,-56115.0117],[1812,-58292.2305],[2950,-67364.6641],[3068,-64396.4023],[3268,-39084.6094],[3276,-38378.2188]]
       }
       Location.Y {
        type=MovieSceneScriptingDoubleChannel
        default=71424.9985
        keys [[1776,79475.4141],[1812,75152.5859],[2950,59434.8008],[3068,64390.0703],[3268,49291.4414],[3276,46966.0195]]
       }
       Location.Z {
        type=MovieSceneScriptingDoubleChannel
        default=23402.6886
        keys [[1776,11972.5723],[1812,22742.582],[2950,24122.5078],[3068,24907.3945],[3268,20823.5488],[3276,22670.5039]]
       }
       Rotation.X {
        type=MovieSceneScriptingDoubleChannel
        default=180
        keys [[1776,67.5868],[1812,140.9987],[2950,399.4828],[3068,-19.4113],[3268,329.5466],[3276,329.5466]]
       }
       Rotation.Y {
        type=MovieSceneScriptingDoubleChannel
        default=-86.7732
        thinned_from=6
        keys [[1776,-86.7732],[2950,-86.7732],[3068,0],[3268,-86.7732],[3276,-86.7732]]
       }
       Rotation.Z {type=MovieSceneScriptingDoubleChannel, default=180, thinned_from=6, keys [[1776,180],[2950,180],[3068,0],[3268,180],[3276,180]]}
       Scale.X {type=MovieSceneScriptingDoubleChannel, default=1, const [1,6]}
       Scale.Y {type=MovieSceneScriptingDoubleChannel, default=1, const [1,6]}
       Scale.Z {type=MovieSceneScriptingDoubleChannel, default=1, const [1,6]}
      ]
     }]
   }
  ]
 }
 NS_DEmonSword_Airflow {
  id=#54
  possessed_class=NiagaraActor
  tracks [
   {
    type=MovieSceneSpawnTrack
    sections [{
      type=MovieSceneSpawnSection
      channels [None {type=MovieSceneScriptingBoolChannel, default=false, keys [[1076,false],[1582,true],[3466,false]]}]
     }]
   }
   {
    type=MovieScene3DTransformTrack
    sections [{
      type=MovieScene3DTransformSection
      channels [
       Location.X {
        type=MovieSceneScriptingDoubleChannel
        default=-76400.8916
        keys [[1605,-60096.0508],[1776,-72268.0859],[2640,-71240.1797],[3293,-32113.5547]]
       }
       Location.Y {type=MovieSceneScriptingDoubleChannel, default=54972.3788, keys [[1605,83471.6875],[1776,84130.1406],[2640,63091.0703],[3293,39456.3125]]}
       Location.Z {type=MovieSceneScriptingDoubleChannel, default=12773.5509, keys [[1605,11136.3926],[1776,10123.1162],[2640,10540.9463],[3293,9885.2383]]}
       Rotation.X {type=MovieSceneScriptingDoubleChannel, default=0, const [0,4]}
       Rotation.Y {type=MovieSceneScriptingDoubleChannel, default=0, const [0,4]}
       Rotation.Z {type=MovieSceneScriptingDoubleChannel, default=0, const [0,4]}
       Scale.X {type=MovieSceneScriptingDoubleChannel, default=1, const [1,4]}
       Scale.Y {type=MovieSceneScriptingDoubleChannel, default=1, const [1,4]}
       Scale.Z {type=MovieSceneScriptingDoubleChannel, default=1, const [1,4]}
      ]
     }]
   }
  ]
 }
 NS_GM_Projectile build up {
  id=#83
  possessed_class=NiagaraActor
  tracks [
   {
    type=MovieSceneSpawnTrack
    sections [{
      type=MovieSceneSpawnSection
      channels [None {type=MovieSceneScriptingBoolChannel, default=true, keys [[3283,false],[3487,false],[3518,true],[3722,true],[3733,false],[3937,false]]}]
     }]
   }
   {
    type=MovieScene3DTransformTrack
    sections [{
      type=MovieScene3DTransformSection
      channels [
       Location.X {type=MovieSceneScriptingDoubleChannel, default=-31278.7175}
       Location.Y {type=MovieSceneScriptingDoubleChannel, default=42465.429}
       Location.Z {type=MovieSceneScriptingDoubleChannel, default=22003.8875}
       Rotation.X {type=MovieSceneScriptingDoubleChannel, default=0}
       Rotation.Y {type=MovieSceneScriptingDoubleChannel, default=0}
       Rotation.Z {type=MovieSceneScriptingDoubleChannel, default=0}
       Scale.X {type=MovieSceneScriptingDoubleChannel, default=1}
       Scale.Y {type=MovieSceneScriptingDoubleChannel, default=1}
       Scale.Z {type=MovieSceneScriptingDoubleChannel, default=1}
      ]
     }]
   }
  ]
 }
 NiagaraComponent0 {
  id=#84
  parent_id=#85
  possessed_class=NiagaraComponent
  tracks [
   {
    type=MovieSceneParticleTrack
    display_name=FX System
    sections [{type=MovieSceneParticleSection, channels [None {type=MovieSceneScriptingParticleChannel, const ["<ParticleKey.TRIGGER: 2>", 1]}]}]
   }
   {
    type=MovieSceneNiagaraVectorParameterTrack
    display_name=SparkDirA
    sections [{
      type=MovieSceneFloatVectorSection
      channels [
       Vector.X {type=MovieSceneScriptingFloatChannel, default=100}
       Vector.Y {type=MovieSceneScriptingFloatChannel, default=100}
       Vector.Z {type=MovieSceneScriptingFloatChannel, default=0}
      ]
     }]
   }
   {
    type=MovieSceneNiagaraVectorParameterTrack
    display_name=SparkDirB
    sections [{
      type=MovieSceneFloatVectorSection
      channels [
       Vector.X {type=MovieSceneScriptingFloatChannel, default=-100}
       Vector.Y {type=MovieSceneScriptingFloatChannel, default=0}
       Vector.Z {type=MovieSceneScriptingFloatChannel, default=50}
      ]
     }]
   }
  ]
 }
 SM_GrillMoon_Countdown_SplinterBlade_A {
  id=#86
  parent_id=#87
  possessed_class=StaticMeshComponent
  tracks [{
    type=MovieSceneComponentMaterialTrack
    display_name=Material Slot: MI_Quest_SplinterBlade_A
    sections [{
      type=MovieSceneComponentMaterialParameterSection
      end_frame=12749
      start_frame=0
      channels [
       "SFX_Emissive_Tint(Expand)R" {type=MovieSceneScriptingFloatChannel, keys [[4393,1],[4396,0]]}
       "SFX_Emissive_Tint(Expand)G" {type=MovieSceneScriptingFloatChannel, keys [[4393,0.0331],[4396,2.2855]]}
       "SFX_Emissive_Tint(Expand)B" {type=MovieSceneScriptingFloatChannel, keys [[4393,0.3968],[4396,3]]}
       "SFX_Emissive_Tint(Expand)A" {type=MovieSceneScriptingFloatChannel, default=0.5, const [1,2]}
      ]
     }]
   }]
 }
 NS_GrillMoonDagger Right {
  id=#74
  possessed_class=NiagaraActor
  tracks [
   {
    type=MovieSceneSpawnTrack
    sections [{type=MovieSceneSpawnSection, channels [None {type=MovieSceneScriptingBoolChannel, default=true, const [true,1]}]}]
   }
   {
    type=MovieScene3DTransformTrack
    sections [{
      type=MovieScene3DTransformSection
      channels [
       Location.X {type=MovieSceneScriptingDoubleChannel, default=0, const [0,2]}
       Location.Y {type=MovieSceneScriptingDoubleChannel, default=0, const [0,2]}
       Location.Z {type=MovieSceneScriptingDoubleChannel, default=0, const [0,2]}
       Rotation.X {type=MovieSceneScriptingDoubleChannel, default=0, const [0,2]}
       Rotation.Y {type=MovieSceneScriptingDoubleChannel, default=0, const [0,2]}
       Rotation.Z {type=MovieSceneScriptingDoubleChannel, default=0, const [0,2]}
       Scale.X {type=MovieSceneScriptingDoubleChannel, default=1, const [1,2]}
       Scale.Y {type=MovieSceneScriptingDoubleChannel, default=1, const [1,2]}
       Scale.Z {type=MovieSceneScriptingDoubleChannel, default=1, const [1,2]}
      ]
     }]
   }
   {type=MovieScene3DAttachTrack, sections [{type=MovieScene3DAttachSection, end_frame=11865, start_frame=-821}]}
  ]
 }
 NS_GrillMoon_Character_AuraBlur_Demon {
  id=#88
  possessed_class=NiagaraActor
  tracks [
   {type=MovieSceneSpawnTrack, sections [{type=MovieSceneSpawnSection, channels [None {type=MovieSceneScriptingBoolChannel, default=true}]}]}
   {
    type=MovieScene3DTransformTrack
    sections [{
      type=MovieScene3DTransformSection
      channels [
       Location.X {type=MovieSceneScriptingDoubleChannel, default=0, keys [[1140,0],[1178,1]]}
       Location.Y {type=MovieSceneScriptingDoubleChannel, default=0}
       Location.Z {type=MovieSceneScriptingDoubleChannel, default=0}
       Rotation.X {type=MovieSceneScriptingDoubleChannel, default=0}
       Rotation.Y {type=MovieSceneScriptingDoubleChannel, default=0}
       Rotation.Z {type=MovieSceneScriptingDoubleChannel, default=0}
       Scale.X {type=MovieSceneScriptingDoubleChannel, default=1}
       Scale.Y {type=MovieSceneScriptingDoubleChannel, default=1}
       Scale.Z {type=MovieSceneScriptingDoubleChannel, default=1}
      ]
     }]
   }
   {type=MovieScene3DAttachTrack, sections [{type=MovieScene3DAttachSection, end_frame=12310, start_frame=-317}]}
   {
    type=MovieSceneVisibilityTrack
    sections [{
      type=MovieSceneVisibilitySection
      channels [None {type=MovieSceneScriptingBoolChannel, default=true, keys [[2515,true],[2544,false],[2593,true]]}]
     }]
   }
  ]
 }
 NiagaraComponent0 {
  id=#89
  parent_id=#66
  possessed_class=NiagaraComponent
  tracks [
   {
    type=MovieSceneParticleTrack
    display_name=FX System
    sections [{
      type=MovieSceneParticleSection
      channels [None {
        type=MovieSceneScriptingParticleChannel
        keys [
         [117, "<ParticleKey.ACTIVATE: 0>"]
         [425.085, "<ParticleKey.DEACTIVATE: 1>"]
         [1996.1625, "<ParticleKey.ACTIVATE: 0>"]
         [2437.3325, "<ParticleKey.DEACTIVATE: 1>"]
         [4186.73, "<ParticleKey.ACTIVATE: 0>"]
         [4575.6825, "<ParticleKey.DEACTIVATE: 1>"]
        ]
       }]
     }]
   }
   {
    type=MovieSceneNiagaraBoolParameterTrack
    display_name=Spawn
    sections [{
      type=MovieSceneBoolSection
      channels [None {
        type=MovieSceneScriptingBoolChannel
        default=false
        keys [[117,false],[186.1225,true],[425.085,false],[1999,false],[2117.0975,true],[2434.4025,false],[4194.45,false],[4235.8075,true],[4557.945,false]]
       }]
     }]
   }
  ]
 }
 "NS_GrillMoon_SmallWeaponsWeaponImpact (2) (2) (2) (3) (2) (2) (2) (2)" {
  id=#72
  possessed_class=NiagaraActor
  tracks [
   {
    type=MovieSceneSpawnTrack
    sections [{
      type=MovieSceneSpawnSection
      channels [None {type=MovieSceneScriptingBoolChannel, default=true, keys [[3188.9975,false],[3193.9975,true],[3281.9975,false]]}]
     }]
   }
   {
    type=MovieScene3DTransformTrack
    sections [{
      type=MovieScene3DTransformSection
      channels [
       Location.X {type=MovieSceneScriptingDoubleChannel, default=0, const [-48393.4219,1]}
       Location.Y {type=MovieSceneScriptingDoubleChannel, default=0, const [57300.0938,1]}
       Location.Z {type=MovieSceneScriptingDoubleChannel, default=0, const [23600.4141,1]}
       Rotation.X {type=MovieSceneScriptingDoubleChannel, default=0}
       Rotation.Y {type=MovieSceneScriptingDoubleChannel, default=0}
       Rotation.Z {type=MovieSceneScriptingDoubleChannel, default=0}
       Scale.X {type=MovieSceneScriptingDoubleChannel, default=1}
       Scale.Y {type=MovieSceneScriptingDoubleChannel, default=1}
       Scale.Z {type=MovieSceneScriptingDoubleChannel, default=1}
      ]
     }]
   }
  ]
 }
 "NS_GrillMoon_MediumWeaponsWeaponImpact (2) (2) (2) (2) (2) (2)" {
  id=#76
  possessed_class=NiagaraActor
  tracks [
   {
    type=MovieSceneSpawnTrack
    sections [{
      type=MovieSceneSpawnSection
      channels [None {type=MovieSceneScriptingBoolChannel, default=true, keys [[2805,false],[2810,true],[2898,false]]}]
     }]
   }
   {
    type=MovieScene3DTransformTrack
    sections [{
      type=MovieScene3DTransformSection
      channels [
       Location.X {type=MovieSceneScriptingDoubleChannel, default=0, const [-69007.3594,1]}
       Location.Y {type=MovieSceneScriptingDoubleChannel, default=0, const [59098.4023,1]}
       Location.Z {type=MovieSceneScriptingDoubleChannel, default=0, const [22932.6777,1]}
       Rotation.X {type=MovieSceneScriptingDoubleChannel, default=0, const [0,1]}
       Rotation.Y {type=MovieSceneScriptingDoubleChannel, default=0, const [0,1]}
       Rotation.Z {type=MovieSceneScriptingDoubleChannel, default=0, const [0,1]}
       Scale.X {type=MovieSceneScriptingDoubleChannel, default=1, const [1,1]}
       Scale.Y {type=MovieSceneScriptingDoubleChannel, default=1, const [1,1]}
       Scale.Z {type=MovieSceneScriptingDoubleChannel, default=1, const [1,1]}
      ]
     }]
   }
  ]
 }
 "NS_GrillMoonFireBall (2) (4)" {
  id=#78
  possessed_class=NiagaraActor
  tracks [
   {
    type=MovieSceneSpawnTrack
    sections [{
      type=MovieSceneSpawnSection
      channels [None {type=MovieSceneScriptingBoolChannel, default=true, keys [[2362,false],[2448,true],[2546,false]]}]
     }]
   }
   {
    type=MovieScene3DTransformTrack
    sections [{
      type=MovieScene3DTransformSection
      channels [
       Location.X {type=MovieSceneScriptingDoubleChannel, default=-59281.6406, keys [[2454,-82736.3203],[2469,-59281.6406]]}
       Location.Y {type=MovieSceneScriptingDoubleChannel, default=66629.1562, keys [[2454,84885.3516],[2469,66629.1562]]}
       Location.Z {type=MovieSceneScriptingDoubleChannel, default=21658.9961, keys [[2454,15794.3174],[2469,21658.9961]]}
       Rotation.X {type=MovieSceneScriptingDoubleChannel, default=-7.3274}
       Rotation.Y {type=MovieSceneScriptingDoubleChannel, default=16.507}
       Rotation.Z {type=MovieSceneScriptingDoubleChannel, default=-25.4923}
       Scale.X {type=MovieSceneScriptingDoubleChannel, default=1}
       Scale.Y {type=MovieSceneScriptingDoubleChannel, default=1}
       Scale.Z {type=MovieSceneScriptingDoubleChannel, default=1}
      ]
     }]
   }
  ]
 }
 "NS_GrillMoon_SmallWeaponsWeaponImpact (2) (2)" {
  id=#85
  possessed_class=NiagaraActor
  tracks [
   {
    type=MovieSceneSpawnTrack
    sections [{type=MovieSceneSpawnSection, channels [None {type=MovieSceneScriptingBoolChannel, default=true, keys [[844,false],[849,true],[967,false]]}]}]
   }
   {
    type=MovieScene3DTransformTrack
    sections [{
      type=MovieScene3DTransformSection
      channels [
       Location.X {type=MovieSceneScriptingDoubleChannel, default=0, const [-61710.8398,1]}
       Location.Y {type=MovieSceneScriptingDoubleChannel, default=0, const [57953.9219,1]}
       Location.Z {type=MovieSceneScriptingDoubleChannel, default=0, const [27944.4668,1]}
       Rotation.X {type=MovieSceneScriptingDoubleChannel, default=0, const [0,1]}
       Rotation.Y {type=MovieSceneScriptingDoubleChannel, default=0, const [0,1]}
       Rotation.Z {type=MovieSceneScriptingDoubleChannel, default=0, const [0,1]}
       Scale.X {type=MovieSceneScriptingDoubleChannel, default=1, const [1,1]}
       Scale.Y {type=MovieSceneScriptingDoubleChannel, default=1, const [1,1]}
       Scale.Z {type=MovieSceneScriptingDoubleChannel, default=1, const [1,1]}
      ]
     }]
   }
  ]
 }
 NiagaraComponent0 {
  id=#90
  parent_id=#20
  possessed_class=NiagaraComponent
  tracks [
   {
    type=MovieSceneParticleTrack
    display_name=FX System
    sections [{type=MovieSceneParticleSection, channels [None {type=MovieSceneScriptingParticleChannel, const ["<ParticleKey.TRIGGER: 2>", 1]}]}]
   }
   {
    type=MovieSceneNiagaraVectorParameterTrack
    display_name=SparkDirA
    sections [{
      type=MovieSceneFloatVectorSection
      channels [
       Vector.X {type=MovieSceneScriptingFloatChannel, default=100}
       Vector.Y {type=MovieSceneScriptingFloatChannel, default=100}
       Vector.Z {type=MovieSceneScriptingFloatChannel, default=0}
      ]
     }]
   }
   {
    type=MovieSceneNiagaraVectorParameterTrack
    display_name=SparkDirB
    sections [{
      type=MovieSceneFloatVectorSection
      channels [
       Vector.X {type=MovieSceneScriptingFloatChannel, default=-100}
       Vector.Y {type=MovieSceneScriptingFloatChannel, default=0}
       Vector.Z {type=MovieSceneScriptingFloatChannel, default=50}
      ]
     }]
   }
  ]
 }
 NiagaraComponent0 {
  id=#91
  parent_id=#38
  possessed_class=NiagaraComponent
  tracks [
   {
    type=MovieSceneParticleTrack
    display_name=FX System
    sections [{type=MovieSceneParticleSection, channels [None {type=MovieSceneScriptingParticleChannel, const ["<ParticleKey.TRIGGER: 2>", 1]}]}]
   }
   {type=MovieScene3DAttachTrack, sections [{type=MovieScene3DAttachSection, end_frame=8301, start_frame=-1193}]}
  ]
 }
 SM_GrillMoon_Countdown_SplinterBlade_A {
  id=#92
  parent_id=#93
  possessed_class=StaticMeshComponent
  tracks [{
    type=MovieSceneComponentMaterialTrack
    display_name=Material Slot: MI_Quest_SplinterBlade_A
    sections [{
      type=MovieSceneComponentMaterialParameterSection
      end_frame=12749
      start_frame=0
      channels [
       "SFX_Emissive_Tint(Expand)R" {type=MovieSceneScriptingFloatChannel, keys [[4393,1],[4396,0]]}
       "SFX_Emissive_Tint(Expand)G" {type=MovieSceneScriptingFloatChannel, keys [[4393,0.0331],[4396,2.2855]]}
       "SFX_Emissive_Tint(Expand)B" {type=MovieSceneScriptingFloatChannel, keys [[4393,0.3968],[4396,3]]}
       "SFX_Emissive_Tint(Expand)A" {type=MovieSceneScriptingFloatChannel, default=0.5, const [1,2]}
      ]
     }]
   }]
 }
 NiagaraComponent0 {
  id=#94
  parent_id=#95
  possessed_class=NiagaraComponent
  tracks [
   {
    type=MovieSceneParticleTrack
    display_name=FX System
    sections [{type=MovieSceneParticleSection, channels [None {type=MovieSceneScriptingParticleChannel, const ["<ParticleKey.TRIGGER: 2>", 1]}]}]
   }
   {
    type=MovieSceneNiagaraVectorParameterTrack
    display_name=SparkDirA
    sections [{
      type=MovieSceneFloatVectorSection
      channels [
       Vector.X {type=MovieSceneScriptingFloatChannel, default=100}
       Vector.Y {type=MovieSceneScriptingFloatChannel, default=100}
       Vector.Z {type=MovieSceneScriptingFloatChannel, default=0}
      ]
     }]
   }
   {
    type=MovieSceneNiagaraVectorParameterTrack
    display_name=SparkDirB
    sections [{
      type=MovieSceneFloatVectorSection
      channels [
       Vector.X {type=MovieSceneScriptingFloatChannel, default=-100}
       Vector.Y {type=MovieSceneScriptingFloatChannel, default=0}
       Vector.Z {type=MovieSceneScriptingFloatChannel, default=50}
      ]
     }]
   }
  ]
 }
 VFX_Gate {
  id=#96
  parent_id=#14
  possessed_class=StaticMeshComponent
  tracks [{
    type=MovieSceneVisibilityTrack
    sections [{type=MovieSceneVisibilitySection, channels [None {type=MovieSceneScriptingBoolChannel, default=true}]}]
   }]
 }
 NS_BlueFire_Muzzle {
  id=#63
  possessed_class=NiagaraActor
  tracks [
   {
    type=MovieSceneSpawnTrack
    sections [{
      type=MovieSceneSpawnSection
      channels [None {type=MovieSceneScriptingBoolChannel, default=true, keys [[3304,false],[3487,true],[3810,false]]}]
     }]
   }
   {
    type=MovieScene3DTransformTrack
    sections [{
      type=MovieScene3DTransformSection
      channels [
       Location.X {type=MovieSceneScriptingDoubleChannel, default=-31282.7537}
       Location.Y {type=MovieSceneScriptingDoubleChannel, default=39266.2679}
       Location.Z {type=MovieSceneScriptingDoubleChannel, default=21401.5924}
       Rotation.X {type=MovieSceneScriptingDoubleChannel, default=0}
       Rotation.Y {type=MovieSceneScriptingDoubleChannel, default=0}
       Rotation.Z {type=MovieSceneScriptingDoubleChannel, default=53.9348}
       Scale.X {type=MovieSceneScriptingDoubleChannel, default=1}
       Scale.Y {type=MovieSceneScriptingDoubleChannel, default=1}
       Scale.Z {type=MovieSceneScriptingDoubleChannel, default=1}
      ]
     }]
   }
  ]
 }
 NiagaraComponent0 {
  id=#97
  parent_id=#7
  possessed_class=NiagaraComponent
  tracks [{
    type=MovieSceneParticleTrack
    display_name=FX System
    sections [{
      type=MovieSceneParticleSection
      channels [None {
        type=MovieSceneScriptingParticleChannel
        keys [[634, "<ParticleKey.DEACTIVATE: 1>"], [640, "<ParticleKey.ACTIVATE: 0>"], [654, "<ParticleKey.DEACTIVATE: 1>"]]
       }]
     }]
   }]
 }
 BP_GrillMoon_Countdown_SplinterBlade_R {id=#93, possessed_class=BP_GrillMoon_Countdown_SplinterBlade_A_C, tracks []}
 BP_ClayTile_DragonGate_B2 {
  id=#14
  possessed_class=BSG_ClayTile_DragonGate_B_C
  tracks [{
    type=MovieScene3DTransformTrack
    sections [{
      type=MovieScene3DTransformSection
      channels [
       Location.X {type=MovieSceneScriptingDoubleChannel, default=-57600}
       Location.Y {type=MovieSceneScriptingDoubleChannel, default=42904}
       Location.Z {type=MovieSceneScriptingDoubleChannel, default=9988}
       Rotation.X {type=MovieSceneScriptingDoubleChannel, default=0}
       Rotation.Y {type=MovieSceneScriptingDoubleChannel, default=0}
       Rotation.Z {type=MovieSceneScriptingDoubleChannel, default=180}
       Scale.X {type=MovieSceneScriptingDoubleChannel, default=2}
       Scale.Y {type=MovieSceneScriptingDoubleChannel, default=2}
       Scale.Z {type=MovieSceneScriptingDoubleChannel, default=2}
      ]
     }]
   }]
 }
 "NS_Demon_Slash_Ribbon (3)" {
  id=#70
  possessed_class=NiagaraActor
  tracks [
   {
    type=MovieSceneSpawnTrack
    sections [{
      type=MovieSceneSpawnSection
      channels [None {type=MovieSceneScriptingBoolChannel, default=true, keys [[3093,false],[3120,true],[3222,false]]}]
     }]
   }
   {
    type=MovieScene3DTransformTrack
    sections [{
      type=MovieScene3DTransformSection
      channels [
       Location.X {type=MovieSceneScriptingDoubleChannel, default=0}
       Location.Y {type=MovieSceneScriptingDoubleChannel, default=0}
       Location.Z {type=MovieSceneScriptingDoubleChannel, default=0}
       Rotation.X {type=MovieSceneScriptingDoubleChannel, default=0}
       Rotation.Y {type=MovieSceneScriptingDoubleChannel, default=0}
       Rotation.Z {type=MovieSceneScriptingDoubleChannel, default=0}
       Scale.X {type=MovieSceneScriptingDoubleChannel, default=1}
       Scale.Y {type=MovieSceneScriptingDoubleChannel, default=1}
       Scale.Z {type=MovieSceneScriptingDoubleChannel, default=1}
      ]
     }]
   }
   {type=MovieScene3DAttachTrack, sections [{type=MovieScene3DAttachSection, end_frame=8301, start_frame=-2195}]}
  ]
 }
 NS_GrillMoon_MediumWeaponsWeaponImpact {
  id=#98
  possessed_class=NiagaraActor
  tracks [
   {
    type=MovieSceneSpawnTrack
    sections [{type=MovieSceneSpawnSection, channels [None {type=MovieSceneScriptingBoolChannel, default=true, keys [[768,false],[773,true],[861,false]]}]}]
   }
   {
    type=MovieScene3DTransformTrack
    sections [{
      type=MovieScene3DTransformSection
      channels [
       Location.X {type=MovieSceneScriptingDoubleChannel, default=0, const [-55568.6016,2]}
       Location.Y {type=MovieSceneScriptingDoubleChannel, default=0, const [55135.418,2]}
       Location.Z {type=MovieSceneScriptingDoubleChannel, default=0, const [26435.8535,2]}
       Rotation.X {type=MovieSceneScriptingDoubleChannel, default=0}
       Rotation.Y {type=MovieSceneScriptingDoubleChannel, default=0}
       Rotation.Z {type=MovieSceneScriptingDoubleChannel, default=0}
       Scale.X {type=MovieSceneScriptingDoubleChannel, default=1}
       Scale.Y {type=MovieSceneScriptingDoubleChannel, default=1}
       Scale.Z {type=MovieSceneScriptingDoubleChannel, default=1}
      ]
     }]
   }
  ]
 }
 "NS_GrillMoon_MediumWeaponsWeaponImpact (2) (2) (2) (2)" {
  id=#43
  possessed_class=NiagaraActor
  tracks [
   {
    type=MovieSceneSpawnTrack
    sections [{
      type=MovieSceneSpawnSection
      channels [None {type=MovieSceneScriptingBoolChannel, default=true, keys [[1443,false],[1448,true],[1536,false]]}]
     }]
   }
   {
    type=MovieScene3DTransformTrack
    sections [{
      type=MovieScene3DTransformSection
      channels [
       Location.X {type=MovieSceneScriptingDoubleChannel, default=0, const [-86215.5859,1]}
       Location.Y {type=MovieSceneScriptingDoubleChannel, default=0, const [89046.9062,1]}
       Location.Z {type=MovieSceneScriptingDoubleChannel, default=0, const [22172.1973,1]}
       Rotation.X {type=MovieSceneScriptingDoubleChannel, default=0, const [0,1]}
       Rotation.Y {type=MovieSceneScriptingDoubleChannel, default=0, const [0,1]}
       Rotation.Z {type=MovieSceneScriptingDoubleChannel, default=0, const [0,1]}
       Scale.X {type=MovieSceneScriptingDoubleChannel, default=1, const [1,1]}
       Scale.Y {type=MovieSceneScriptingDoubleChannel, default=1, const [1,1]}
       Scale.Z {type=MovieSceneScriptingDoubleChannel, default=1, const [1,1]}
      ]
     }]
   }
  ]
 }
 NiagaraComponent0 {
  id=#99
  parent_id=#21
  possessed_class=NiagaraComponent
  tracks [{
    type=MovieSceneParticleTrack
    display_name=FX System
    sections [{
      type=MovieSceneParticleSection
      channels [None {
        type=MovieSceneScriptingParticleChannel
        keys [
         [-107, "<ParticleKey.DEACTIVATE: 1>"]
         [3667, "<ParticleKey.DEACTIVATE: 1>"]
         [3674, "<ParticleKey.ACTIVATE: 0>"]
         [3681, "<ParticleKey.DEACTIVATE: 1>"]
        ]
       }]
     }]
   }]
 }
 NS_GrillMoonDagger Left {
  id=#100
  possessed_class=NiagaraActor
  tracks [
   {
    type=MovieSceneSpawnTrack
    sections [{type=MovieSceneSpawnSection, channels [None {type=MovieSceneScriptingBoolChannel, default=false, keys [[4367,true],[4386,false]]}]}]
   }
   {
    type=MovieScene3DTransformTrack
    sections [{
      type=MovieScene3DTransformSection
      channels [
       Location.X {type=MovieSceneScriptingDoubleChannel, default=0, const [0,1]}
       Location.Y {type=MovieSceneScriptingDoubleChannel, default=0, const [0,1]}
       Location.Z {type=MovieSceneScriptingDoubleChannel, default=0, const [0,1]}
       Rotation.X {type=MovieSceneScriptingDoubleChannel, default=0, const [0,1]}
       Rotation.Y {type=MovieSceneScriptingDoubleChannel, default=0, const [0,1]}
       Rotation.Z {type=MovieSceneScriptingDoubleChannel, default=0, const [0,1]}
       Scale.X {type=MovieSceneScriptingDoubleChannel, default=1, const [1,1]}
       Scale.Y {type=MovieSceneScriptingDoubleChannel, default=1, const [1,1]}
       Scale.Z {type=MovieSceneScriptingDoubleChannel, default=1, const [1,1]}
      ]
     }]
   }
   {type=MovieScene3DAttachTrack, sections [{type=MovieScene3DAttachSection, end_frame=10784, start_frame=-1091}]}
  ]
 }
 NiagaraComponent0 {
  id=#101
  parent_id=#82
  possessed_class=NiagaraComponent
  tracks [
   {
    type=MovieSceneParticleTrack
    display_name=FX System
    sections [{type=MovieSceneParticleSection, channels [None {type=MovieSceneScriptingParticleChannel, const ["<ParticleKey.TRIGGER: 2>", 5]}]}]
   }
   {
    type=MovieSceneNiagaraFloatParameterTrack
    display_name=System_FX scale
    sections [{
      type=MovieSceneFloatSection
      channels [None {type=MovieSceneScriptingFloatChannel, default=35, keys [[1776,17],[1812,50],[3068,25],[3245,25],[3270,35]]}]
     }]
   }
   {
    type=MovieSceneNiagaraFloatParameterTrack
    display_name=CharTint
    sections [{type=MovieSceneFloatSection, channels [None {type=MovieSceneScriptingFloatChannel, default=0, keys [[2925,1],[3051,0],[3231,1]]}]}]
   }
   {
    type=MovieSceneNiagaraFloatParameterTrack
    display_name=Use Extra
    sections [{type=MovieSceneFloatSection, channels [None {type=MovieSceneScriptingFloatChannel, default=1, keys [[1810,0],[1814,1],[1820,1],[1823,0]]}]}]
   }
  ]
 }
 NiagaraComponent0 {
  id=#102
  parent_id=#64
  possessed_class=NiagaraComponent
  tracks [
   {
    type=MovieSceneParticleTrack
    display_name=FX System
    sections [{type=MovieSceneParticleSection, channels [None {type=MovieSceneScriptingParticleChannel, const ["<ParticleKey.TRIGGER: 2>", 1]}]}]
   }
   {
    type=MovieSceneNiagaraVectorParameterTrack
    display_name=SparkDirA
    sections [{
      type=MovieSceneFloatVectorSection
      channels [
       Vector.X {type=MovieSceneScriptingFloatChannel, default=100}
       Vector.Y {type=MovieSceneScriptingFloatChannel, default=100}
       Vector.Z {type=MovieSceneScriptingFloatChannel, default=0}
      ]
     }]
   }
   {
    type=MovieSceneNiagaraVectorParameterTrack
    display_name=SparkDirB
    sections [{
      type=MovieSceneFloatVectorSection
      channels [
       Vector.X {type=MovieSceneScriptingFloatChannel, default=-100}
       Vector.Y {type=MovieSceneScriptingFloatChannel, default=0}
       Vector.Z {type=MovieSceneScriptingFloatChannel, default=50}
      ]
     }]
   }
  ]
 }
 NiagaraComponent0 {
  id=#103
  parent_id=#104
  possessed_class=NiagaraComponent
  tracks [
   {
    type=MovieSceneParticleTrack
    display_name=FX System
    sections [{type=MovieSceneParticleSection, channels [None {type=MovieSceneScriptingParticleChannel, const ["<ParticleKey.TRIGGER: 2>", 1]}]}]
   }
   {type=MovieScene3DAttachTrack, sections [{type=MovieScene3DAttachSection, end_frame=8301, start_frame=-993}]}
  ]
 }
 NS_GrillMoon_Character_AuraBlur_MaskMaker {
  id=#105
  possessed_class=NiagaraActor
  tracks [
   {type=MovieSceneSpawnTrack, sections [{type=MovieSceneSpawnSection, channels [None {type=MovieSceneScriptingBoolChannel, default=true}]}]}
   {
    type=MovieScene3DTransformTrack
    sections [{
      type=MovieScene3DTransformSection
      channels [
       Location.X {type=MovieSceneScriptingDoubleChannel, default=0}
       Location.Y {type=MovieSceneScriptingDoubleChannel, default=0}
       Location.Z {type=MovieSceneScriptingDoubleChannel, default=0}
       Rotation.X {type=MovieSceneScriptingDoubleChannel, default=0}
       Rotation.Y {type=MovieSceneScriptingDoubleChannel, default=0}
       Rotation.Z {type=MovieSceneScriptingDoubleChannel, default=0}
       Scale.X {type=MovieSceneScriptingDoubleChannel, default=1}
       Scale.Y {type=MovieSceneScriptingDoubleChannel, default=1}
       Scale.Z {type=MovieSceneScriptingDoubleChannel, default=1}
      ]
     }]
   }
   {type=MovieScene3DAttachTrack, sections [{type=MovieScene3DAttachSection, end_frame=9708, start_frame=-317}]}
  ]
 }
 "NS_GrillMoon_SmallWeaponsWeaponImpact (2)" {
  id=#95
  possessed_class=NiagaraActor
  tracks [
   {
    type=MovieSceneSpawnTrack
    sections [{type=MovieSceneSpawnSection, channels [None {type=MovieSceneScriptingBoolChannel, default=true, keys [[806,false],[811,true],[928,false]]}]}]
   }
   {
    type=MovieScene3DTransformTrack
    sections [{
      type=MovieScene3DTransformSection
      channels [
       Location.X {type=MovieSceneScriptingDoubleChannel, default=0, const [-58308.1484,1]}
       Location.Y {type=MovieSceneScriptingDoubleChannel, default=0, const [55520.625,1]}
       Location.Z {type=MovieSceneScriptingDoubleChannel, default=0, const [25667.3496,1]}
       Rotation.X {type=MovieSceneScriptingDoubleChannel, default=0, const [0,1]}
       Rotation.Y {type=MovieSceneScriptingDoubleChannel, default=0, const [0,1]}
       Rotation.Z {type=MovieSceneScriptingDoubleChannel, default=0, const [0,1]}
       Scale.X {type=MovieSceneScriptingDoubleChannel, default=1, const [1,1]}
       Scale.Y {type=MovieSceneScriptingDoubleChannel, default=1, const [1,1]}
       Scale.Z {type=MovieSceneScriptingDoubleChannel, default=1, const [1,1]}
      ]
     }]
   }
  ]
 }
 BP_GrillMoon_Countdown_SplinterBlade_L {id=#56, possessed_class=BP_GrillMoon_Countdown_SplinterBlade_A_C, tracks []}
 NS_Grillmoon_Katana_Charge {
  id=#104
  possessed_class=NiagaraActor
  tracks [
   {
    type=MovieSceneSpawnTrack
    sections [{
      type=MovieSceneSpawnSection
      channels [None {type=MovieSceneScriptingBoolChannel, default=false, keys [[3372,false],[3419,true],[3900,false]]}]
     }]
   }
   {
    type=MovieScene3DTransformTrack
    sections [{
      type=MovieScene3DTransformSection
      channels [
       Location.X {type=MovieSceneScriptingDoubleChannel, default=0}
       Location.Y {type=MovieSceneScriptingDoubleChannel, default=0}
       Location.Z {type=MovieSceneScriptingDoubleChannel, default=9.597}
       Rotation.X {type=MovieSceneScriptingDoubleChannel, default=0}
       Rotation.Y {type=MovieSceneScriptingDoubleChannel, default=0}
       Rotation.Z {type=MovieSceneScriptingDoubleChannel, default=0}
       Scale.X {type=MovieSceneScriptingDoubleChannel, default=1}
       Scale.Y {type=MovieSceneScriptingDoubleChannel, default=1}
       Scale.Z {type=MovieSceneScriptingDoubleChannel, default=1}
      ]
     }]
   }
  ]
 }
 NiagaraComponent0 {
  id=#106
  parent_id=#107
  possessed_class=NiagaraComponent
  tracks [
   {
    type=MovieSceneParticleTrack
    display_name=FX System
    sections [{type=MovieSceneParticleSection, channels [None {type=MovieSceneScriptingParticleChannel, const ["<ParticleKey.TRIGGER: 2>", 1]}]}]
   }
   {
    type=MovieSceneNiagaraVectorParameterTrack
    display_name=SparkDirA
    sections [{
      type=MovieSceneFloatVectorSection
      channels [
       Vector.X {type=MovieSceneScriptingFloatChannel, default=100}
       Vector.Y {type=MovieSceneScriptingFloatChannel, default=100}
       Vector.Z {type=MovieSceneScriptingFloatChannel, default=0}
      ]
     }]
   }
   {
    type=MovieSceneNiagaraVectorParameterTrack
    display_name=SparkDirB
    sections [{
      type=MovieSceneFloatVectorSection
      channels [
       Vector.X {type=MovieSceneScriptingFloatChannel, default=-100}
       Vector.Y {type=MovieSceneScriptingFloatChannel, default=0}
       Vector.Z {type=MovieSceneScriptingFloatChannel, default=50}
      ]
     }]
   }
  ]
 }
 NS_GrillMoon_SmallWeaponsWeaponImpact {
  id=#107
  possessed_class=NiagaraActor
  tracks [
   {
    type=MovieSceneSpawnTrack
    sections [{type=MovieSceneSpawnSection, channels [None {type=MovieSceneScriptingBoolChannel, default=true, keys [[460,false],[465,true],[578,false]]}]}]
   }
   {
    type=MovieScene3DTransformTrack
    sections [{
      type=MovieScene3DTransformSection
      channels [
       Location.X {type=MovieSceneScriptingDoubleChannel, default=0, const [-52191.0234,1]}
       Location.Y {type=MovieSceneScriptingDoubleChannel, default=0, const [51989.3086,1]}
       Location.Z {type=MovieSceneScriptingDoubleChannel, default=0, const [19737.3672,1]}
       Rotation.X {type=MovieSceneScriptingDoubleChannel, default=0, const [0,1]}
       Rotation.Y {type=MovieSceneScriptingDoubleChannel, default=0, const [0,1]}
       Rotation.Z {type=MovieSceneScriptingDoubleChannel, default=0, const [0,1]}
       Scale.X {type=MovieSceneScriptingDoubleChannel, default=1, const [1,1]}
       Scale.Y {type=MovieSceneScriptingDoubleChannel, default=1, const [1,1]}
       Scale.Z {type=MovieSceneScriptingDoubleChannel, default=1, const [1,1]}
      ]
     }]
   }
  ]
 }
 NS_GrillMoon_Portal_Absorb {
  id=#108
  possessed_class=NiagaraActor
  tracks [
   {
    type=MovieSceneSpawnTrack
    sections [{
      type=MovieSceneSpawnSection
      channels [None {
        type=MovieSceneScriptingBoolChannel
        default=false
        keys [[-9,false],[87.71,true],[497.6725,false],[1981.34,true],[2541.2675,false],[4177.82,true],[4708.1075,false]]
       }]
     }]
   }
   {
    type=MovieScene3DTransformTrack
    sections [{
      type=MovieScene3DTransformSection
      channels [
       Location.X {type=MovieSceneScriptingDoubleChannel, default=-57600}
       Location.Y {type=MovieSceneScriptingDoubleChannel, default=42904}
       Location.Z {type=MovieSceneScriptingDoubleChannel, default=9988}
       Rotation.X {type=MovieSceneScriptingDoubleChannel, default=-63.4349}
       Rotation.Y {type=MovieSceneScriptingDoubleChannel, default=90}
       Rotation.Z {type=MovieSceneScriptingDoubleChannel, default=-153.4349}
       Scale.X {type=MovieSceneScriptingDoubleChannel, default=1}
       Scale.Y {type=MovieSceneScriptingDoubleChannel, default=1}
       Scale.Z {type=MovieSceneScriptingDoubleChannel, default=1}
      ]
     }]
   }
  ]
 }
 NS_GrillMoon_SwordImpact {
  id=#109
  possessed_class=NiagaraActor
  tracks [
   {type=MovieSceneSpawnTrack, sections [{type=MovieSceneSpawnSection, channels [None {type=MovieSceneScriptingBoolChannel, default=false}]}]}
   {
    type=MovieScene3DTransformTrack
    sections [{
      type=MovieScene3DTransformSection
      channels [
       Location.X {type=MovieSceneScriptingDoubleChannel, default=-10856.1198, const [-10856.1201,1]}
       Location.Y {type=MovieSceneScriptingDoubleChannel, default=42709.1481, const [42709.1484,1]}
       Location.Z {type=MovieSceneScriptingDoubleChannel, default=21187.1581, const [21187.1582,1]}
       Rotation.X {type=MovieSceneScriptingDoubleChannel, default=0, const [0,1]}
       Rotation.Y {type=MovieSceneScriptingDoubleChannel, default=22.451, const [22.451,1]}
       Rotation.Z {type=MovieSceneScriptingDoubleChannel, default=100.5534, const [100.5534,1]}
       Scale.X {type=MovieSceneScriptingDoubleChannel, default=1, const [1,1]}
       Scale.Y {type=MovieSceneScriptingDoubleChannel, default=1, const [1,1]}
       Scale.Z {type=MovieSceneScriptingDoubleChannel, default=1, const [1,1]}
      ]
     }]
   }
  ]
 }
 NS_GrillMoon_Demon_AbsorbTrail {
  id=#61
  possessed_class=NiagaraActor
  tracks [
   {
    type=MovieSceneSpawnTrack
    sections [{
      type=MovieSceneSpawnSection
      channels [None {
        type=MovieSceneScriptingBoolChannel
        default=false
        keys [[-9,false],[84.7425,true],[595.01,false],[1985.0225,true],[2671.855,false],[4176.74,true],[4734.425,false]]
       }]
     }]
   }
   {
    type=MovieScene3DTransformTrack
    sections [{
      type=MovieScene3DTransformSection
      channels [
       Location.X {type=MovieSceneScriptingDoubleChannel, default=0}
       Location.Y {type=MovieSceneScriptingDoubleChannel, default=0}
       Location.Z {type=MovieSceneScriptingDoubleChannel, default=0}
       Rotation.X {type=MovieSceneScriptingDoubleChannel, default=0}
       Rotation.Y {type=MovieSceneScriptingDoubleChannel, default=0}
       Rotation.Z {type=MovieSceneScriptingDoubleChannel, default=0}
       Scale.X {type=MovieSceneScriptingDoubleChannel, default=1}
       Scale.Y {type=MovieSceneScriptingDoubleChannel, default=1}
       Scale.Z {type=MovieSceneScriptingDoubleChannel, default=1}
      ]
     }]
   }
   {type=MovieScene3DAttachTrack, sections [{type=MovieScene3DAttachSection, end_frame=12921, start_frame=25}]}
  ]
 }
 NiagaraComponent0 {
  id=#110
  parent_id=#65
  possessed_class=NiagaraComponent
  tracks [
   {
    type=MovieSceneParticleTrack
    display_name=FX System
    sections [{
      type=MovieSceneParticleSection
      channels [None {
        type=MovieSceneScriptingParticleChannel
        keys [
         [432, "<ParticleKey.ACTIVATE: 0>"]
         [470, "<ParticleKey.DEACTIVATE: 1>"]
         [694, "<ParticleKey.ACTIVATE: 0>"]
         [732, "<ParticleKey.DEACTIVATE: 1>"]
         [778, "<ParticleKey.ACTIVATE: 0>"]
         [816, "<ParticleKey.DEACTIVATE: 1>"]
         [832, "<ParticleKey.ACTIVATE: 0>"]
         [856, "<ParticleKey.DEACTIVATE: 1>"]
         [1140, "<ParticleKey.ACTIVATE: 0>"]
         [1163, "<ParticleKey.DEACTIVATE: 1>"]
         [1222, "<ParticleKey.ACTIVATE: 0>"]
         [1258, "<ParticleKey.DEACTIVATE: 1>"]
         [1707, "<ParticleKey.ACTIVATE: 0>"]
         [1748, "<ParticleKey.DEACTIVATE: 1>"]
         [1776, "<ParticleKey.ACTIVATE: 0>"]
         [1799, "<ParticleKey.DEACTIVATE: 1>"]
         [2593, "<ParticleKey.ACTIVATE: 0>"]
         [2599, "<ParticleKey.DEACTIVATE: 1>"]
         [2648, "<ParticleKey.ACTIVATE: 0>"]
         [2671, "<ParticleKey.DEACTIVATE: 1>"]
         [2786, "<ParticleKey.ACTIVATE: 0>"]
         [2809, "<ParticleKey.DEACTIVATE: 1>"]
         [4503, "<ParticleKey.ACTIVATE: 0>"]
         [4526, "<ParticleKey.DEACTIVATE: 1>"]
         [4561, "<ParticleKey.ACTIVATE: 0>"]
         [4584, "<ParticleKey.DEACTIVATE: 1>"]
        ]
       }]
     }]
   }
   {
    type=MovieSceneNiagaraFloatParameterTrack
    display_name=UseFinalSlash
    sections [{type=MovieSceneFloatSection, channels [None {type=MovieSceneScriptingFloatChannel, default=1, keys [[4483,0],[4489,1],[4601,1]]}]}]
   }
  ]
 }
 NiagaraComponent0 {
  id=#111
  parent_id=#109
  possessed_class=NiagaraComponent
  tracks [{
    type=MovieSceneParticleTrack
    display_name=FX System
    sections [{type=MovieSceneParticleSection, channels [None {type=MovieSceneScriptingParticleChannel, const ["<ParticleKey.TRIGGER: 2>", 1]}]}]
   }]
 }
 NiagaraComponent0 {
  id=#112
  parent_id=#83
  possessed_class=NiagaraComponent
  tracks [
   {
    type=MovieSceneNiagaraFloatParameterTrack
    display_name=System_FX Scale
    sections [{type=MovieSceneFloatSection, channels [None {type=MovieSceneScriptingFloatChannel, default=0, keys [[3581,0],[3627,60]]}]}]
   }
   {
    type=MovieSceneParticleTrack
    display_name=FX System
    sections [{
      type=MovieSceneParticleSection
      channels [None {type=MovieSceneScriptingParticleChannel, keys [[3581, "<ParticleKey.ACTIVATE: 0>"], [3627, "<ParticleKey.DEACTIVATE: 1>"]]}]
     }]
   }
  ]
 }
 NiagaraComponent0 {
  id=#113
  parent_id=#100
  possessed_class=NiagaraComponent
  tracks [{
    type=MovieSceneNiagaraFloatParameterTrack
    display_name=DaggerCharge0to1
    sections [{
      type=MovieSceneFloatSection
      channels [None {
        type=MovieSceneScriptingFloatChannel
        default=0.05
        keys [[446,0.3555],[452,0.5555],[536,0.3555],[670,0.3555],[676,0.5555],[760,0.3555],[1238,0.3555],[1244,0.5555],[1328,0.3555]]
       }]
     }]
   }]
 }
 NiagaraComponent0 {
  id=#114
  parent_id=#98
  possessed_class=NiagaraComponent
  tracks [
   {
    type=MovieSceneParticleTrack
    display_name=FX System
    sections [{type=MovieSceneParticleSection, channels [None {type=MovieSceneScriptingParticleChannel, const ["<ParticleKey.TRIGGER: 2>", 1]}]}]
   }
   {
    type=MovieSceneNiagaraVectorParameterTrack
    display_name=SparkDirA
    sections [{
      type=MovieSceneFloatVectorSection
      channels [
       Vector.X {type=MovieSceneScriptingFloatChannel, default=100, const [0,3]}
       Vector.Y {type=MovieSceneScriptingFloatChannel, default=100}
       Vector.Z {type=MovieSceneScriptingFloatChannel, default=0}
      ]
     }]
   }
   {
    type=MovieSceneNiagaraVectorParameterTrack
    display_name=SparkDirB
    sections [{
      type=MovieSceneFloatVectorSection
      channels [
       Vector.X {type=MovieSceneScriptingFloatChannel, default=-100}
       Vector.Y {type=MovieSceneScriptingFloatChannel, default=0}
       Vector.Z {type=MovieSceneScriptingFloatChannel, default=50}
      ]
     }]
   }
  ]
 }
 NS_GrillMoonTeleportEnd {
  id=#68
  possessed_class=NiagaraActor
  tracks [
   {
    type=MovieSceneSpawnTrack
    sections [{
      type=MovieSceneSpawnSection
      channels [None {type=MovieSceneScriptingBoolChannel, default=true, keys [[2530,false],[2558,true],[2623,false]]}]
     }]
   }
   {
    type=MovieScene3DTransformTrack
    sections [{
      type=MovieScene3DTransformSection
      channels [
       Location.X {type=MovieSceneScriptingDoubleChannel, default=-65872.4033}
       Location.Y {type=MovieSceneScriptingDoubleChannel, default=59994.334}
       Location.Z {type=MovieSceneScriptingDoubleChannel, default=19606.6469}
       Rotation.X {type=MovieSceneScriptingDoubleChannel, default=0}
       Rotation.Y {type=MovieSceneScriptingDoubleChannel, default=0}
       Rotation.Z {type=MovieSceneScriptingDoubleChannel, default=0}
       Scale.X {type=MovieSceneScriptingDoubleChannel, default=1}
       Scale.Y {type=MovieSceneScriptingDoubleChannel, default=1}
       Scale.Z {type=MovieSceneScriptingDoubleChannel, default=1}
      ]
     }]
   }
  ]
 }
 NiagaraComponent0 {
  id=#115
  parent_id=#80
  possessed_class=NiagaraComponent
  tracks [{
    type=MovieSceneParticleTrack
    display_name=FX System
    sections [{
      type=MovieSceneParticleSection
      channels [None {
        type=MovieSceneScriptingParticleChannel
        keys [[-9, "<ParticleKey.DEACTIVATE: 1>"], [3069, "<ParticleKey.ACTIVATE: 0>"], [3325, "<ParticleKey.DEACTIVATE: 1>"]]
       }]
     }]
   }]
 }
 BP_GrillMoon_Countdown_SplinterBlade_A_R {
  id=#87
  possessed_class=BP_GrillMoon_Countdown_SplinterBlade_A_C
  tracks [{type=MovieSceneSpawnTrack, sections [{type=MovieSceneSpawnSection, channels [None {type=MovieSceneScriptingBoolChannel, default=true}]}]}]
 }
 "NS_GrillMoon_SmallWeaponsWeaponImpact (2) (2) (2) (3) (2) (2) (2) (2) (2)" {
  id=#9
  possessed_class=NiagaraActor
  tracks [
   {
    type=MovieSceneSpawnTrack
    sections [{
      type=MovieSceneSpawnSection
      channels [None {type=MovieSceneScriptingBoolChannel, default=false, keys [[1460,false],[1729,true],[1817,false]]}]
     }]
   }
   {
    type=MovieScene3DTransformTrack
    sections [{
      type=MovieScene3DTransformSection
      channels [
       Location.X {type=MovieSceneScriptingDoubleChannel, default=0, const [-53304.1758,1]}
       Location.Y {type=MovieSceneScriptingDoubleChannel, default=0, const [75540.6172,1]}
       Location.Z {type=MovieSceneScriptingDoubleChannel, default=0, const [22927.1953,1]}
       Rotation.X {type=MovieSceneScriptingDoubleChannel, default=0}
       Rotation.Y {type=MovieSceneScriptingDoubleChannel, default=0}
       Rotation.Z {type=MovieSceneScriptingDoubleChannel, default=0}
       Scale.X {type=MovieSceneScriptingDoubleChannel, default=1}
       Scale.Y {type=MovieSceneScriptingDoubleChannel, default=1}
       Scale.Z {type=MovieSceneScriptingDoubleChannel, default=1}
      ]
     }]
   }
  ]
 }
 NiagaraComponent0 {
  id=#116
  parent_id=#108
  possessed_class=NiagaraComponent
  tracks [
   {
    type=MovieSceneParticleTrack
    display_name=FX System
    sections [{
      type=MovieSceneParticleSection
      channels [None {
        type=MovieSceneScriptingParticleChannel
        keys [
         [117, "<ParticleKey.ACTIVATE: 0>"]
         [410.7, "<ParticleKey.DEACTIVATE: 1>"]
         [1991.34, "<ParticleKey.ACTIVATE: 0>"]
         [2404.1325, "<ParticleKey.DEACTIVATE: 1>"]
         [4181, "<ParticleKey.ACTIVATE: 0>"]
         [4563.265, "<ParticleKey.DEACTIVATE: 1>"]
        ]
       }]
     }]
   }
   {
    type=MovieSceneNiagaraFloatParameterTrack
    display_name=Alpha
    sections [{
      type=MovieSceneFloatSection
      channels [None {
        type=MovieSceneScriptingFloatChannel
        default=0
        keys [[117,1],[385.6725,0],[410.7,0],[1991.34,1],[2381.535,1],[2404.1325,0],[4181,1],[4549.05,1],[4575.6825,0]]
       }]
     }]
   }
  ]
 }
 NS_GrillMoon_Portal_Flash {
  id=#47
  possessed_class=NiagaraActor
  tracks [
   {
    type=MovieSceneSpawnTrack
    sections [{
      type=MovieSceneSpawnSection
      channels [None {
        type=MovieSceneScriptingBoolChannel
        default=false
        keys [[-9,false],[85.8125,true],[450.2475,false],[1929.355,false],[1974,true],[2507.3475,false],[4152,true],[4634.1025,false]]
       }]
     }]
   }
   {
    type=MovieScene3DTransformTrack
    sections [{
      type=MovieScene3DTransformSection
      channels [
       Location.X {type=MovieSceneScriptingDoubleChannel, default=-57600}
       Location.Y {type=MovieSceneScriptingDoubleChannel, default=42904}
       Location.Z {type=MovieSceneScriptingDoubleChannel, default=10400}
       Rotation.X {type=MovieSceneScriptingDoubleChannel, default=0}
       Rotation.Y {type=MovieSceneScriptingDoubleChannel, default=0}
       Rotation.Z {type=MovieSceneScriptingDoubleChannel, default=-90}
       Scale.X {type=MovieSceneScriptingDoubleChannel, default=1}
       Scale.Y {type=MovieSceneScriptingDoubleChannel, default=1}
       Scale.Z {type=MovieSceneScriptingDoubleChannel, default=1}
      ]
     }]
   }
  ]
 }
]
display_rate {denominator=1, numerator=60}
tick_resolution {denominator=1, numerator=24000}
tracks []
